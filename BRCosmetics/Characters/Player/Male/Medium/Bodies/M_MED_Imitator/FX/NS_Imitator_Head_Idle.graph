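# Niagara idle effect for the Imitator character's head, showing pale blue mesh drips that dangle and fall from the skeletal mesh.
# DripsFalling uses randomized bursts at five bone/socket spawn groups, re-sampling the pivot each update for a damped pendulum constraint that orients and stretches each drip.
# It passes pivot, tip, falloff and rigidity to dynamic material parameters.
# Drops spawns small meshes at the same points that fall under gravity and shrink away.
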NiagaraSystem /BRCosmetics/Characters/Player/Male/Medium/Bodies/M_MED_Imitator/FX/NS_Imitator_Head_Idle
dependencies {usedDataInterfaces [&NiagaraDataInterfaceCurve], usedDynamicInputs [], usedModules [&SystemState]}
emitters {
 DripsFalling {
  emitterUpdateScript [
   &EmitterState(
    Life Cycle Mode=::NewEnumerator1
    Loop Behavior=::NewEnumerator1
    Loop Delay=&RandomRangeFloat(Minimum=0.3, Maximum=0.8, Recalculate Random Each Loop=false)
    Loop Duration Mode=::NewEnumerator1
    UseLoopDelay=false
   )
   &SpawnBurst_Instantaneous(
    Spawn Count=1
    Spawn Probability=0.8
    Spawn Time=&RandomRangeFloat(Minimum=0.1, Maximum=0.5)
    Use Spawn Probability=true
   )
   &SpawnBurst_Instantaneous(
    Spawn Count=1
    Spawn Probability=0.8
    Spawn Time=&RandomRangeFloat(Minimum=0.1, Maximum=0.5)
    Use Spawn Probability=true
   ) {module=SpawnBurst_Instantaneous005}
   &SpawnBurst_Instantaneous(
    Spawn Count=1
    Spawn Probability=0.8
    Spawn Time=&RandomRangeFloat(Minimum=0.4, Maximum=1)
    Use Spawn Probability=true
   ) {module=SpawnBurst_Instantaneous001}
   &SpawnBurst_Instantaneous(
    Spawn Count=1
    Spawn Probability=0.8
    Spawn Time=&RandomRangeFloat(Minimum=1.2, Maximum=1.5)
    Use Spawn Probability=true
   ) {module=SpawnBurst_Instantaneous003}
   &SpawnBurst_Instantaneous(
    Spawn Count=1
    Spawn Probability=0.8
    Spawn Time=&RandomRangeFloat(Minimum=1.5, Maximum=2)
    Use Spawn Probability=true
   ) {module=SpawnBurst_Instantaneous002}
  ]
  particleSpawnScript [
   &FN_SetSpawnGroup(Number of Spawn Groups=5)
   /SetVariables(Particles.ParticleRandom:NiagaraFloat=&RandomRangeFloat(Minimum=0.5, Maximum=0.9)) {isSetParameters=true, module=SetVariables_#1}
   /SetVariables(
    Particles.PendulumDrag=5
    Particles.PendulumLengthScale:NiagaraFloat=&Multiply_Float(A=&RandomRangeFloat(Minimum=0.35, Maximum=0.45), B=$Particles.ParticleRandom)
    Particles.PendulumMaskFalloff=1.5
    Particles.PendulumRigidity=0.2
   ) {isSetParameters=true}
   /Niagara/Modules/Spawn/Initialization/InitializeParticle(
    Color=<rgba 0.7031 0.8148 1 0.9>
    Lifetime=&UniformRangedFloat(Minimum=2.5, Maximum=4)
    Mesh Scale=&Multiply_Vector(
     A=<1 1 1>
     B=&MakeVector(
      X=&Multiply_Float(A=0.1, B=$Particles.ParticleRandom)
      Y=&Multiply_Float(A=0.1, B=$Particles.ParticleRandom)
      Z=$Particles.PendulumLengthScale
     )
    )
    Module.Write Color=true
    Module.Write Lifetime=true
    Module.Write Mass=true
    Module.Write Material Random=true
    Module.Write Position=false
    Module.Write Scale=true
    Module.Write SpriteRotation=false
    Module.Write SpriteSize=true
    Sprite Size=<1 1>
   )
  ]
  particleUpdateScript [
   &SkeletalMeshLocation(
    Apply To Particle Position=false
    Bone / Socket Index=$Particles.SpawnGroup
    Inherit Velocity=false
    MaskSpawnGroup=false
    Mesh Sampling Type=::NewEnumerator5
    Skeletal Mesh=$User.SkeletalMesh
    Socket Sampling Mode=::NewEnumerator5
   )
   &ParticleState()
   /SetVariables(Particles.MeshOrigin:Vector3f=&Add_Vector(A=$Output.SkeletalMeshLocation.SampledPosition, B=&MakeVector(X=0, Y=0, Z=0))) {isSetParameters=true}
   &Drag(Drag=$Particles.PendulumDrag, Use Linear Drag=true, Use Rotational Drag=false)
   &GravityForce(Gravity=<0 0 -900>)
   &PendulumSetup(
    Module.Enable Spring Driver=false
    Pendulum Length:NiagaraFloat=&Multiply_Float(A=100, B=$Particles.PendulumLengthScale)
    Pendulum Pivot:Vector3f=$Particles.MeshOrigin
    "Spring Driven Constraint (Tightness)"=3
   )
   &SolveForcesAndVelocity(Clamp Velocity=false, Limit Acceleration=false)
   &PendulumConstraint()
   &CalculateAccurateVelocity()
   /SetVariables(Particles.MeshFacing:Vector3f=&NormalizeVector(Vector To Normalize=&Subtract_Vector(A=$Particles.MeshOrigin, B=Linked() {linkedVariable {name=Particles.Position, type=NiagaraPosition}}))) {isSetParameters=true}
   /Niagara/Modules/Update/Orientation/V2/OrientMeshToVector(
    Look At Direction:Vector3f=&Multiply_VectorByFloat(Vector=$Particles.MeshFacing, Float=1)
    Override Quat=false
    Reference Axis Vector:Vector3f=<0 0 1>
    UseOptionalSideAxis=false
    UseOptionalUpAxis=false
   )
   &DynamicMaterialParameters(
    Index 2 Param 1=HlslExpression() {hlslExpression=Particles.MeshOrigin.x}
    Index 2 Param 2=HlslExpression() {hlslExpression=Particles.MeshOrigin.y}
    Index 2 Param 3=HlslExpression() {hlslExpression=Particles.MeshOrigin.z}
    Index 2 Param 4=$Particles.PendulumMaskFalloff
    Index 3 Param 1=&MakeFloatFromVector(Vector=Linked() {linkedVariable {name=Particles.Position, type=NiagaraPosition}}, Channel=::NewEnumerator0)
    Index 3 Param 2=&MakeFloatFromVector(Vector=Linked() {linkedVariable {name=Particles.Position, type=NiagaraPosition}}, Channel=::NewEnumerator1)
    Index 3 Param 3=&MakeFloatFromVector(Vector=Linked() {linkedVariable {name=Particles.Position, type=NiagaraPosition}}, Channel=::NewEnumerator2)
    Index 3 Param 4=$Particles.PendulumRigidity
    Module.Param0WriteEnabled2=true
    Module.Param0WriteEnabled3=true
    Module.Param1WriteEnabled2=true
    Module.Param1WriteEnabled3=true
    Module.Param2WriteEnabled2=true
    Module.Param2WriteEnabled3=true
    Module.Param3WriteEnabled2=true
    Module.Param3WriteEnabled3=true
    Write Parameter Index 0=false
    Write Parameter Index 2=true
    Write Parameter Index 3=true
   )
   &ScaleMeshSize(Scale Factor=&VectorFromFloat(Value=&FloatFromCurve(FloatCurve=DataInterface() {propertyValues {Curve {keys [
          {arriveTangent=0.0227, interpMode=RCIM_Cubic, leaveTangent=0.0227, tangentMode=RCTM_User, time=0, value=0}
          {arriveTangent=0.0356, interpMode=RCIM_Cubic, leaveTangent=0.0356, tangentMode=RCTM_User, time=0.2, value=1}
          {arriveTangent=-0.0004, interpMode=RCIM_Cubic, leaveTangent=-0.0004, tangentMode=RCTM_User, time=0.9, value=0.9847}
          {arriveTangent=-0.0405, interpMode=RCIM_Cubic, leaveTangent=-0.0405, tangentMode=RCTM_User, time=1, value=0}
         ]}}})))
   &ScaleMeshSize(Scale Factor=&MakeVector(
     X=1
     Y=1
     Z=&FloatFromCurve(FloatCurve=DataInterface() {propertyValues {Curve {keys [
          {arriveTangent=0.0227, interpMode=RCIM_Cubic, leaveTangent=0.0227, tangentMode=RCTM_User, time=0.2, value=0}
          {arriveTangent=0.6552, interpMode=RCIM_Cubic, leaveTangent=0.6552, tangentMode=RCTM_User, time=0.3, value=1}
          {arriveTangent=-0.6135, interpMode=RCIM_Cubic, leaveTangent=-0.6135, tangentMode=RCTM_User, time=0.9, value=1.1}
          {arriveTangent=-0.0405, interpMode=RCIM_Cubic, leaveTangent=-0.0405, tangentMode=RCTM_User, time=1, value=0}
         ]}}})
    )) {module=ScaleMeshSize001}
  ]
  rendererClasses [&NiagaraMeshRendererProperties]
  renderers [{rendererClass=&NiagaraMeshRendererProperties, rendererIndex=0}]
 }
 Drops {
  emitterUpdateScript [
   &EmitterState()
   &SpawnRate(SpawnRate=&RandomRangeFloat(Minimum=1, Maximum=2), Use Spawn Probability=false)
  ]
  particleSpawnScript [
   &InitializeParticle(
    Color=<rgba 0.7031 0.8148 1 0.9>
    Lifetime=&UniformRangedFloat(Minimum=1.2, Maximum=1.5)
    Mesh Scale=&VectorFromFloat(Value=&RandomRangeFloat(Minimum=0.025, Maximum=0.0375))
    Module.Write Color=true
    Module.Write Lifetime=true
    Module.Write Mass=true
    Module.Write Material Random=true
    Module.Write Position=false
    Module.Write Scale=true
    Module.Write SpriteRotation=false
    Module.Write SpriteSize=true
    Sprite Size=<1 1>
   )
   &FN_SetSpawnGroup(Number of Spawn Groups=5)
   /SetVariables(Particles.ParticleRandom:NiagaraFloat=&RandomRangeFloat(Minimum=0.4, Maximum=0.6)) {isSetParameters=true, module=SetVariables_#1}
   &SkeletalMeshLocation(
    Bone / Socket Index=$Particles.SpawnGroup
    Inherit Velocity=false
    MaskSpawnGroup=false
    Mesh Sampling Type=::NewEnumerator5
    Skeletal Mesh=$User.SkeletalMesh
    Socket Sampling Mode=::NewEnumerator5
   )
  ]
  particleUpdateScript [
   &ParticleState()
   &GravityForce(Gravity=<0 0 -800>)
   &SolveForcesAndVelocity(Clamp Velocity=false, Limit Acceleration=false)
   &ScaleMeshSize(Scale Factor=&VectorFromFloat(Value=&FloatFromCurve(FloatCurve=DataInterface() {propertyValues {Curve {keys [
          {arriveTangent=9.0725, interpMode=RCIM_Cubic, leaveTangent=9.0725, tangentMode=RCTM_User, time=0, value=0}
          {arriveTangent=-0.0777, interpMode=RCIM_Cubic, leaveTangent=-0.0777, tangentMode=RCTM_User, time=0.1408, value=0.9847}
          {arriveTangent=-0.0405, interpMode=RCIM_Cubic, leaveTangent=-0.0405, tangentMode=RCTM_User, time=1, value=0}
         ]}}}))) {module=ScaleMeshSize002}
  ]
  rendererClasses [&NiagaraMeshRendererProperties]
  renderers [{rendererClass=&NiagaraMeshRendererProperties, rendererIndex=0}]
 }
}
summary {
 systemName=NS_Imitator_Head_Idle
 emitters [
  {emitterName=DripsFalling, rendererClasses [&NiagaraMeshRendererProperties]}
  {emitterName=Drops, rendererClasses [&NiagaraMeshRendererProperties]}
 ]
 userVariables [
  User.bIsFrontEnd {type=NiagaraBool}
  User.bIsFrontEndPreview {type=NiagaraBool}
  User.SkeletalMesh {
   type=NiagaraDataInterfaceSkeletalMesh
   defaultValue {
    struct=NiagaraExt_VariableValue_DataInterface
    value {dataInterface=&self:NiagaraDataInterfaceSkeletalMesh_0}
   }
  }
  User.SlurpColorAlpha {type=NiagaraFloat, defaultValue=0, description=""}
 ]
}
