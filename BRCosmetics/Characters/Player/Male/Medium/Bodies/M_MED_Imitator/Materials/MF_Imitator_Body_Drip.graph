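# Material function that layers animated dripping liquid over the Imitator character body's material attributes.
# Two triplanar drip projections pan per axis at different speeds, slowed by a pants mask, over positions warped by Perlin noise and a Gmap input.
# Their sum is sharpened into a mask and scaled by Alpha to blend in altered color, roughness, specular, metallic and angle-corrected normals.
# A quality switch supplies a separate low-quality drip color.
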MaterialFunction /BRCosmetics/Characters/Player/Male/Medium/Bodies/M_MED_Imitator/Materials/MF_Imitator_Body_Drip
graph {nodes [
  n1 = NamedRerouteUsage()
  Add(A=@n1, B=Multiply(B=Time()) {constA=0.3})
  Constant() {r=1}
  Constant() {r=1}
  Constant() {r=0.01}
  n2 = FunctionInput() {inputType=FunctionInput_MaterialAttributes}
  n3 = GetMaterialAttributes(None=@n2) {attributeGetTypes [
    #1
    #2
    #3
    #4
    #5
    #6
   ]}
  n4 = SetMaterialAttributes!(
   MaterialAttributes=@n3.MaterialAttributes
   Roughness=NamedRerouteUsage()
   Specular=NamedRerouteUsage()
   Normal=NamedRerouteUsage()
   BaseColor=NamedRerouteUsage()
   EmissiveColor
   Metallic=NamedRerouteUsage()
   CustomData0=Constant()
  ) {attributeSetTypes [
    #3
    #4
    #1
    #5
    #7
    #6
    #8
   ]}
  n5 = TextureObject() {samplerType=SAMPLERTYPE_LinearColor, texture=/BRCosmetics/Characters/Player/Male/Medium/Bodies/M_MED_Imitator/Textures/T_Drips_M}
  n6 = Constant()
  n7 = ScalarParameter() {defaultValue=5, group=Imitator, parameterName=NormalWarp}
  n8 = TextureObjectParameter() {
   group=Imitator
   parameterName=Panning Texture Diffuse
   samplerType=SAMPLERTYPE_Masks
   texture=/Game/Global/Textures/Default/Noise/T_Generic_Perlin
  }
  n9 = ComponentMask(None=MaterialFunctionCall!(
    TextureObject=@n8
    TextureSize=ScalarParameter() {defaultValue=55, group=Imitator, parameterName=Diffuse Scale}
    WorldPosition=@n1
    Export Float 4
    In
    World Space Normal
    ProjectionTransitionContrast=Constant() {r=7}
   ) {materialFunction=/Game/Athena/SafeZone/WorldAlignedTexture_ScalePan}.XYZ Texture) {g=true}
  n10 = Add(
   A=Multiply(A=FunctionInput(Preview=@n6) {bUsePreviewValueAsDefault=true, inputName=Gmap, inputType=FunctionInput_Scalar}) {constB=3}
   B=Add(A=Multiply(A=@n7, B=@n9), B=NamedRerouteUsage())
  )
  n11 = NamedRerouteUsage()
  n12 = NamedRerouteUsage()
  n13 = MaterialFunctionCall!(
   Export Float 4
   TextureObject=@n5
   TextureSize=@n12
   WorldPosition=@n10
   World Space Normal
   ProjectionTransitionContrast=@n11
   Pan X=NamedRerouteUsage()
   Pan Y=NamedRerouteUsage()
   Pan Z=NamedRerouteUsage()
  ) {materialFunction=/BRCosmetics/Global/MaterialFunctions/TechArt/TriPlanerProjection_SeperatePanningDirections}
  n14 = ScalarParameter() {defaultValue=1.15, parameterName=CounterSpeed}
  n15 = MaterialFunctionCall(
   TextureObject=@n5
   TextureSize=Multiply(A=@n12) {constB=1.25}
   WorldPosition=@n10
   ProjectionTransitionContrast=@n11
   Pan X=Multiply(A=NamedRerouteUsage(), B=@n14) {constB=1.33}
   Pan Y=Multiply(A=NamedRerouteUsage(), B=@n14) {constB=1.33}
   Pan Z=Multiply(A=NamedRerouteUsage(), B=@n14) {constB=1.33}
  ) {materialFunction=&TriPlanerProjection_SeperatePanningDirections}
  n16 = Multiply(A=ComponentMask(None=Add(A=@n13.XY Texture, B=@n15.XY Texture)) {r=true}) {constB=9}
  n17 = ComponentMask(None=Saturate(None=Power(Base=@n16) {constExponent=4})) {r=true}
  n18 = Multiply(A=@n17, B=FunctionInput() {inputName=Alpha, inputType=FunctionInput_Scalar})
  FunctionOutput(None=BlendMaterialAttributes(A=@n2, B=@n4, Alpha=Saturate(None=@n18)))
  MaterialFunctionCall(TextureObject=@n5, TextureSize=@n12, WorldPosition=@n10, ProjectionTransitionContrast=@n11) {materialFunction=&WorldAlignedTexture}
  Multiply(A=@n10) {constB=1.25}
  n19 = NamedRerouteUsage()
  Multiply(A=ScalarParameter() {defaultValue=2, group=Imitator, parameterName=Drip Darken}, B=@n19)
  NamedRerouteDeclaration(None=FunctionInput() {inputName=Pre SKinned Local}) {name=PreSkinnedLocal, nodeColor=<rgba 0.0312 0 1 1>}
  NamedRerouteDeclaration(None=ScalarParameter() {defaultValue=82, group=Imitator, parameterName=Drip Scale}) {name=ThrobbingScale, nodeColor=<rgba 0 0.0625 1 1>}
  NamedRerouteDeclaration(None=@n17) {name=DripMask, nodeColor=<rgba 0.5234 1 0 1>}
  NamedRerouteDeclaration(None=@n3.Specular) {name=inSpec, nodeColor=<rgba 1 0.6328 0 1>}
  n20 = NamedRerouteUsage()
  n21 = ScalarParameter() {defaultValue=0.5, group=Imitator, parameterName=Drip Specular}
  n22 = Clamp(None=LinearInterpolate(A=Constant(), B=Constant() {r=8}, Alpha=NamedRerouteUsage()))
  NamedRerouteDeclaration(None=LinearInterpolate(A=@n20, B=Add(A=@n21, B=@n20), Alpha=@n22)) {name=Spec_Out, nodeColor=<rgba 0.0547 0 1 1>}
  NamedRerouteDeclaration(None=@n3.BaseColor) {name=inBaseColor, nodeColor=<rgba 0.1953 1 0 1>}
  n23 = VectorParameter() {parameterName=Drip Color, defaultValue=<rgba 0.06 0.06 0.06 1>}
  n24 = VectorParameter() {parameterName=Drip ColorMobile, defaultValue=<rgba 0.004 0.004 0.004 1>}
  NamedRerouteDeclaration(None=LinearInterpolate(A=@n19, B=QualitySwitch(Default=@n23.RGB, Low=@n24.RGB), Alpha=@n22)) {name=BaseColor_Out, nodeColor=<rgba 0 1 0.2969 1>}
  NamedRerouteDeclaration(None=LinearInterpolate(A=NamedRerouteUsage(), B=Constant() {r=0.5}, Alpha=@n22)) {name=Metalness_Out, nodeColor=<rgba 0 0.6811 1 1>}
  NamedRerouteDeclaration(None=@n3.Metallic) {name=inMetalness, nodeColor=<rgba 1 0.9844 0 1>}
  n25 = VectorParameter() {parameterName=Drip Panning Y, defaultValue=<rgba 0 -0.7 0 1>}
  n26 = Constant() {r=0.1}
  n27 = Multiply(
   A=LinearInterpolate(A=@n26, B=Multiply(A=@n26) {constB=0.2}, Alpha=FunctionInput() {inputName=Pants Mask, inputType=FunctionInput_Scalar})
   B=Time()
  )
  NamedRerouteDeclaration(None=ComponentMask(None=Multiply(A=@n25.RGB, B=@n27)) {g=true, r=true}) {name=Pan Y, nodeColor=<rgba 1 0.2578 0 1>}
  NamedRerouteDeclaration(None=ScalarParameter() {defaultValue=0.5, group=Imitator, parameterName=NormalBlendStrength}) {name=AxisBlendStrength, nodeColor=<rgba 0.2188 1 0 1>}
  n28 = VectorParameter() {parameterName=Drip Panning Z, defaultValue=<rgba 0 -0.4 0 1>}
  NamedRerouteDeclaration(None=ComponentMask(None=Multiply(A=@n27, B=@n28.RGB)) {g=true, r=true}) {name=Pan Z, nodeColor=<rgba 1 0.2578 0 1>}
  NamedRerouteDeclaration(None=@n3.WorldPositionOffset) {name=inWPO, nodeColor=<rgba 0 1 0.9531 1>}
  NamedRerouteDeclaration(None=@n3.Normal) {name=inNormal, nodeColor=<rgba 0 1 0.0859 1>}
  NamedRerouteDeclaration(None=@n8) {name=PanningTexture, nodeColor=<rgba 0 0.7891 1 1>}
  n29 = TextureObject() {samplerType=SAMPLERTYPE_Normal, texture=/BRCosmetics/Characters/Player/Male/Medium/Bodies/M_MED_Imitator/Textures/T_Drips_N}
  n30 = NamedRerouteUsage()
  n31 = NamedRerouteUsage()
  n32 = NamedRerouteUsage()
  n33 = NamedRerouteUsage()
  n34 = Multiply(A=NamedRerouteUsage(), B=@n14) {constB=1.33}
  n35 = Multiply(A=NamedRerouteUsage(), B=@n14) {constB=1.33}
  n36 = Multiply(A=NamedRerouteUsage(), B=@n14) {constB=1.33}
  NamedRerouteDeclaration(None=MaterialFunctionCall(
    BaseNormal=MaterialFunctionCall(TextureObject=@n29, TextureSize=@n12, WorldPosition=@n10, ProjectionTransitionContrast=@n30, Pan X=@n31, Pan Y=@n32, Pan Z=@n33) {materialFunction=&TriPlanerProjection_SeperatePanningDirections}.XY Texture
    AdditionalNormal=MaterialFunctionCall(TextureObject=@n29, WorldPosition=@n10, ProjectionTransitionContrast=@n30, Pan X=@n34, Pan Y=@n35, Pan Z=@n36) {materialFunction=&TriPlanerProjection_SeperatePanningDirections}.XY Texture
   ) {materialFunction=&BlendAngleCorrectedNormals}.Result) {name=Normal Out, nodeColor=<rgba 0 1 0.2969 1>}
  n37 = VectorParameter() {parameterName=Drip Panning X, defaultValue=<rgba 0.0001 -0.6 3 1>}
  NamedRerouteDeclaration(None=ComponentMask(None=Multiply(A=@n37.RGB, B=@n27)) {g=true, r=true}) {name=Pan X, nodeColor=<rgba 1 0.2578 0 1>}
  NamedRerouteDeclaration(None=@n3.Roughness) {name=inRoughness, nodeColor=<rgba 1 0.5156 0 1>}
  n38 = NamedRerouteUsage()
  n39 = ScalarParameter() {defaultValue=0.25, group=Imitator, parameterName=Drip Roughness}
  NamedRerouteDeclaration(None=LinearInterpolate(A=@n38, B=Multiply(A=@n39, B=@n38), Alpha=@n22)) {name=Roughness_Out, nodeColor=<rgba 0.9922 0 1 1>}
  TextureSample(Tex=@n5) {texture=/BRCosmetics/Characters/Player/Male/Medium/Bodies/M_MED_Imitator/Textures/T_DripMask}
 ]}
