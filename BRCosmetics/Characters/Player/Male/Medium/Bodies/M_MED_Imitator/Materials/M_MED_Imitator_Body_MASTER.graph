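# Master character body material for the Imitator skin (male medium), built on material attributes, with an optional drip effect.
# Shared PreFX, CharacterSkinFX and PostFX functions form the chain, with MF_Imitator_Body_Drip inserted before PostFX, driven by drip and skin FX masks.
# The useImitatorDrip static switch also enables a World Position Offset blend toward the object's origin, weighted by an offset, saturated DripMask.
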Material /BRCosmetics/Characters/Player/Male/Medium/Bodies/M_MED_Imitator/Materials/M_MED_Imitator_Body_MASTER
graph {
 properties {
  bAutomaticallySetUsageInEditor=false
  bUseMaterialAttributes=true
  bUsedWithMeshDeformer=true
  bUsedWithMorphTargets=true
  bUsedWithSkeletalMesh=true
  pixelDepthOffsetMode=PDOM_Legacy
  refractionMethod=RM_IndexOfRefraction
  shadingModel=MSM_FromMaterialExpression
  subsurfaceProfile=&SS_HeroSkin_02
  usedShadingModels=Default Lit | Subsurface | Clear Coat | Subsurface Profile | Cloth | Eye
 }
 nodes [
  n1 = MaterialFunctionCall() {materialFunction=&BaseFN_CharacterFunction_PreFX}
  n2 = TextureSampleParameter2D() {parameterName=DripMask, texture=/BRCosmetics/Characters/Player/Male/Medium/Bodies/M_MED_Imitator/Textures/T_DripMask}
  n3 = TextureSample(Tex=TextureObjectParameter() {group=Skin FX, parameterName=SkinFX_Mask}) {texture=/BRCosmetics/Characters/Player/Male/Medium/Bodies/M_MED_Imitator/Textures/T_DripMask1}
  n4 = MaterialFunctionCall(
   PreskinnedPosition=@n1.Pre skinned Local Position
   Gmap=@n1.Pre skinned Local Position
   MaterialAttributes=@n1.Pre skinned Local Position
   AO=@n1.Pre skinned Local Position
  ) {materialFunction=&CharacterSkinFX}
  n5 = MaterialFunctionCall(
   AO=@n1.AO
   Distance Blend=@n1.AO
   Fresnel=@n1.AO
   Gmap=@n1.AO
   Pre Skinned Local Position=@n1.AO
   Material=StaticSwitchParameter(
    True=MaterialFunctionCall!(Alpha=@n2.R, Pre SKinned Local=@n1.Pre skinned Local Position, Gmap=@n1.AO, Pants Mask=@n3.G, In=@n4.Result) {materialFunction=&/MF_Imitator_Body_Drip}.Result
    False=@n4.Result
   ) {defaultValue=true, parameterName=useImitatorDrip}
  ) {materialFunction=&BaseFN_CharacterFunction_PostFX}
  MaterialFunctionCall(In=@n5.Result) {materialFunction=&CINE_MF_CharPostFX}
  n6 = GetMaterialAttributes(None=@n5.Result) {attributeGetTypes [#1]}
  n7 = Multiply(A=Subtract(A=WorldPosition().XYZ, B=ObjectPositionWS())) {constB=-1}
  n8 = Saturate(None=Subtract(A=@n2.RGBA))
  StaticSwitchParameter(
   True=SetMaterialAttributes!(MaterialAttributes=@n6.MaterialAttributes, World Position Offset=LinearInterpolate(A=@n6.World Position Offset, B=@n7, Alpha=@n8)) {attributeSetTypes [#1]}
   False=@n5.Result
  ) {defaultValue=true, parameterName=useImitatorDrip}
 ]
}
