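# Material function that draws an animated, screen-like flipbook face with glitch styling on the Imitator character.
# Facial pose parameters pass through threshold tests to pick eye, eyebrow and mouth frames from a shared sub-UV texture.
# Each feature has its own placement, scale and bump-offset parallax, alongside procedural CRT, scanline, flicker and hash-based glitch effects.
# The tinted result modifies base color, emissive and opacity mask, then blends onto incoming material attributes through an FX mask.
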MaterialFunction /BRCosmetics/Characters/Player/Male/Medium/Bodies/M_MED_Imitator/Materials/MF_Imitator_Flipbook
graph {nodes [
  n1 = FunctionInput() {inputName=Material, inputType=FunctionInput_MaterialAttributes}
  n2 = GetMaterialAttributes(None=@n1) {attributeGetTypes [
    #1
    #2
    #3
    #4
   ]}
  n3 = NamedRerouteUsage()
  n4 = NamedRerouteUsage()
  n5 = VectorParameter() {group=Imitator Flipbook, parameterName=Face Tint, defaultValue=<rgba 0.2353 0.0078 0.0196 1>}
  n6 = NamedRerouteUsage()
  n7 = Add(A=NamedRerouteUsage(), B=NamedRerouteUsage())
  n8 = NamedRerouteUsage()
  n9 = ScalarParameter() {group=Imitator Flipbook, parameterName=use AlphaAsMask}
  n10 = OneMinus(None=@n6)
  n11 = SetMaterialAttributes!(
   MaterialAttributes=@n2.MaterialAttributes
   Opacity=@n2.MaterialAttributes
   BaseColor=LinearInterpolate(A=LinearInterpolate(A=@n2.BaseColor, B=@n3, Alpha=@n4), B=@n5.RGB, Alpha=@n6)
   OpacityMask=LinearInterpolate(A=@n2.OpacityMask, B=Add(A=@n7, B=@n8), Alpha=@n9)
   EmissiveColor=LinearInterpolate(A=@n5.RGB, B=@n2.EmissiveColor, Alpha=Saturate(None=@n10))
  ) {attributeSetTypes [
    #1
    #2
    #3
    #4
   ]}
  n12 = TextureObjectParameter() {
   group=Imitator Flipbook
   parameterName=FX_Mask
   samplerType=SAMPLERTYPE_Masks
   texture=/BRCosmetics/Characters/Player/Male/Medium/Bodies/M_MED_Imitator/Textures/T_Character_Imitator_FaceAcc_FX
  }
  n13 = TextureSample(UVs=TextureCoordinate(), Tex=@n12)
  n14 = Add(A=Multiply(A=NamedRerouteUsage(), B=NamedRerouteUsage()), B=Multiply(A=NamedRerouteUsage(), B=NamedRerouteUsage()))
  n15 = Multiply(A=NamedRerouteUsage(), B=NamedRerouteUsage())
  n16 = LinearInterpolate(A=@n13.B, B=Multiply(A=@n13.B, B=Add(A=@n14, B=@n15)), Alpha=NamedRerouteUsage())
  FunctionOutput(None=BlendMaterialAttributes(A=@n1, B=@n11, Alpha=LinearInterpolate(A=@n16, B=@n13.B, Alpha=@n9)) {vertexAttributeBlendType=UseA})
  Min()
  n17 = Add(
   A=NamedRerouteUsage()
   B=AppendVector(
    A=ScalarParameter() {group=Imitator Flipbook Eyes, parameterName=Flipbook X Eyes}
    B=ScalarParameter() {defaultValue=0.15, group=Imitator Flipbook Eyes, parameterName=Flipbook Y Eyes}
   )
  )
  n18 = ScalarParameter() {defaultValue=-0.25, group=Imitator Flipbook Eyes, parameterName=Bump Height Eyes}
  NamedRerouteDeclaration(None=LinearInterpolate(
    A=@n17
    B=BumpOffset(Coordinate=@n17, Height=@n18)
    Alpha=ScalarParameter() {defaultValue=1, group=Imitator Flipbook Eyes, parameterName=BumpOffset Intensity Eyes}
   )) {name=Eye BumpOffset, nodeColor=<rgba 0.3359 1 0 1>}
  NamedRerouteDeclaration(None=LinearInterpolate(
    A=TextureCoordinate()
    B=TextureCoordinate() {coordinateIndex=1}
    Alpha=ScalarParameter() {group=Imitator Flipbook, parameterName=use Second UVChannel}
   )) {name=Master SecondUVChannel, nodeColor=<rgba 0 0.1562 1 1>}
  n19 = ScalarParameter() {defaultValue=0.4, group=Imitator Flipbook Mouth, parameterName=Flipbook scale Mouth}
  n20 = ScalarParameter() {defaultValue=1, group=Imitator Flipbook Mouth, parameterName=Flipbook scale V relative Mouth}
  n21 = MaterialFunctionCall(UVs=Add(A=NamedRerouteUsage(), B=NamedRerouteUsage()), Texture Scale=AppendVector(A=@n19, B=Multiply(A=@n19, B=@n20))) {materialFunction=&ScaleUVsByCenter}
  n22 = ScalarParameter() {defaultValue=8, group=Imitator Flipbook Mouth, parameterName=SubImages X Mouth}
  n23 = ScalarParameter() {defaultValue=8, group=Imitator Flipbook Mouth, parameterName=SubImages Y Mouth}
  n24 = MaterialFunctionCall!(
   Texture=NamedRerouteUsage()
   UVs=@n21.UVs
   SubImages=LinearInterpolate(
    A=ScalarParameter() {defaultValue=8, group=Imitator Flipbook Mouth, parameterName=SubImages Mouth}
    B=AppendVector(A=@n22, B=@n23)
    Alpha=ScalarParameter() {group=Imitator Flipbook Mouth, parameterName=useNonSquareTexture Mouth}
   )
   Frame=NamedRerouteUsage()
  ) {materialFunction=/Game/Weapons/WeaponSkins/MaterialFunction/CustomNumbers_SubUV}
  NamedRerouteDeclaration(None=@n24.Alpha) {name=Mouth Alpha, nodeColor=<rgba 0.5703 0 1 1>}
  NamedRerouteDeclaration(None=StaticComponentMaskParameter(None=@n24.RGB) {defaultG=true, group=Imitator Flipbook, parameterName=ChannelMouth}) {name=Mouth RGB, nodeColor=<rgba 0.0312 0 1 1>}
  n25 = Multiply(A=NamedRerouteUsage(), B=NamedRerouteUsage())
  n26 = Multiply(A=NamedRerouteUsage(), B=NamedRerouteUsage())
  n27 = NamedRerouteUsage()
  n28 = NamedRerouteUsage()
  n29 = Multiply(
   A=ScalarParameter() {defaultValue=2, group=Imitator Flipbook, parameterName=Face Intensity}
   B=Add(A=Add(A=@n25, B=@n26), B=Multiply(A=@n27, B=@n28))
  )
  n30 = Multiply(A=ScalarParameter() {defaultValue=0.5, group=Imitator Glitch, parameterName=CRTLine Intensity}, B=NamedRerouteUsage())
  n31 = Multiply(
   A=Multiply(A=NamedRerouteUsage(), B=ScalarParameter() {defaultValue=0.4, group=Imitator Glitch, parameterName=CRT_Amount})
   B=ScalarParameter() {defaultValue=0.2, group=Imitator Glitch, parameterName=CRT Intensity}
  )
  n32 = Add(A=@n29, B=Multiply(A=ScalarParameter() {defaultValue=0.1, group=Imitator Flipbook, parameterName=Flipbook Intensity}, B=Add(A=@n30, B=@n31)))
  n33 = NamedRerouteUsage()
  NamedRerouteDeclaration(None=Multiply(
    A=VectorParameter() {group=Imitator Flipbook, parameterName=Flipbook Tint, defaultValue=<rgba 0.2353 0.0078 0.0196 1>}.RGB
    B=Add(A=@n32, B=@n33)
   )) {name=outColor, nodeColor=<rgba 1 0.3047 0 1>}
  NamedRerouteDeclaration(None=ScalarParameter() {defaultValue=1, group=Imitator Flipbook, parameterName=Affects BaseColor}) {name=affectsBaseColor, nodeColor=<rgba 1 0.9844 0 1>}
  n34 = Add(
   A=NamedRerouteUsage()
   B=AppendVector(
    A=ScalarParameter() {defaultValue=0.125, group=Imitator Flipbook Eyebrow, parameterName=Flipbook X Eyebrow}
    B=ScalarParameter() {group=Imitator Flipbook Eyebrow, parameterName=Flipbook Y Eyebrow}
   )
  )
  n35 = ScalarParameter() {defaultValue=-0.25, group=Imitator Flipbook Eyebrow, parameterName=Bump Height Eyebrow}
  NamedRerouteDeclaration(None=LinearInterpolate(
    A=@n34
    B=BumpOffset(Coordinate=@n34, Height=@n35)
    Alpha=ScalarParameter() {defaultValue=1, group=Imitator Flipbook Eyebrow, parameterName=BumpOffset Intensity Eyebrow}
   )) {name=Eyebrow BumpOffset, nodeColor=<rgba 0.3359 1 0 1>}
  n36 = ScalarParameter() {defaultValue=0.5, group=Imitator Flipbook Eyebrow, parameterName=Flipbook scale Eyebrow}
  n37 = ScalarParameter() {defaultValue=1, group=Imitator Flipbook Eyebrow, parameterName=Flipbook scale V relative Eyebrow}
  n38 = MaterialFunctionCall(UVs=Add(A=NamedRerouteUsage(), B=NamedRerouteUsage()), Texture Scale=AppendVector(A=@n36, B=Multiply(A=@n36, B=@n37))) {materialFunction=&ScaleUVsByCenter}
  NamedRerouteDeclaration(None=@n38.0-1 mask) {name=Eyebrow 01Mask, nodeColor=<rgba 0 0.1328 1 1>}
  n39 = ScalarParameter() {defaultValue=8, group=Imitator Flipbook Eyebrow, parameterName=SubImages X Eyebrow}
  n40 = ScalarParameter() {defaultValue=8, group=Imitator Flipbook Eyebrow, parameterName=SubImages Y Eyebrow}
  n41 = MaterialFunctionCall(
   Texture=NamedRerouteUsage()
   UVs=@n38.UVs
   SubImages=LinearInterpolate(
    A=ScalarParameter() {defaultValue=8, group=Imitator Flipbook Eyebrow, parameterName=SubImages Eyebrow}
    B=AppendVector(A=@n39, B=@n40)
    Alpha=ScalarParameter() {group=Imitator Flipbook Eyebrow, parameterName=useNonSquareTexture Eyebrow}
   )
   Frame=NamedRerouteUsage()
  ) {materialFunction=&CustomNumbers_SubUV}
  NamedRerouteDeclaration(None=@n41.Alpha) {name=Eyebrow Alpha, nodeColor=<rgba 0.5703 0 1 1>}
  NamedRerouteDeclaration(None=StaticComponentMaskParameter(None=@n41.RGB) {defaultR=true, group=Imitator Flipbook, parameterName=Channel Eyebrow}) {name=Eyebrow RGB, nodeColor=<rgba 0.0312 0 1 1>}
  n42 = ScalarParameter() {defaultValue=0.1, group=Banana, parameterName=Mouth Threshold}
  n43 = ScalarParameter() {group=Imitator Flipbook Mouth, parameterName=Frame Value Open Mouth Default}
  n44 = If(
   A=ScalarParameter() {group=Banana, parameterName=R_Frown_pose}
   B=@n42
   "A > B"=ScalarParameter() {defaultValue=10, group=Imitator Flipbook Mouth, parameterName=Frame Value Open Mouth Frown}
   "A == B"=@n43
   "A < B"=@n43
  )
  n45 = If(
   A=ScalarParameter() {group=Banana, parameterName=R_smile_pose}
   B=@n42
   "A > B"=ScalarParameter() {defaultValue=6, group=Imitator Flipbook Mouth, parameterName=Frame Value Open Mouth Smile}
   "A == B"=@n43
   "A < B"=@n43
  )
  n46 = If(
   A=ScalarParameter() {parameterName=L_upper_lip_raiser_pose}
   B=@n42
   "A > B"=ScalarParameter() {defaultValue=9, group=Imitator Flipbook Mouth, parameterName=Frame Value Raise L Lip Corner}
   "A == B"=@n43
   "A < B"=@n43
  )
  n47 = Max(
   A=@n46
   B=If(
    A=ScalarParameter() {parameterName=R_upper_lip_raiser_pose}
    B=@n42
    "A > B"=ScalarParameter() {defaultValue=7, group=Imitator Flipbook Mouth, parameterName=Frame Value Raise R Lip Corner}
    "A == B"=@n46
    "A < B"=@n46
   )
  )
  n48 = If(
   A=Max(A=ScalarParameter() {parameterName=R_lip_corner_narrow_pose}, B=ScalarParameter() {parameterName=PeelyBlenderFace})
   B=@n42
   "A > B"=ScalarParameter() {defaultValue=2, group=Imitator Flipbook Mouth, parameterName=Frame Value Narrow Lip}
   "A == B"=@n47
   "A < B"=@n47
  )
  n49 = Max(
   A=If(
    A=ScalarParameter() {group=Banana, parameterName=R_frown_pose}
    B=@n42
    "A > B"=ScalarParameter() {defaultValue=4, group=Imitator Flipbook Mouth, parameterName=Frame Value Closed Frown}
    "A == B"=@n42
    "A < B"=@n42
   )
   B=If(
    A=ScalarParameter() {group=Banana, parameterName=R_smile_pose}
    B=@n42
    "A > B"=ScalarParameter() {defaultValue=8, group=Imitator Flipbook Mouth, parameterName=Frame Value Closed Mouth Smile}
    "A == B"=@n48
    "A < B"=@n48
   )
  )
  NamedRerouteDeclaration(None=If(A=ScalarParameter() {group=Banana, parameterName=Jaw_open_Pose}, B=@n42, "A > B"=Max(A=@n44, B=@n45), "A == B"=@n49, "A < B"=@n49)) {name=mouthValues, nodeColor=<rgba 0.8047 1 0 1>}
  n50 = ScalarParameter() {defaultValue=0.5, group=Imitator Flipbook Eyes, parameterName=Flipbook scale Eyes}
  n51 = ScalarParameter() {defaultValue=1, group=Imitator Flipbook Eyes, parameterName=Flipbook scale V relative Eyes}
  n52 = MaterialFunctionCall(UVs=Add(A=NamedRerouteUsage(), B=NamedRerouteUsage()), Texture Scale=AppendVector(A=@n50, B=Multiply(A=@n50, B=@n51))) {materialFunction=&ScaleUVsByCenter}
  NamedRerouteDeclaration(None=@n52.0-1 mask) {name=Eye 01Mask, nodeColor=<rgba 0 0.1328 1 1>}
  NamedRerouteDeclaration(None=@n29) {name=faceAlpha, nodeColor=<rgba 0.6406 0 1 1>}
  n53 = ScalarParameter() {group=Banana, parameterName=R_blink_pose}
  n54 = ScalarParameter() {defaultValue=0.1, group=Banana, parameterName=Threshold R Blink}
  n55 = ScalarParameter() {defaultValue=2, group=Imitator Flipbook Eyes, parameterName=Frame Value Blink}
  n56 = ScalarParameter() {group=Imitator Flipbook Eyes, parameterName=Frame Value Eye Default}
  n57 = ScalarParameter() {group=Banana, parameterName=R_squint_inner_pose}
  n58 = ScalarParameter() {defaultValue=0.1, group=Banana, parameterName=Threshold Squint}
  n59 = ScalarParameter() {defaultValue=1, group=Imitator Flipbook Eyes, parameterName=Frame Value Squint}
  NamedRerouteDeclaration(None=Max(A=If(A=@n53, B=@n54, "A > B"=@n55, "A == B"=@n56, "A < B"=@n56), B=If(A=@n57, B=@n58, "A > B"=@n59, "A == B"=@n56, "A < B"=@n56))) {name=eyeValues, nodeColor=<rgba 0.4062 1 0 1>}
  n60 = TextureCoordinate() {coordinateIndex=1}
  n61 = Ceil(None=Multiply(A=ComponentMask(None=@n60) {g=true}, B=ScalarParameter() {defaultValue=100, group=Imitator Glitch, parameterName=RB_GlitchLine_Tiling}))
  n62 = ComponentMask(None=MaterialFunctionCall(Seed=Time(), Float=ComponentMask(None=@n61) {g=true}, Vec2=@n61) {materialFunction=&MF_Hash}.Vec2) {r=true}
  n63 = ScalarParameter() {defaultValue=1, group=Imitator Glitch, parameterName=RB_GlitchDistortion_Amt}
  n64 = Frac(None=Multiply(A=Time(), B=ScalarParameter() {defaultValue=0.1, group=Imitator Glitch, parameterName=RB_GlitchFrequency}) {constB=0.05})
  n65 = MaterialFunctionCall(Float=Time()) {materialFunction=&MF_Hash}
  n66 = LinearInterpolate(
   A=ScalarParameter() {defaultValue=0.8, group=Imitator Glitch, parameterName=RB_GlitchThreshold_Min}
   B=ScalarParameter() {defaultValue=0.9, group=Imitator Glitch, parameterName=RB_GlitchThreshold_Max}
   Alpha=Round(None=ConstantBiasScale(None=@n65.Float))
  )
  n67 = Saturate(None=Ceil(None=Subtract(A=@n64, B=@n66)))
  NamedRerouteDeclaration(None=Multiply(A=Multiply(A=@n62, B=@n62), B=Multiply(A=@n63, B=@n67))) {name=glitchEffect, nodeColor=<rgba 1 0 0.7969 1>}
  n68 = CameraVectorWS()
  n69 = VertexNormalWS()
  n70 = ComponentMask(None=TextureCoordinate() {coordinateIndex=1}) {g=true}
  n71 = Multiply(A=@n70, B=ScalarParameter() {defaultValue=50, group=Imitator Glitch, parameterName=RB_ScanLine_Tiling})
  n72 = Clamp(None=Multiply(A=DotProduct(A=@n68, B=@n69), B=Sine(None=@n71)))
  n73 = Time()
  n74 = ScalarParameter() {defaultValue=-0.1, group=Imitator Glitch, parameterName=ScanLineSpeed}
  n75 = Frac(None=Add(A=@n70, B=Multiply(A=@n73, B=@n74)))
  n76 = Constant() {r=0.5}
  n77 = ScalarParameter() {defaultValue=0.1, group=Imitator Glitch, parameterName=RB_Scanline_Radius}
  n78 = Constant()
  NamedRerouteDeclaration(None=Add(A=@n72, B=SphereMask(A=@n75, B=@n76, Radius=@n77, Hardness=@n78))) {name=CRTLineAlpha, nodeColor=<rgba 0 0.4375 1 1>}
  NamedRerouteDeclaration(None=@n72) {name=ScanLineClamp, nodeColor=<rgba 0 0.0391 1 1>}
  n79 = ScalarParameter() {defaultValue=1, group=Imitator Glitch, parameterName=Flicker Speed}
  n80 = Time()
  n81 = Frac(None=Sine(None=Multiply(A=@n79, B=@n80)))
  n82 = ScalarParameter() {defaultValue=0.4, group=Imitator Glitch, parameterName=Flicker Opacity}
  NamedRerouteDeclaration(None=Multiply(A=Clamp(None=@n81, Min=@n82), B=ScalarParameter() {defaultValue=0.2, group=Imitator Glitch, parameterName=Flicker Intensity})) {name=Flicker, nodeColor=<rgba 0.7344 0 1 1>}
  n83 = TextureObjectParameter() {
   group=Imitator Flipbook
   parameterName=Flipbook Texture
   texture=/BRCosmetics/Characters/Player/Male/Medium/Bodies/M_MED_Imitator/Textures/T_Imitator_Flipbook
  }
  NamedRerouteDeclaration(None=@n83) {name=TextureFlipbook, nodeColor=<rgba 0.6641 1 0 1>}
  n84 = ScalarParameter() {defaultValue=4, group=Imitator Flipbook Eyebrow, parameterName=Frame Value Glabella Up}
  n85 = ScalarParameter() {group=Banana, parameterName=R_brow_up_pose, sortPriority=1}
  n86 = ScalarParameter() {defaultValue=0.1, group=Banana, parameterName=Threshold Brow Up}
  n87 = ScalarParameter() {defaultValue=3, group=Imitator Flipbook Eyebrow, parameterName=Frame Value Brow Up}
  n88 = ScalarParameter() {defaultValue=5, group=Imitator Flipbook Eyebrow, parameterName=Frame Value Glabella Down}
  n89 = If(
   A=ScalarParameter() {group=Banana, parameterName=C_glabella_down_pose}
   B=ScalarParameter() {defaultValue=0.3, group=Banana, parameterName=Threshold Glabella Down}
   "A > B"=@n88
   "A == B"=@n88
   "A < B"=ScalarParameter() {defaultValue=6, group=Imitator Flipbook Eyebrow, parameterName=Frame Value Eyebrow Default}
  )
  NamedRerouteDeclaration(None=If(
    A=ScalarParameter() {group=Banana, parameterName=C_glabella_up_pose}
    B=ScalarParameter() {defaultValue=0.1, group=Banana, parameterName=Threshold_Glabella_up}
    "A > B"=@n84
    "A == B"=@n84
    "A < B"=If(A=@n85, B=@n86, "A > B"=@n87, "A == B"=@n87, "A < B"=@n89)
   )) {name=eyebrowValues, nodeColor=<rgba 1 0.0703 0 1>}
  n90 = ScalarParameter() {defaultValue=8, group=Imitator Flipbook Eyes, parameterName=SubImages X Eyes}
  n91 = ScalarParameter() {defaultValue=8, group=Imitator Flipbook Eyes, parameterName=SubImages Y Eyes}
  n92 = MaterialFunctionCall(
   Texture=NamedRerouteUsage()
   UVs=@n52.UVs
   SubImages=LinearInterpolate(
    A=ScalarParameter() {defaultValue=8, group=Imitator Flipbook Eyes, parameterName=SubImages Eyes}
    B=AppendVector(A=@n90, B=@n91)
    Alpha=ScalarParameter() {group=Imitator Flipbook Eyes, parameterName=useNonSquareTexture Eyes}
   )
   Frame=NamedRerouteUsage()
  ) {materialFunction=&CustomNumbers_SubUV}
  NamedRerouteDeclaration(None=StaticComponentMaskParameter(None=@n92.RGB) {defaultR=true, group=Imitator Flipbook, parameterName=Channel Eye}) {name=Eye RGB, nodeColor=<rgba 0.7109 1 0 1>}
  NamedRerouteDeclaration(None=@n92.Alpha) {name=Eye Alpha, nodeColor=<rgba 0 0.6484 1 1>}
  n93 = Add(
   A=NamedRerouteUsage()
   B=AppendVector(
    A=ScalarParameter() {group=Imitator Flipbook Mouth, parameterName=Flipbook X Mouth}
    B=ScalarParameter() {defaultValue=0.1, group=Imitator Flipbook Mouth, parameterName=Flipbook Y Mouth}
   )
  )
  n94 = ScalarParameter() {defaultValue=-0.25, group=Imitator Flipbook Mouth, parameterName=Bump Height Mouth}
  NamedRerouteDeclaration(None=LinearInterpolate(
    A=@n93
    B=BumpOffset(Coordinate=@n93, Height=@n94)
    Alpha=ScalarParameter() {defaultValue=1, group=Imitator Flipbook Mouth, parameterName=BumpOffset Intensity Mouth}
   )) {name=Mouth BumpOffset, nodeColor=<rgba 0.3359 1 0 1>}
  NamedRerouteDeclaration(None=@n21.0-1 mask) {name=Mouth 01Mask, nodeColor=<rgba 0 0.1328 1 1>}
  n95 = Multiply(A=TextureCoordinate(), B=ScalarParameter() {defaultValue=10, group=Imitator Glitch, parameterName=CRTScaleUV})
  n96 = Sign(None=Sine(None=ComponentMask(None=@n95) {g=true}) {period=2})
  n97 = Constant()
  n98 = Add(A=@n95, B=Multiply(A=AppendVector(A=@n96, B=@n97)) {constB=0.25})
  NamedRerouteDeclaration(None=Multiply(
    A=TextureSample(UVs=@n98) {samplerType=SAMPLERTYPE_LinearColor, texture=/Game/Accessories/FORT_Backpacks/Backpack_M_MED_Jumbotron_S20/Textures/T_Jumbotron_CRT}.G
    B=ScalarParameter() {defaultValue=0.8, group=Imitator Glitch, parameterName=CRT_ColorScale}
   )) {name=CRTPattern, nodeColor=<rgba 1 0 0.2344 1>}
  ScalarParameter() {defaultValue=0.5, group=Banana, parameterName=Threshhold Frown}
 ]}
