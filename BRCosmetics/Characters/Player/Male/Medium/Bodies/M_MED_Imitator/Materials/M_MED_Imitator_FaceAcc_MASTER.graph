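# Master material for the face accessory of the male medium Imitator player character, using material attributes on skeletal meshes with morph target and mesh deformer support.
# Shared character PreFX feeds CharacterSkinFX, then character PostFX and finally a cinematic post-FX function.
# A static switch, useImitatorFace, optionally routes the skin result through MF_Imitator_Flipbook before PostFX.
# Shading model comes from the material graph, with a hero skin subsurface profile assigned.
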Material /BRCosmetics/Characters/Player/Male/Medium/Bodies/M_MED_Imitator/Materials/M_MED_Imitator_FaceAcc_MASTER
graph {
 properties {
  bAutomaticallySetUsageInEditor=false
  bUseMaterialAttributes=true
  bUsedWithMeshDeformer=true
  bUsedWithMorphTargets=true
  bUsedWithSkeletalMesh=true
  pixelDepthOffsetMode=PDOM_Legacy
  refractionMethod=RM_IndexOfRefraction
  shadingModel=MSM_FromMaterialExpression
  subsurfaceProfile=&SS_HeroSkin_02
  usedShadingModels=Default Lit | Subsurface | Clear Coat | Subsurface Profile | Cloth | Eye
 }
 nodes [
  n1 = MaterialFunctionCall() {materialFunction=&BaseFN_CharacterFunction_PreFX}
  n2 = MaterialFunctionCall(
   PreskinnedPosition=@n1.Pre skinned Local Position
   Gmap=@n1.Pre skinned Local Position
   MaterialAttributes=@n1.Pre skinned Local Position
   AO=@n1.Pre skinned Local Position
  ) {materialFunction=&CharacterSkinFX}
  n3 = MaterialFunctionCall(
   AO=@n1.AO
   Distance Blend=@n1.AO
   Fresnel=@n1.AO
   Gmap=@n1.AO
   Pre Skinned Local Position=@n1.AO
   Material=StaticSwitchParameter(
    True=MaterialFunctionCall!(Material=@n2.Result) {materialFunction=&/MF_Imitator_Flipbook}.Result
    False=@n2.Result
   ) {parameterName=useImitatorFace}
  ) {materialFunction=&BaseFN_CharacterFunction_PostFX}
  MaterialFunctionCall(In=@n3.Result) {materialFunction=&CINE_MF_CharPostFX}
 ]
}
