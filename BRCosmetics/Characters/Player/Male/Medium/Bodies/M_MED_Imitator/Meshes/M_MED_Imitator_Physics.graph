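# Physics asset for the medium male Imitator player body, covering the pelvis, legs and dynamic left and right straps and side straps.
# Strap capsules form tightly limited chains from master bones on the pelvis, with the main straps branching differently on each side.
# Pelvis, leg and master-bone joints have free angular motion.
# The root carries two zero-length capsules acting as spheres, one very large and offset far below.
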PhysicsAsset /BRCosmetics/Characters/Player/Male/Medium/Bodies/M_MED_Imitator/Meshes/M_MED_Imitator_Physics
bodies {
 calf_l [{shapeType=Capsule, length=24.4745, radius=8.7907, shapeName=calf_l_capsule, center=<-21.1378 0 0>, rotation=<rot 90 0 0>}]
 calf_r [{shapeType=Capsule, length=24.4745, radius=8.7907, shapeName=calf_r_capsule, center=<21.3393 0 0>, rotation=<rot 90 0 0>}]
 dyn_l_Sidestraps_01 [{shapeType=Capsule, length=5, radius=5, shapeName=dyn_l_Sidestraps_01_capsule, center=<5.6543 0 0>, rotation=<rot 90 0 0>}]
 dyn_l_Sidestraps_02 [{shapeType=Capsule, length=7.6289, radius=2.3711, shapeName=dyn_l_Sidestraps_02_capsule, center=<5.6543 0 0>, rotation=<rot 90 0 0>}]
 dyn_l_Sidestraps_MSTR [{shapeType=Capsule, length=0.505, radius=0.505, shapeName=dyn_l_Sidestraps_MSTR_capsule, rotation=<rot 0 -90 0>}]
 dyn_l_straps_01 [{shapeType=Capsule, length=5, radius=5, shapeName=dyn_l_straps_01_capsule, center=<5.6543 0 0>, rotation=<rot 90 0 0>}]
 dyn_l_straps_02 [{shapeType=Capsule, length=5, radius=5, shapeName=dyn_l_straps_02_capsule, center=<5.6543 0 0>, rotation=<rot 90 0 0>}]
 dyn_l_straps_03 [{shapeType=Capsule, length=5, radius=5, shapeName=dyn_l_straps_03_capsule, center=<5.6543 0 0>, rotation=<rot 90 0 0>}]
 dyn_l_straps_04 [{shapeType=Capsule, length=5, radius=5, shapeName=dyn_l_straps_04_capsule, center=<5.6543 0 0>, rotation=<rot 90 0 0>}]
 dyn_l_straps_05 [{shapeType=Capsule, length=5, radius=5, shapeName=dyn_l_straps_05_capsule, center=<5.6543 0 0>, rotation=<rot 90 0 0>}]
 dyn_l_straps_06 [{shapeType=Capsule, length=5, radius=5, shapeName=dyn_l_straps_06_capsule, center=<5.6543 0 0>, rotation=<rot 90 0 0>}]
 dyn_l_straps_07 [{shapeType=Capsule, length=5, radius=5, shapeName=dyn_l_straps_07_capsule, center=<5.6543 0 0>, rotation=<rot 90 0 0>}]
 dyn_l_straps_08 [{shapeType=Capsule, length=5, radius=5, shapeName=dyn_l_straps_08_capsule, center=<5.6543 0 0>, rotation=<rot 90 0 0>}]
 dyn_l_straps_MSTR [{shapeType=Capsule, length=0.505, radius=0.505, shapeName=dyn_l_straps_MSTR_capsule, rotation=<rot 0 -90 0>}]
 dyn_r_Sidestraps_01 [{shapeType=Capsule, length=5, radius=5, shapeName=dyn_r_Sidestraps_01_capsule, center=<5.6543 0 0>, rotation=<rot 90 0 0>}]
 dyn_r_Sidestraps_02 [{shapeType=Capsule, length=7.6289, radius=2.3711, shapeName=dyn_r_Sidestraps_02_capsule, center=<5.6543 0 0>, rotation=<rot 90 0 0>}]
 dyn_r_Sidestraps_MSTR [{shapeType=Capsule, length=0.505, radius=0.505, shapeName=dyn_r_Sidestraps_MSTR_capsule, rotation=<rot 0 -90 0>}]
 dyn_r_straps_01 [{shapeType=Capsule, length=5, radius=5, shapeName=dyn_r_straps_01_capsule, center=<5.6543 0 0>, rotation=<rot 90 0 0>}]
 dyn_r_straps_02 [{shapeType=Capsule, length=5, radius=5, shapeName=dyn_r_straps_02_capsule, center=<5.6543 0 0>, rotation=<rot 90 0 0>}]
 dyn_r_straps_03 [{shapeType=Capsule, length=5, radius=5, shapeName=dyn_r_straps_03_capsule, center=<5.6543 0 0>, rotation=<rot 90 0 0>}]
 dyn_r_straps_04 [{shapeType=Capsule, length=5, radius=5, shapeName=dyn_r_straps_04_capsule, center=<5.6543 0 0>, rotation=<rot 90 0 0>}]
 dyn_r_straps_05 [{shapeType=Capsule, length=5, radius=5, shapeName=dyn_r_straps_05_capsule, center=<5.6543 0 0>, rotation=<rot 90 0 0>}]
 dyn_r_straps_06 [{shapeType=Capsule, length=5, radius=5, shapeName=dyn_r_straps_06_capsule, center=<5.6543 0 0>, rotation=<rot 90 0 0>}]
 dyn_r_straps_07 [{shapeType=Capsule, length=5, radius=5, shapeName=dyn_r_straps_07_capsule, center=<5.6543 0 0>, rotation=<rot 90 0 0>}]
 dyn_r_straps_08 [{shapeType=Capsule, length=5, radius=5, shapeName=dyn_r_straps_08_capsule, center=<5.6543 0 0>, rotation=<rot 90 0 0>}]
 dyn_r_straps_MSTR [{shapeType=Capsule, length=0.505, radius=0.505, shapeName=dyn_r_straps_MSTR_capsule, rotation=<rot 0 -90 0>}]
 pelvis [{shapeType=Capsule, length=6.5636, radius=11.5859, shapeName=pelvis_capsule}]
 root [
  {shapeType=Capsule, length=0, radius=187, shapeName=root_capsule, center=<-3.4923 -2.714 -174.8455>}
  {shapeType=Capsule, length=0, radius=27, shapeName=root_capsule, center=<0 -8.9708 0>, rotation=<rot 0 -90 0>}
 ]
 thigh_l [{shapeType=Capsule, length=24.4745, radius=8.7907, shapeName=thigh_l_capsule, center=<-21.1378 0 0>, rotation=<rot 90 0 0>}]
 thigh_r [{shapeType=Capsule, length=24.4745, radius=8.7907, shapeName=thigh_r_capsule, center=<21.3393 0 0>, rotation=<rot 90 0 0>}]
}
constraints [
 {
  bone1Name=thigh_l
  bone2Name=pelvis
  swing1LimitDegrees=45
  swing1Motion=Free
  swing2LimitDegrees=45
  swing2Motion=Free
  twistLimitDegrees=45
  twistMotion=Free
 }
 {
  bone1Name=thigh_r
  bone2Name=pelvis
  swing1LimitDegrees=45
  swing1Motion=Free
  swing2LimitDegrees=45
  swing2Motion=Free
  twistLimitDegrees=45
  twistMotion=Free
 }
 {
  bone1Name=calf_l
  bone2Name=thigh_l
  swing1LimitDegrees=45
  swing1Motion=Free
  swing2LimitDegrees=45
  swing2Motion=Free
  twistLimitDegrees=45
  twistMotion=Free
 }
 {
  bone1Name=calf_r
  bone2Name=thigh_r
  swing1LimitDegrees=45
  swing1Motion=Free
  swing2LimitDegrees=45
  swing2Motion=Free
  twistLimitDegrees=45
  twistMotion=Free
 }
 {bone1Name=dyn_l_straps_01, bone2Name=dyn_l_straps_MSTR, swing1LimitDegrees=20, swing2LimitDegrees=20, twistLimitDegrees=5}
 {bone1Name=dyn_l_straps_02, bone2Name=dyn_l_straps_01, swing1LimitDegrees=20, swing2LimitDegrees=20, twistLimitDegrees=5}
 {bone1Name=dyn_l_straps_03, bone2Name=dyn_l_straps_02, swing1LimitDegrees=20, swing2LimitDegrees=20, twistLimitDegrees=5}
 {bone1Name=dyn_l_straps_04, bone2Name=dyn_l_straps_03, swing1LimitDegrees=20, swing2LimitDegrees=20, twistLimitDegrees=5}
 {bone1Name=dyn_l_straps_06, bone2Name=dyn_l_straps_04, swing1LimitDegrees=20, swing2LimitDegrees=20, twistLimitDegrees=5}
 {bone1Name=dyn_l_straps_07, bone2Name=dyn_l_straps_06, swing1LimitDegrees=20, swing2LimitDegrees=20, twistLimitDegrees=5}
 {bone1Name=dyn_l_straps_08, bone2Name=dyn_l_straps_06, swing1LimitDegrees=20, swing2LimitDegrees=20, twistLimitDegrees=5}
 {bone1Name=dyn_l_straps_05, bone2Name=dyn_l_straps_03, swing1LimitDegrees=20, swing2LimitDegrees=20, twistLimitDegrees=5}
 {bone1Name=dyn_r_Sidestraps_01, bone2Name=dyn_r_Sidestraps_MSTR, swing1LimitDegrees=20, swing2LimitDegrees=20, twistLimitDegrees=5}
 {bone1Name=dyn_r_Sidestraps_02, bone2Name=dyn_r_Sidestraps_01, swing1LimitDegrees=20, swing2LimitDegrees=20, twistLimitDegrees=5}
 {bone1Name=dyn_r_straps_01, bone2Name=dyn_r_straps_MSTR, swing1LimitDegrees=20, swing2LimitDegrees=20, twistLimitDegrees=5}
 {bone1Name=dyn_r_straps_02, bone2Name=dyn_r_straps_01, swing1LimitDegrees=20, swing2LimitDegrees=20, twistLimitDegrees=5}
 {bone1Name=dyn_r_straps_03, bone2Name=dyn_r_straps_02, swing1LimitDegrees=20, swing2LimitDegrees=20, twistLimitDegrees=5}
 {bone1Name=dyn_r_straps_04, bone2Name=dyn_r_straps_03, swing1LimitDegrees=20, swing2LimitDegrees=20, twistLimitDegrees=5}
 {bone1Name=dyn_r_straps_05, bone2Name=dyn_r_straps_03, swing1LimitDegrees=20, swing2LimitDegrees=20, twistLimitDegrees=5}
 {bone1Name=dyn_r_straps_06, bone2Name=dyn_r_straps_05, swing1LimitDegrees=20, swing2LimitDegrees=20, twistLimitDegrees=5}
 {bone1Name=dyn_r_straps_08, bone2Name=dyn_r_straps_06, swing1LimitDegrees=20, swing2LimitDegrees=20, twistLimitDegrees=5}
 {bone1Name=dyn_r_straps_07, bone2Name=dyn_r_straps_05, swing1LimitDegrees=20, swing2LimitDegrees=20, twistLimitDegrees=5}
 {bone1Name=dyn_l_Sidestraps_02, bone2Name=dyn_l_Sidestraps_01, swing1LimitDegrees=20, swing2LimitDegrees=20, twistLimitDegrees=5}
 {bone1Name=dyn_l_Sidestraps_01, bone2Name=dyn_l_Sidestraps_MSTR, swing1LimitDegrees=20, swing2LimitDegrees=20, twistLimitDegrees=5}
 {
  bone1Name=dyn_l_straps_MSTR
  bone2Name=pelvis
  swing1LimitDegrees=45
  swing1Motion=Free
  swing2LimitDegrees=45
  swing2Motion=Free
  twistLimitDegrees=45
  twistMotion=Free
 }
 {
  bone1Name=dyn_r_Sidestraps_MSTR
  bone2Name=pelvis
  swing1LimitDegrees=45
  swing1Motion=Free
  swing2LimitDegrees=45
  swing2Motion=Free
  twistLimitDegrees=45
  twistMotion=Free
 }
 {
  bone1Name=dyn_r_straps_MSTR
  bone2Name=pelvis
  swing1LimitDegrees=45
  swing1Motion=Free
  swing2LimitDegrees=45
  swing2Motion=Free
  twistLimitDegrees=45
  twistMotion=Free
 }
 {
  bone1Name=dyn_l_Sidestraps_MSTR
  bone2Name=pelvis
  swing1LimitDegrees=45
  swing1Motion=Free
  swing2LimitDegrees=45
  swing2Motion=Free
  twistLimitDegrees=45
  twistMotion=Free
 }
 {
  bone1Name=pelvis
  bone2Name=root
  swing1LimitDegrees=45
  swing1Motion=Free
  swing2LimitDegrees=45
  swing2Motion=Free
  twistLimitDegrees=45
  twistMotion=Free
 }
]
