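# Post-process style cosmetic AnimBP for the M_MED_Imitator body, adding secondary motion to its straps and side straps.
# It copies the parent mesh pose through the Despair deform Control Rig, switching in a RigidBody branch for front-end or DBNO carry.
# Constraints, additive bone offsets and half-weight per-bone AnimDynamics, some planar-limited by thighs, then shape the straps.
# EventGraph updates movement, carry, gravity and backbling variables; two strap chains sit unconnected.
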AnimBlueprint /BRCosmetics/Characters/Player/Male/Medium/Bodies/M_MED_Imitator/Meshes/M_MED_Imitator_AnimBP
target_skeleton=&/M_MED_Imitator_Skeleton
graphs [
 AnimGraph {nodes [
   n1 = Get isDBNOCarryFloat() -> (isDBNOCarryFloat:double)
   n2 = "Anim Dynamics - Chain: dyn_r_straps_01 -> dyn_r_straps_08"!(ComponentPose, Alpha=@n1, GravityOverride:Vector="0,0,-980", ExternalForce="0,0,0") -> (Pose) {
    angularDampingOverride=1
    bChain=true
    bOverrideAngularDamping=true
    bUseGravityOverride=true
    simulationSpace=RootRelative
    alphaBoolBlend {blendOption=Linear}
    alphaScaleBiasClamp {bClampResult=true, bInterpResult=true}
    boundBone {boneName=dyn_r_straps_01}
    chainEnd {boneName=dyn_r_straps_08}
    componentAppliedLinearAccClamp=<0 0 0>
    gravityOverride {z=-980}
    physicsBodyDefinitions [
     {
      boundBone {boneName=dyn_r_straps_01}
      boxExtents=<25 35 35>
      constraintSetup {angularLimitsMax=<45 45 30>, angularLimitsMin=<-45 -45 -30>, angularTarget=<0 0 0>}
      localJointOffset=<4 0 0>
     }
     {
      boundBone {boneName=dyn_r_straps_02}
      boxExtents=<25 35 35>
      constraintSetup {angularLimitsMax=<45 45 30>, angularLimitsMin=<-45 -45 -30>, angularTarget=<0 0 0>}
      localJointOffset=<4 0 0>
     }
     {
      boundBone {boneName=dyn_r_straps_03}
      boxExtents=<25 35 35>
      constraintSetup {angularLimitsMax=<45 45 30>, angularLimitsMin=<-45 -45 -30>, angularTarget=<0 0 0>}
      localJointOffset=<4 0 0>
     }
     {
      boundBone {boneName=dyn_r_straps_05}
      boxExtents=<25 35 35>
      constraintSetup {angularLimitsMax=<45 45 30>, angularLimitsMin=<-45 -45 -30>, angularTarget=<0 0 0>}
      localJointOffset=<4 0 0>
     }
     {
      boundBone {boneName=dyn_r_straps_06}
      boxExtents=<25 35 35>
      constraintSetup {angularLimitsMax=<45 45 30>, angularLimitsMin=<-45 -45 -30>, angularTarget=<0 0 0>}
      localJointOffset=<4 0 0>
     }
     {
      boundBone {boneName=dyn_r_straps_08}
      boxExtents=<25 35 35>
      constraintSetup {angularLimitsMax=<45 45 30>, angularLimitsMin=<-45 -45 -30>, angularTarget=<0 0 0>}
      localJointOffset=<4 0 0>
     }
    ]
    relativeSpaceBone {boneName=root}
   }
   "Anim Dynamics - Chain: dyn_l_straps_01 -> dyn_l_straps_08"!(ComponentPose=@n2, Alpha=@n1, GravityOverride:Vector="0,0,-980", ExternalForce="0,0,0") -> (Pose) {
    angularDampingOverride=1
    bChain=true
    bOverrideAngularDamping=true
    bUseGravityOverride=true
    simulationSpace=RootRelative
    alphaBoolBlend {blendOption=Linear}
    alphaScaleBiasClamp {bClampResult=true, bInterpResult=true}
    boundBone {boneName=dyn_l_straps_01}
    chainEnd {boneName=dyn_l_straps_08}
    componentAppliedLinearAccClamp=<0 0 0>
    gravityOverride {z=-980}
    physicsBodyDefinitions [
     {
      boundBone {boneName=dyn_l_straps_01}
      boxExtents=<10 20 20>
      constraintSetup {angularLimitsMax=<45 45 30>, angularLimitsMin=<-45 -45 -30>, angularTarget=<0 0 0>}
      localJointOffset=<4 0 0>
     }
     {
      boundBone {boneName=dyn_l_straps_02}
      boxExtents=<10 20 20>
      constraintSetup {angularLimitsMax=<45 45 30>, angularLimitsMin=<-45 -45 -30>, angularTarget=<0 0 0>}
      localJointOffset=<4 0 0>
     }
     {
      boundBone {boneName=dyn_l_straps_03}
      boxExtents=<10 20 20>
      constraintSetup {angularLimitsMax=<45 45 30>, angularLimitsMin=<-45 -45 -30>, angularTarget=<0 0 0>}
      localJointOffset=<4 0 0>
     }
     {
      boundBone {boneName=dyn_l_straps_04}
      boxExtents=<10 20 20>
      constraintSetup {angularLimitsMax=<45 45 30>, angularLimitsMin=<-45 -45 -30>, angularTarget=<0 0 0>}
      localJointOffset=<4 0 0>
     }
     {
      boundBone {boneName=dyn_l_straps_06}
      boxExtents=<10 20 20>
      constraintSetup {angularLimitsMax=<45 45 30>, angularLimitsMin=<-45 -45 -30>, angularTarget=<0 0 0>}
      localJointOffset=<4 0 0>
     }
     {
      boundBone {boneName=dyn_l_straps_08}
      boxExtents=<10 20 20>
      constraintSetup {angularLimitsMax=<45 45 30>, angularLimitsMin=<-45 -45 -30>, angularTarget=<0 0 0>}
      localJointOffset=<4 0 0>
     }
    ]
    relativeSpaceBone {boneName=root}
   }
   n3 = "Transform (Modify) Bone - Bone: dyn_l_Sidestraps_MSTR"!(ComponentPose, bAlphaBoolEnabled:bool=True, Translation="0,0,0", Rotation="60,0,0", Scale="1,1,1") -> (Pose) {
    alphaInputType=Bool
    rotationMode=BMM_Additive
    rotationSpace=BCS_BoneSpace
    scaleMode=BMM_Additive
    scaleSpace=BCS_BoneSpace
    translationMode=BMM_Additive
    translationSpace=BCS_BoneSpace
    alphaBoolBlend {blendInTime=0.1, blendOption=Linear, blendOutTime=0.1}
    boneToModify {boneName=dyn_l_Sidestraps_MSTR}
    rotation {pitch=60}
   }
   "Transform (Modify) Bone - Bone: dyn_r_Sidestraps_MSTR"!(ComponentPose=@n3, bAlphaBoolEnabled:bool=True, Translation="0,0,0", Rotation="60,0,0", Scale="1,1,1") -> (Pose) {
    alphaInputType=Bool
    rotationMode=BMM_Additive
    rotationSpace=BCS_BoneSpace
    scaleMode=BMM_Additive
    scaleSpace=BCS_BoneSpace
    translationMode=BMM_Additive
    translationSpace=BCS_BoneSpace
    alphaBoolBlend {blendInTime=0.1, blendOption=Linear, blendOutTime=0.1}
    boneToModify {boneName=dyn_r_Sidestraps_MSTR}
    rotation {pitch=60}
   }
   n4 = Control Rig(Source=Copy Pose From Mesh() {bUseAttachedParent=true}) {
    alphaBoolBlend {blendOption=Linear}
    controlRigAssetReference {blueprintRigClass=/BRCosmetics/Characters/Player/Male/Medium/Bodies/M_MED_Despair/Meshes/M_MED_Despair_DeformRig_Master_ControlRig_new.M_MED_Despair_DeformRig_Master_ControlRig_new_C}
   }
   n5 = RigidBody!(ComponentPose=Local To Component(LocalPose=@n4), Alpha, OverrideWorldGravity:Vector="0,0,-980", ExternalForce) -> (Pose) {
    bClampLinearTranslationLimitToRefPose=true
    bOverrideWorldGravity=true
    overridePhysicsAsset=&/M_MED_Imitator_Physics
    simulationSpace=BaseBoneSpace
    alphaBoolBlend {blendOption=Linear}
    baseBoneRef {boneName=pelvis}
    componentAppliedLinearAccClamp=<5000 5000 5000>
    componentLinearAccScale=<0.4 0.4 0.5>
    overrideWorldGravity {z=-980}
    simSpaceSettings {
     maxAngularAcceleration=1000
     maxAngularVelocity=3
     maxLinearAcceleration=700
     maxLinearVelocity=500
     velocityScaleZ=0
     worldAlpha=1
     externalLinearDragV=<0.2 0.2>
    }
   }
   n6 = Component To Local(ComponentPose=@n5)
   n7 = Control Rig(Source=Copy Pose From Mesh() {bUseAttachedParent=true}) {
    alphaBoolBlend {blendOption=Linear}
    controlRigAssetReference {blueprintRigClass=&M_MED_Despair_DeformRig_Master_ControlRig_new_C}
   }
   n8 = OR Boolean!(A=Get bIsInFrontEnd(), B=Get bIsDBNOCarried(), C:bool=Get bIsDBNOCarrying() -> (bIsDBNOCarrying:bool)) -> (ReturnValue)
   n9 = "Constraint - dyn_l_Sidestraps_MSTR (PPPP)"!(
    ComponentPose=Local To Component(LocalPose=Blend Poses by bool(BlendPose_0=@n6, BlendPose_1=@n7, bActiveValue=@n8) {blendType=Linear})
    Alpha=1
    ConstraintWeights_0=0.1896
    ConstraintWeights_1:float=0.0001
    ConstraintWeights_2:float=0.3846
    ConstraintWeights_3:float=0.4256
   ) -> (Pose) {
    alphaBoolBlend {blendOption=Linear}
    boneToModify {boneName=dyn_l_Sidestraps_MSTR}
    constraintSetup [
     {targetBone {boneName=pelvis}}
     {targetBone {boneName=thigh_l}}
     {targetBone {boneName=thigh_twist_01_l}}
     {targetBone {boneName=deform_glute_l}}
    ]
    constraintWeights [1,1,1,1]
   }
   n10 = "Constraint - dyn_r_Sidestraps_MSTR (PPPP)"!(
    ComponentPose=@n9
    Alpha=1
    ConstraintWeights_0=0.1868
    ConstraintWeights_1:float=0.0003
    ConstraintWeights_2:float=0.3906
    ConstraintWeights_3:float=0.4223
   ) -> (Pose) {
    alphaBoolBlend {blendOption=Linear}
    boneToModify {boneName=dyn_r_Sidestraps_MSTR}
    constraintSetup [
     {targetBone {boneName=pelvis}}
     {targetBone {boneName=thigh_r}}
     {targetBone {boneName=thigh_twist_01_r}}
     {targetBone {boneName=deform_glute_r}}
    ]
    constraintWeights [1,1,1,1]
   }
   n11 = "Transform (Modify) Bone - Bone: dyn_r_straps_MSTR"!(ComponentPose=@n10, bAlphaBoolEnabled:bool=True, Translation="0,0,0", Rotation="20,0.000,0.000", Scale="1,1,1") -> (Pose) {
    alphaInputType=Bool
    rotationMode=BMM_Additive
    rotationSpace=BCS_BoneSpace
    scaleMode=BMM_Additive
    scaleSpace=BCS_BoneSpace
    translationMode=BMM_Additive
    translationSpace=BCS_BoneSpace
    alphaBoolBlend {blendInTime=0.1, blendOption=Linear, blendOutTime=0.1}
    boneToModify {boneName=dyn_r_straps_MSTR}
    rotation {pitch=20}
   }
   n12 = "Transform (Modify) Bone - Bone: dyn_l_straps_MSTR"!(ComponentPose=@n11, bAlphaBoolEnabled:bool=True, Translation="0,0,0", Rotation="20,0,0", Scale="1,1,1") -> (Pose) {
    alphaInputType=Bool
    rotationMode=BMM_Additive
    rotationSpace=BCS_BoneSpace
    scaleMode=BMM_Additive
    scaleSpace=BCS_BoneSpace
    translationMode=BMM_Additive
    translationSpace=BCS_BoneSpace
    alphaBoolBlend {blendInTime=0.1, blendOption=Linear, blendOutTime=0.1}
    boneToModify {boneName=dyn_l_straps_MSTR}
    rotation {pitch=20}
   }
   n13 = "Transform (Modify) Bone - Bone: dyn_l_straps_MSTR"(ComponentPose=@n12, bAlphaBoolEnabled=True, Translation="0,0,0", Rotation="60,0.000,0.000", Scale="1,1,1") {
    alphaInputType=Bool
    rotationMode=BMM_Additive
    rotationSpace=BCS_BoneSpace
    scaleMode=BMM_Additive
    scaleSpace=BCS_BoneSpace
    translationMode=BMM_Additive
    translationSpace=BCS_BoneSpace
    alphaBoolBlend {blendInTime=1, blendOption=Linear, blendOutTime=1}
    boneToModify {boneName=dyn_l_straps_MSTR}
    rotation {pitch=60}
   }
   n14 = "Transform (Modify) Bone - Bone: dyn_r_straps_MSTR"(ComponentPose=@n13, bAlphaBoolEnabled=True, Translation="0,0,0", Rotation="60,0,0", Scale="1,1,1") {
    alphaInputType=Bool
    rotationMode=BMM_Additive
    rotationSpace=BCS_BoneSpace
    scaleMode=BMM_Additive
    scaleSpace=BCS_BoneSpace
    translationMode=BMM_Additive
    translationSpace=BCS_BoneSpace
    alphaBoolBlend {blendInTime=1, blendOption=Linear, blendOutTime=1}
    boneToModify {boneName=dyn_r_straps_MSTR}
    rotation {pitch=60}
   }
   n15 = "Transform (Modify) Bone - Bone: dyn_r_straps_01"!(ComponentPose=@n14, bAlphaBoolEnabled:bool=True, Translation="0,0,0", Rotation="-60,0.000,0.000", Scale="1,1,1") -> (Pose) {
    alphaInputType=Bool
    rotationMode=BMM_Additive
    rotationSpace=BCS_BoneSpace
    scaleMode=BMM_Additive
    scaleSpace=BCS_BoneSpace
    translationMode=BMM_Additive
    translationSpace=BCS_BoneSpace
    alphaBoolBlend {blendInTime=1, blendOption=Linear, blendOutTime=1}
    boneToModify {boneName=dyn_r_straps_01}
    rotation {pitch=-60}
   }
   n16 = "Transform (Modify) Bone - Bone: dyn_r_straps_MSTR"(ComponentPose=@n15, bAlphaBoolEnabled=True, Translation="0,0,0", Rotation="60,0,0", Scale="1,1,1") {
    alphaInputType=Bool
    rotationMode=BMM_Additive
    rotationSpace=BCS_BoneSpace
    scaleMode=BMM_Additive
    scaleSpace=BCS_BoneSpace
    translationMode=BMM_Additive
    translationSpace=BCS_BoneSpace
    alphaBoolBlend {blendInTime=1, blendOption=Linear, blendOutTime=1}
    boneToModify {boneName=dyn_r_straps_MSTR}
    rotation {pitch=60}
   }
   n17 = Anim Dynamics - Bone: dyn_l_straps_01!(ComponentPose=@n16, Alpha=0.5, GravityOverride:Vector="0,0,-980", ExternalForce="0,0,0") -> (Pose) {
    angularDampingOverride=1
    bOverrideAngularDamping=true
    bUseGravityOverride=true
    simulationSpace=BoneRelative
    alphaBoolBlend {blendOption=Linear}
    boundBone {boneName=dyn_l_straps_01}
    componentAppliedLinearAccClamp=<3500 3500 3500>
    componentLinearAccScale=<0.3 0.3 0.3>
    componentLinearVelScale=<0.3 0.3 0.3>
    gravityOverride {z=-980}
    physicsBodyDefinitions [{
      boundBone {boneName=dyn_l_straps_01}
      boxExtents=<10 20 20>
      constraintSetup {angularLimitsMax=<45 45 30>, angularLimitsMin=<-45 -45 -30>, angularTarget=<0 0 0>}
      localJointOffset=<4 0 0>
     }]
    planarLimits [{drivingBone {boneName=thigh_l}, planeTransform {location=<0 -3 0>, rotation=<rot 0 0 -90>}}]
    relativeSpaceBone {boneName=root}
   }
   n18 = Anim Dynamics - Bone: dyn_r_straps_01!(ComponentPose=@n17, Alpha=0.5, GravityOverride:Vector="0,0,-980", ExternalForce="0,0,0") -> (Pose) {
    angularDampingOverride=1
    bOverrideAngularDamping=true
    bUseGravityOverride=true
    bUsePlanarLimit=false
    simulationSpace=BoneRelative
    alphaBoolBlend {blendOption=Linear}
    boundBone {boneName=dyn_r_straps_01}
    componentAppliedLinearAccClamp=<3500 3500 3500>
    componentLinearAccScale=<0.3 0.3 0.3>
    componentLinearVelScale=<0.3 0.3 0.3>
    gravityOverride {z=-980}
    physicsBodyDefinitions [{
      boundBone {boneName=dyn_r_straps_01}
      boxExtents=<12 15 15>
      constraintSetup {angularLimitsMax=<45 45 30>, angularLimitsMin=<-45 -45 -30>, angularTarget=<0 0 0>}
      localJointOffset=<4 0 0>
     }]
    planarLimits [{drivingBone {boneName=thigh_r}, planeTransform {location=<0 3 0>, rotation=<rot 0 0 90>}}]
    relativeSpaceBone {boneName=root}
   }
   n19 = Anim Dynamics - Bone: dyn_l_straps_02!(ComponentPose=@n18, Alpha=0.5, GravityOverride:Vector="0,0,-980", ExternalForce="0,0,0") -> (Pose) {
    angularDampingOverride=1
    bOverrideAngularDamping=true
    bUseGravityOverride=true
    simulationSpace=BoneRelative
    alphaBoolBlend {blendOption=Linear}
    boundBone {boneName=dyn_l_straps_02}
    componentAppliedLinearAccClamp=<3500 3500 3500>
    componentLinearAccScale=<0.3 0.3 0.3>
    componentLinearVelScale=<0.3 0.3 0.3>
    gravityOverride {z=-980}
    physicsBodyDefinitions [{
      boundBone {boneName=dyn_l_straps_02}
      boxExtents=<10 20 20>
      constraintSetup {angularLimitsMax=<45 45 30>, angularLimitsMin=<-45 -45 -30>, angularTarget=<0 0 0>}
      localJointOffset=<4 0 0>
     }]
    planarLimits [{drivingBone {boneName=thigh_l}, planeTransform {location=<0 -4 0>, rotation=<rot 0 0 -90>}}]
    relativeSpaceBone {boneName=root}
   }
   n20 = Anim Dynamics - Bone: dyn_r_straps_02!(ComponentPose=@n19, Alpha=0.5, GravityOverride:Vector="0,0,-980", ExternalForce="0,0,0") -> (Pose) {
    angularDampingOverride=1
    bOverrideAngularDamping=true
    bUseGravityOverride=true
    bUsePlanarLimit=false
    simulationSpace=BoneRelative
    alphaBoolBlend {blendOption=Linear}
    boundBone {boneName=dyn_r_straps_02}
    componentAppliedLinearAccClamp=<3500 3500 3500>
    componentLinearAccScale=<0.3 0.3 0.3>
    componentLinearVelScale=<0.3 0.3 0.3>
    gravityOverride {z=-980}
    physicsBodyDefinitions [{
      boundBone {boneName=dyn_r_straps_02}
      boxExtents=<12 15 15>
      constraintSetup {angularLimitsMax=<45 45 30>, angularLimitsMin=<-45 -45 -30>, angularTarget=<0 0 0>}
      localJointOffset=<4 0 0>
     }]
    planarLimits [{drivingBone {boneName=thigh_r}, planeTransform {location=<0 4 0>, rotation=<rot 0 0 90>}}]
    relativeSpaceBone {boneName=root}
   }
   n21 = Anim Dynamics - Bone: dyn_l_straps_03!(ComponentPose=@n20, Alpha=0.5, GravityOverride:Vector="0,0,-980", ExternalForce="0,0,0") -> (Pose) {
    angularDampingOverride=1
    bOverrideAngularDamping=true
    bUseGravityOverride=true
    simulationSpace=BoneRelative
    alphaBoolBlend {blendOption=Linear}
    boundBone {boneName=dyn_l_straps_03}
    componentAppliedLinearAccClamp=<3500 3500 3500>
    componentLinearAccScale=<0.3 0.3 0.3>
    componentLinearVelScale=<0.3 0.3 0.3>
    gravityOverride {z=-980}
    physicsBodyDefinitions [{
      boundBone {boneName=dyn_l_straps_03}
      boxExtents=<10 20 20>
      constraintSetup {angularLimitsMax=<45 45 30>, angularLimitsMin=<-45 -45 -30>, angularTarget=<0 0 0>}
      localJointOffset=<4 0 0>
     }]
    planarLimits [{drivingBone {boneName=thigh_l}, planeTransform {location=<0 -5 0>, rotation=<rot 0 0 -90>}}]
    relativeSpaceBone {boneName=root}
   }
   n22 = Anim Dynamics - Bone: dyn_r_straps_03!(ComponentPose=@n21, Alpha=0.5, GravityOverride:Vector="0,0,-980", ExternalForce="0,0,0") -> (Pose) {
    angularDampingOverride=1
    bOverrideAngularDamping=true
    bUseGravityOverride=true
    bUsePlanarLimit=false
    simulationSpace=BoneRelative
    alphaBoolBlend {blendOption=Linear}
    boundBone {boneName=dyn_r_straps_03}
    componentAppliedLinearAccClamp=<3500 3500 3500>
    componentLinearAccScale=<0.3 0.3 0.3>
    componentLinearVelScale=<0.3 0.3 0.3>
    gravityOverride {z=-980}
    physicsBodyDefinitions [{
      boundBone {boneName=dyn_r_straps_03}
      boxExtents=<12 15 15>
      constraintSetup {angularLimitsMax=<45 45 30>, angularLimitsMin=<-45 -45 -30>, angularTarget=<0 0 0>}
      localJointOffset=<4 0 0>
     }]
    planarLimits [{drivingBone {boneName=thigh_r}, planeTransform {location=<0 5 0>, rotation=<rot 0 0 90>}}]
    relativeSpaceBone {boneName=root}
   }
   n23 = Anim Dynamics - Bone: dyn_l_straps_04!(ComponentPose=@n22, Alpha=0.5, GravityOverride:Vector="0,0,-980", ExternalForce="0,0,0") -> (Pose) {
    angularDampingOverride=1
    bOverrideAngularDamping=true
    bUseGravityOverride=true
    simulationSpace=BoneRelative
    alphaBoolBlend {blendOption=Linear}
    boundBone {boneName=dyn_l_straps_04}
    componentAppliedLinearAccClamp=<3500 3500 3500>
    componentLinearAccScale=<0.3 0.3 0.3>
    componentLinearVelScale=<0.3 0.3 0.3>
    gravityOverride {z=-980}
    physicsBodyDefinitions [{
      boundBone {boneName=dyn_l_straps_04}
      boxExtents=<10 20 20>
      constraintSetup {angularLimitsMax=<45 45 30>, angularLimitsMin=<-45 -45 -30>, angularTarget=<0 0 0>}
      localJointOffset=<4 0 0>
     }]
    planarLimits [{drivingBone {boneName=thigh_l}, planeTransform {location=<-20 -6 0>, rotation=<rot 0 0 -90>}}]
    relativeSpaceBone {boneName=root}
   }
   n24 = Anim Dynamics - Bone: dyn_r_straps_04!(ComponentPose=@n23, Alpha=0.5, GravityOverride:Vector="0,0,-980", ExternalForce="0,0,0") -> (Pose) {
    angularDampingOverride=1
    bOverrideAngularDamping=true
    bUseGravityOverride=true
    simulationSpace=BoneRelative
    alphaBoolBlend {blendOption=Linear}
    boundBone {boneName=dyn_r_straps_04}
    componentAppliedLinearAccClamp=<3500 3500 3500>
    componentLinearAccScale=<0.3 0.3 0.3>
    componentLinearVelScale=<0.3 0.3 0.3>
    gravityOverride {z=-980}
    physicsBodyDefinitions [{
      boundBone {boneName=dyn_r_straps_04}
      boxExtents=<12 15 15>
      constraintSetup {angularLimitsMax=<45 45 30>, angularLimitsMin=<-45 -45 -30>, angularTarget=<0 0 0>}
      localJointOffset=<4 0 0>
     }]
    planarLimits [{drivingBone {boneName=thigh_r}, planeTransform {location=<20 6 0>, rotation=<rot 0 0 90>}}]
    relativeSpaceBone {boneName=root}
   }
   n25 = Anim Dynamics - Bone: dyn_l_straps_05!(ComponentPose=@n24, Alpha=0.5, GravityOverride:Vector="0,0,-980", ExternalForce="0,0,0") -> (Pose) {
    angularDampingOverride=1
    bOverrideAngularDamping=true
    bUseGravityOverride=true
    simulationSpace=BoneRelative
    alphaBoolBlend {blendOption=Linear}
    boundBone {boneName=dyn_l_straps_05}
    componentAppliedLinearAccClamp=<3500 3500 3500>
    componentLinearAccScale=<0.3 0.3 0.3>
    componentLinearVelScale=<0.3 0.3 0.3>
    gravityOverride {z=-980}
    physicsBodyDefinitions [{
      boundBone {boneName=dyn_l_straps_05}
      boxExtents=<10 20 20>
      constraintSetup {angularLimitsMax=<45 45 30>, angularLimitsMin=<-45 -45 -30>, angularTarget=<0 0 0>}
      localJointOffset=<4 0 0>
     }]
    planarLimits [{drivingBone {boneName=thigh_l}, planeTransform {location=<-30 -7 0>, rotation=<rot 0 0 -90>}}]
    relativeSpaceBone {boneName=root}
   }
   n26 = Anim Dynamics - Bone: dyn_r_straps_05!(ComponentPose=@n25, Alpha=0.5, GravityOverride:Vector="0,0,-980", ExternalForce="0,0,0") -> (Pose) {
    angularDampingOverride=1
    bOverrideAngularDamping=true
    bUseGravityOverride=true
    simulationSpace=BoneRelative
    alphaBoolBlend {blendOption=Linear}
    boundBone {boneName=dyn_r_straps_05}
    componentAppliedLinearAccClamp=<3500 3500 3500>
    componentLinearAccScale=<0.3 0.3 0.3>
    componentLinearVelScale=<0.3 0.3 0.3>
    gravityOverride {z=-980}
    physicsBodyDefinitions [{
      boundBone {boneName=dyn_r_straps_05}
      boxExtents=<12 15 15>
      constraintSetup {angularLimitsMax=<45 45 30>, angularLimitsMin=<-45 -45 -30>, angularTarget=<0 0 0>}
      localJointOffset=<4 0 0>
     }]
    planarLimits [{drivingBone {boneName=thigh_r}, planeTransform {location=<30 7 0>, rotation=<rot 0 0 90>}}]
    relativeSpaceBone {boneName=root}
   }
   n27 = Anim Dynamics - Bone: dyn_l_straps_06!(ComponentPose=@n26, Alpha=0.5, GravityOverride:Vector="0,0,-980", ExternalForce="0,0,0") -> (Pose) {
    angularDampingOverride=1
    bOverrideAngularDamping=true
    bUseGravityOverride=true
    simulationSpace=BoneRelative
    alphaBoolBlend {blendOption=Linear}
    boundBone {boneName=dyn_l_straps_06}
    componentAppliedLinearAccClamp=<3500 3500 3500>
    componentLinearAccScale=<0.3 0.3 0.3>
    componentLinearVelScale=<0.3 0.3 0.3>
    gravityOverride {z=-980}
    physicsBodyDefinitions [{
      boundBone {boneName=dyn_l_straps_06}
      boxExtents=<10 20 20>
      constraintSetup {angularLimitsMax=<45 45 30>, angularLimitsMin=<-45 -45 -30>, angularTarget=<0 0 0>}
      localJointOffset=<4 0 0>
     }]
    planarLimits [{drivingBone {boneName=thigh_l}, planeTransform {location=<-50 -8 0>, rotation=<rot 0 0 -90>}}]
    relativeSpaceBone {boneName=root}
   }
   n28 = Anim Dynamics - Bone: dyn_r_straps_06!(ComponentPose=@n27, Alpha=0.5, GravityOverride:Vector="0,0,-980", ExternalForce="0,0,0") -> (Pose) {
    angularDampingOverride=1
    bOverrideAngularDamping=true
    bUseGravityOverride=true
    simulationSpace=BoneRelative
    alphaBoolBlend {blendOption=Linear}
    boundBone {boneName=dyn_r_straps_06}
    componentAppliedLinearAccClamp=<3500 3500 3500>
    componentLinearAccScale=<0.3 0.3 0.3>
    componentLinearVelScale=<0.3 0.3 0.3>
    gravityOverride {z=-980}
    physicsBodyDefinitions [{
      boundBone {boneName=dyn_r_straps_06}
      boxExtents=<12 15 15>
      constraintSetup {angularLimitsMax=<45 45 30>, angularLimitsMin=<-45 -45 -30>, angularTarget=<0 0 0>}
      localJointOffset=<4 0 0>
     }]
    planarLimits [{drivingBone {boneName=thigh_r}, planeTransform {location=<50 8 0>, rotation=<rot 0 0 90>}}]
    relativeSpaceBone {boneName=root}
   }
   n29 = Anim Dynamics - Bone: dyn_l_straps_07!(ComponentPose=@n28, Alpha=0.5, GravityOverride:Vector="0,0,-980", ExternalForce="0,0,0") -> (Pose) {
    angularDampingOverride=1
    bOverrideAngularDamping=true
    bUseGravityOverride=true
    simulationSpace=BoneRelative
    alphaBoolBlend {blendOption=Linear}
    boundBone {boneName=dyn_l_straps_07}
    componentAppliedLinearAccClamp=<3500 3500 3500>
    componentLinearAccScale=<0.3 0.3 0.3>
    componentLinearVelScale=<0.3 0.3 0.3>
    gravityOverride {z=-980}
    physicsBodyDefinitions [{
      boundBone {boneName=dyn_l_straps_07}
      boxExtents=<10 20 20>
      constraintSetup {angularLimitsMax=<45 45 30>, angularLimitsMin=<-45 -45 -30>, angularTarget=<0 0 0>}
      localJointOffset=<4 0 0>
     }]
    planarLimits [{drivingBone {boneName=thigh_l}, planeTransform {location=<-50 -9 0>, rotation=<rot 0 0 -90>}}]
    relativeSpaceBone {boneName=root}
   }
   n30 = Anim Dynamics - Bone: dyn_r_straps_07!(ComponentPose=@n29, Alpha=0.5, GravityOverride:Vector="0,0,-980", ExternalForce="0,0,0") -> (Pose) {
    angularDampingOverride=1
    bOverrideAngularDamping=true
    bUseGravityOverride=true
    simulationSpace=BoneRelative
    alphaBoolBlend {blendOption=Linear}
    boundBone {boneName=dyn_r_straps_07}
    componentAppliedLinearAccClamp=<3500 3500 3500>
    componentLinearAccScale=<0.3 0.3 0.3>
    componentLinearVelScale=<0.3 0.3 0.3>
    gravityOverride {z=-980}
    physicsBodyDefinitions [{
      boundBone {boneName=dyn_r_straps_07}
      boxExtents=<12 15 15>
      constraintSetup {angularLimitsMax=<45 45 30>, angularLimitsMin=<-45 -45 -30>, angularTarget=<0 0 0>}
      localJointOffset=<4 0 0>
     }]
    planarLimits [{drivingBone {boneName=thigh_r}, planeTransform {location=<50 9 0>, rotation=<rot 0 0 90>}}]
    relativeSpaceBone {boneName=root}
   }
   n31 = Anim Dynamics - Bone: dyn_l_straps_08!(ComponentPose=@n30, Alpha=0.5, GravityOverride:Vector="0,0,-980", ExternalForce="0,0,0") -> (Pose) {
    angularDampingOverride=1
    bOverrideAngularDamping=true
    bUseGravityOverride=true
    simulationSpace=BoneRelative
    alphaBoolBlend {blendOption=Linear}
    boundBone {boneName=dyn_l_straps_08}
    componentAppliedLinearAccClamp=<3500 3500 3500>
    componentLinearAccScale=<0.3 0.3 0.3>
    componentLinearVelScale=<0.3 0.3 0.3>
    gravityOverride {z=-980}
    physicsBodyDefinitions [{
      boundBone {boneName=dyn_l_straps_08}
      boxExtents=<10 20 20>
      constraintSetup {angularLimitsMax=<45 45 30>, angularLimitsMin=<-45 -45 -30>, angularTarget=<0 0 0>}
      localJointOffset=<4 0 0>
     }]
    planarLimits [{drivingBone {boneName=thigh_l}, planeTransform {location=<-50 -10 0>, rotation=<rot 0 0 -90>}}]
    relativeSpaceBone {boneName=root}
   }
   n32 = Anim Dynamics - Bone: dyn_r_straps_08!(ComponentPose=@n31, Alpha=0.5, GravityOverride:Vector="0,0,-980", ExternalForce="0,0,0") -> (Pose) {
    angularDampingOverride=1
    bOverrideAngularDamping=true
    bUseGravityOverride=true
    simulationSpace=BoneRelative
    alphaBoolBlend {blendOption=Linear}
    boundBone {boneName=dyn_r_straps_08}
    componentAppliedLinearAccClamp=<3500 3500 3500>
    componentLinearAccScale=<0.3 0.3 0.3>
    componentLinearVelScale=<0.3 0.3 0.3>
    gravityOverride {z=-980}
    physicsBodyDefinitions [{
      boundBone {boneName=dyn_r_straps_08}
      boxExtents=<12 15 15>
      constraintSetup {angularLimitsMax=<45 45 30>, angularLimitsMin=<-45 -45 -30>, angularTarget=<0 0 0>}
      localJointOffset=<4 0 0>
     }]
    planarLimits [{drivingBone {boneName=thigh_r}, planeTransform {location=<50 10 0>, rotation=<rot 0 0 90>}}]
    relativeSpaceBone {boneName=root}
   }
   n33 = "Anim Dynamics - Chain: dyn_l_Sidestraps_01 -> dyn_l_Sidestraps_02"!(ComponentPose=@n32, Alpha=1, GravityOverride:Vector="0,0,-980", ExternalForce="0,0,0") -> (Pose) {
    angularDampingOverride=1
    bChain=true
    bOverrideAngularDamping=true
    bUseGravityOverride=true
    simulationSpace=BoneRelative
    alphaBoolBlend {blendOption=Linear}
    boundBone {boneName=dyn_l_Sidestraps_01}
    chainEnd {boneName=dyn_l_Sidestraps_02}
    componentAppliedLinearAccClamp=<2000 2000 2000>
    componentLinearAccScale=<0.3 0.3 0.3>
    componentLinearVelScale=<0.3 0.3 0.3>
    gravityOverride {z=-980}
    physicsBodyDefinitions [
     {
      boundBone {boneName=dyn_l_Sidestraps_01}
      boxExtents=<25 35 35>
      constraintSetup {angularLimitsMax=<1 -3 30>, angularLimitsMin=<-1 -45 -30>, angularTarget=<0 0 0>}
      localJointOffset=<4 0 0>
     }
     {
      boundBone {boneName=dyn_l_Sidestraps_02}
      boxExtents=<25 35 35>
      constraintSetup {angularLimitsMax=<1 0 30>, angularLimitsMin=<-1 -45 -30>, angularTarget=<0 0 0>}
      localJointOffset=<4 0 0>
     }
    ]
    relativeSpaceBone {boneName=root}
   }
   n34 = "Anim Dynamics - Chain: dyn_r_Sidestraps_01 -> dyn_r_Sidestraps_02"!(ComponentPose=@n33, Alpha=1, GravityOverride:Vector="0,0,-980", ExternalForce="0,0,0") -> (Pose) {
    angularDampingOverride=1
    bChain=true
    bOverrideAngularDamping=true
    bUseGravityOverride=true
    simulationSpace=BoneRelative
    alphaBoolBlend {blendOption=Linear}
    boundBone {boneName=dyn_r_Sidestraps_01}
    chainEnd {boneName=dyn_r_Sidestraps_02}
    componentAppliedLinearAccClamp=<3500 3500 3500>
    componentLinearAccScale=<0.3 0.3 0.3>
    componentLinearVelScale=<0.3 0.3 0.3>
    gravityOverride {z=-980}
    physicsBodyDefinitions [
     {
      boundBone {boneName=dyn_r_Sidestraps_01}
      boxExtents=<10 20 20>
      constraintSetup {angularLimitsMax=<1 -3 30>, angularLimitsMin=<-1 -45 -30>, angularTarget=<0 0 0>}
      localJointOffset=<4 0 0>
     }
     {
      boundBone {boneName=dyn_r_Sidestraps_02}
      boxExtents=<10 20 20>
      constraintSetup {angularLimitsMax=<1 0 30>, angularLimitsMin=<-1 -45 -30>, angularTarget=<0 0 0>}
      localJointOffset=<4 0 0>
     }
    ]
    relativeSpaceBone {boneName=root}
   }
   n35 = "Transform (Modify) Bone - Bone: dyn_l_Sidestraps_MSTR"(ComponentPose=@n34, bAlphaBoolEnabled=True, Translation="0,0,0", Rotation="1.859,-36.447,2.826", Scale="1,1,1") {
    alphaInputType=Bool
    rotationMode=BMM_Additive
    rotationSpace=BCS_BoneSpace
    scaleMode=BMM_Additive
    scaleSpace=BCS_BoneSpace
    translationMode=BMM_Additive
    translationSpace=BCS_BoneSpace
    alphaBoolBlend {blendInTime=1, blendOption=Linear, blendOutTime=1}
    boneToModify {boneName=dyn_l_Sidestraps_MSTR}
    rotation=<rot 1.859 -36.447 2.826>
   }
   n36 = "Transform (Modify) Bone - Bone: dyn_r_Sidestraps_MSTR"(ComponentPose=@n35, bAlphaBoolEnabled=True, Translation="0,0,0", Rotation="-17.993,30.751,0.036", Scale="1,1,1") {
    alphaInputType=Bool
    rotationMode=BMM_Additive
    rotationSpace=BCS_BoneSpace
    scaleMode=BMM_Additive
    scaleSpace=BCS_BoneSpace
    translationMode=BMM_Additive
    translationSpace=BCS_BoneSpace
    alphaBoolBlend {blendInTime=1, blendOption=Linear, blendOutTime=1}
    boneToModify {boneName=dyn_r_Sidestraps_MSTR}
    rotation=<rot -17.993 30.751 0.036>
   }
   n37 = "Transform (Modify) Bone - Bone: dyn_l_straps_MSTR"(ComponentPose=@n36, bAlphaBoolEnabled=True, Translation="0,0,0", Rotation="0,10,0.000", Scale="1,1,1") {
    alphaInputType=Bool
    rotationMode=BMM_Additive
    rotationSpace=BCS_BoneSpace
    scaleMode=BMM_Additive
    scaleSpace=BCS_BoneSpace
    translationMode=BMM_Additive
    translationSpace=BCS_BoneSpace
    alphaBoolBlend {blendInTime=1, blendOption=Linear, blendOutTime=1}
    boneToModify {boneName=dyn_l_straps_MSTR}
    rotation {yaw=10}
   }
   n38 = "Transform (Modify) Bone - Bone: dyn_r_straps_MSTR"(ComponentPose=@n37, bAlphaBoolEnabled=True, Translation="0,0,0", Rotation="0,-10,0", Scale="1,1,1") {
    alphaInputType=Bool
    rotationMode=BMM_Additive
    rotationSpace=BCS_BoneSpace
    scaleMode=BMM_Additive
    scaleSpace=BCS_BoneSpace
    translationMode=BMM_Additive
    translationSpace=BCS_BoneSpace
    alphaBoolBlend {blendInTime=1, blendOption=Linear, blendOutTime=1}
    boneToModify {boneName=dyn_r_straps_MSTR}
    rotation {yaw=-10}
   }
   n39 = "Transform (Modify) Bone - Bone: dyn_l_Sidestraps_MSTR"(ComponentPose=@n38, bAlphaBoolEnabled=True, Translation="0,0,0", Rotation="0,15,0.000", Scale="1,1,1") {
    alphaInputType=Bool
    rotationMode=BMM_Additive
    rotationSpace=BCS_BoneSpace
    scaleMode=BMM_Additive
    scaleSpace=BCS_BoneSpace
    translationMode=BMM_Additive
    translationSpace=BCS_BoneSpace
    alphaBoolBlend {blendInTime=1, blendOption=Linear, blendOutTime=1}
    boneToModify {boneName=dyn_l_Sidestraps_MSTR}
    rotation {yaw=15}
   }
   n40 = "Transform (Modify) Bone - Bone: dyn_r_Sidestraps_MSTR"(ComponentPose=@n39, bAlphaBoolEnabled=True, Translation="0,0,0", Rotation="0,-10,0", Scale="1,1,1") {
    alphaInputType=Bool
    rotationMode=BMM_Additive
    rotationSpace=BCS_BoneSpace
    scaleMode=BMM_Additive
    scaleSpace=BCS_BoneSpace
    translationMode=BMM_Additive
    translationSpace=BCS_BoneSpace
    alphaBoolBlend {blendInTime=1, blendOption=Linear, blendOutTime=1}
    boneToModify {boneName=dyn_r_Sidestraps_MSTR}
    rotation {yaw=-10}
   }
   n41 = "Transform (Modify) Bone - Bone: dyn_l_straps_MSTR"(ComponentPose=@n40, bAlphaBoolEnabled=True, Translation="0,0,0", Rotation="90,0,0", Scale="1,1,1") {
    alphaInputType=Bool
    rotationMode=BMM_Additive
    rotationSpace=BCS_BoneSpace
    scaleMode=BMM_Additive
    scaleSpace=BCS_BoneSpace
    translationMode=BMM_Additive
    translationSpace=BCS_BoneSpace
    alphaBoolBlend {blendInTime=0.5, blendOption=Linear, blendOutTime=0.5}
    boneToModify {boneName=dyn_l_straps_MSTR}
    rotation {pitch=90}
   }
   n42 = "Transform (Modify) Bone - Bone: dyn_r_straps_MSTR"(ComponentPose=@n41, bAlphaBoolEnabled=True, Translation="0,0,0", Rotation="90,0,0", Scale="1,1,1") {
    alphaInputType=Bool
    rotationMode=BMM_Additive
    rotationSpace=BCS_BoneSpace
    scaleMode=BMM_Additive
    scaleSpace=BCS_BoneSpace
    translationMode=BMM_Additive
    translationSpace=BCS_BoneSpace
    alphaBoolBlend {blendInTime=0.5, blendOption=Linear, blendOutTime=0.5}
    boneToModify {boneName=dyn_r_straps_MSTR}
    rotation {pitch=90}
   }
   Output Pose(Result=Component To Local(ComponentPose=@n42)) {layerGroup=None}
  ]}
 EventGraph {nodes [
   Event BlueprintUpdateAnimation()
   > n1 = Sequence!(execute) -> (then_0, then_1, then_2:Exec, then_3:Exec, then_4:Exec)
   n2 = Gravity Override State Function DEPRECATED(
    execute=@n1.then_1
    CharacterPartAnimInstance=Self-Reference()
    GameTimeInSeconds=GetGameTimeInSeconds()
    GravityOverrideParams=Get Gravity Override Params()
   )
   > Set GravityOverride(GravityOverride=@n2)
   n3 = Get bIsCrouching()
   n4 = Get bIsTargeting()
   n5 = Get bIsPlayingMeleeAnim()
   n6 = Get bIsSkyDiving()
   Set ADS Hair Velocity(
    execute=@n1.then_2
    ADS Hair Velocity:Vector=SelectVector(A="0.15,0.15,0.25", B="0.38,0.38,0.38", bPickA=OR Boolean!(A=@n4, B=@n5, C=@n6, D:bool=@n3) -> (ReturnValue))
   ) -> (Output_Get:Vector)
   Set Body_Mesh(execute=[@n9.then, @n1.then_0], Body_Mesh=GetOwnerSkeletalMeshComponent())
   n7 = Branch(execute=@n1.then_3, Condition=OR Boolean!(A=Get bIsJumping(), B=Get bIsFalling()) -> (ReturnValue))
   > Set IsJumping(IsJumping:bool=true) -> (Output_Get:bool)
   Set IsJumping(execute=@n7.else, IsJumping=false)
   n8 = Branch(execute=@n1.then_4, Condition=OR Boolean(A=Get bIsDBNOCarried(), B=Get bIsDBNOCarrying()))
   > Set isDBNOCarry(isDBNOCarry:bool=true) -> (Output_Get:bool)
   > Set isDBNOCarryFloat(isDBNOCarryFloat:double=0) -> (Output_Get:double)
   Set isDBNOCarry(execute=@n8.else, isDBNOCarry=false)
   > Set isDBNOCarryFloat(isDBNOCarryFloat=0.5)
   n9 = Event BlueprintInitializeAnimation()
   Event OnCharacterCustomizationLoaded()
   n10 = CheckForBackbling()
   > Set isWearingBackpack(isWearingBackpack:bool=@n10) -> (Output_Get:bool)
   > Set isWearingCape(isWearingCape=OR Boolean(A=@n10.bIsBackblingACape, B=@n10.bIsBackblingLong))
  ]}
 OnNewFunction {nodes [
   OnNewFunction!() -> (then, Context:"Anim Update Context Structure (by ref)", Node:"Anim Node Reference Structure (by ref)") {is=title K2Node_FunctionEntry}
   > Return Node()
  ]}
]
variables [
 ADS Hair Velocity {type=Vector}
 Body_Mesh {type=SkeletalMeshComponent}
 Gravity Override Params {type=GravityOverrideParamsStruct}
 GravityOverride {type=Vector}
 isDBNOCarry {type=boolean}
 isDBNOCarryFloat {type=number}
 IsJumping {type=boolean}
 isWearingBackpack {type=boolean}
 isWearingCape {type=boolean, default=true}
]
