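# Physics asset for the M_MED_Imitator face accessory part, providing capsule bodies and joints for dynamic hair on the male medium character.
# Six four-bone back-hair chains hang from a main bone under a head-linked master, and two five-bone front chains hang from the head.
# Most hair joints share one angular limit, with tighter limits at the back main and upper D–F links.
# Torso, neck, head, clavicle and upper-arm capsules form a body proxy with free angular joints.
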PhysicsAsset /BRCosmetics/Characters/Player/Male/Medium/Bodies/M_MED_Imitator/Meshes/Parts/M_MED_Imitator_FE_FaceAcc_Physics
bodies {
 clavicle_l [{shapeType=Capsule, length=8.5121, radius=9.8926, shapeName=clavicle_l_capsule, center=<9.5266 0 0>, rotation=<rot 90 0 0>}]
 clavicle_r [
  {shapeType=Capsule, length=8.5419, radius=7, shapeName=clavicle_r_capsule, center=<-9.0713 0 0>, rotation=<rot 90 0 0>}
  {shapeType=Capsule, length=7.9905, radius=5.0095, shapeName=clavicle_r_capsule, center=<-8.5137 1.9364 -5.9716>, rotation=<rot 90 0 0>}
 ]
 dyn_backHair_A_01 [{shapeType=Capsule, length=3.6869, radius=3.7736, shapeName=dyn_backHair_A_01_capsule, center=<3.5 0 3>, rotation=<rot 90 0 0>}]
 dyn_backHair_A_02 [{shapeType=Capsule, length=3.6869, radius=3.7736, shapeName=dyn_backHair_A_02_capsule, center=<3.5 0 3>, rotation=<rot 90 0 0>}]
 dyn_backHair_A_03 [{shapeType=Capsule, length=3.6869, radius=3.7736, shapeName=dyn_backHair_A_03_capsule, center=<3.5 0 3>, rotation=<rot 90 0 0>}]
 dyn_backHair_A_04 [{shapeType=Capsule, length=5.648, radius=3.9454, shapeName=dyn_backHair_A_04_capsule, center=<5.2881 0 3>, rotation=<rot 90 0 0>}]
 dyn_backHair_B_01 [{shapeType=Capsule, length=3.6869, radius=3.7736, shapeName=dyn_backHair_B_01_capsule, center=<3.5 0 3>, rotation=<rot 90 0 0>}]
 dyn_backHair_B_02 [{shapeType=Capsule, length=3.6869, radius=3.7736, shapeName=dyn_backHair_B_02_capsule, center=<3.5 0 3>, rotation=<rot 90 0 0>}]
 dyn_backHair_B_03 [{shapeType=Capsule, length=3.6869, radius=3.7736, shapeName=dyn_backHair_B_03_capsule, center=<3.5 0 3>, rotation=<rot 90 0 0>}]
 dyn_backHair_B_04 [{shapeType=Capsule, length=3.6869, radius=3.7736, shapeName=dyn_backHair_B_04_capsule, center=<4.7868 0 3>, rotation=<rot 90 0 0>}]
 dyn_backHair_C_01 [{shapeType=Capsule, length=3.6869, radius=3.7736, shapeName=dyn_backHair_C_01_capsule, center=<3.5 0 3>, rotation=<rot 90 0 0>}]
 dyn_backHair_C_02 [{shapeType=Capsule, length=3.6869, radius=3.7736, shapeName=dyn_backHair_C_02_capsule, center=<3.5 0 3>, rotation=<rot 90 0 0>}]
 dyn_backHair_C_03 [{shapeType=Capsule, length=3.0399, radius=4.4207, shapeName=dyn_backHair_C_03_capsule, center=<5.549 0 3>, rotation=<rot 90 0 0>}]
 dyn_backHair_C_04 [{shapeType=Capsule, length=5.8034, radius=4.2521, shapeName=dyn_backHair_C_04_capsule, center=<6.2163 0 3>, rotation=<rot 81.8955 0 0>}]
 dyn_backHair_D_01 [{shapeType=Capsule, length=0.8422, radius=6.6183, shapeName=dyn_backHair_D_01_capsule, center=<3.5 0 3>, rotation=<rot 90 0 0>}]
 dyn_backHair_D_02 [{shapeType=Capsule, length=0.8422, radius=6.6183, shapeName=dyn_backHair_D_02_capsule, center=<3.5 0 3>, rotation=<rot 90 0 0>}]
 dyn_backHair_D_03 [{shapeType=Capsule, length=1.3419, radius=7.1351, shapeName=dyn_backHair_D_03_capsule, center=<5.6259 0 3>, rotation=<rot 90 0 0>}]
 dyn_backHair_D_04 [{shapeType=Capsule, length=3.4639, radius=7.3875, shapeName=dyn_backHair_D_04_capsule, center=<6.2447 0 3>, rotation=<rot 78.0632 0 0>}]
 dyn_backHair_E_01 [{shapeType=Capsule, length=4.1972, radius=3.2633, shapeName=dyn_backHair_D_01_capsule, center=<6.7364 0 3>, rotation=<rot 90 0 0>}]
 dyn_backHair_E_02 [{shapeType=Capsule, length=3.6869, radius=3.7736, shapeName=dyn_backHair_D_02_capsule, center=<3.5 0 3>, rotation=<rot 90 0 0>}]
 dyn_backHair_E_03 [{shapeType=Capsule, length=4.1866, radius=4.2904, shapeName=dyn_backHair_D_03_capsule, center=<5.6259 0 3>, rotation=<rot 90 0 0>}]
 dyn_backHair_E_04 [{shapeType=Capsule, length=6.3086, radius=4.5428, shapeName=dyn_backHair_D_04_capsule, center=<6.2447 0 3>, rotation=<rot 78.0632 0 0>}]
 dyn_backHair_F_01 [{
   shapeType=Capsule
   length=3.5305
   radius=4.7527
   shapeName=dyn_backHair_D_01_capsule
   center=<2.6963 0 2.5>
   rotation=<rot 80.0949 179.9999 179.9999>
  }]
 dyn_backHair_F_02 [{shapeType=Capsule, length=3.6869, radius=3.7736, shapeName=dyn_backHair_D_02_capsule, center=<3.5 0 3>, rotation=<rot 90 0 0>}]
 dyn_backHair_F_03 [{shapeType=Capsule, length=4.1866, radius=4.2904, shapeName=dyn_backHair_D_03_capsule, center=<5.6259 0 3>, rotation=<rot 90 0 0>}]
 dyn_backHair_F_04 [{shapeType=Capsule, length=6.3086, radius=4.5428, shapeName=dyn_backHair_D_04_capsule, center=<6.2447 0 3>, rotation=<rot 78.0632 0 0>}]
 dyn_backhair_MSTR [{shapeType=Capsule, length=0.2, radius=0.2, shapeName=dyn_backhair_MSTR_capsule}]
 dyn_backhair_Main [{shapeType=Capsule, length=7.1709, radius=6.039, shapeName=dyn_backhair_Main_capsule, center=<4.1502 0 3>, rotation=<rot 0 0 90>}]
 dyn_l_frontHair_01 [{shapeType=Capsule, length=4.7757, radius=2.6848, shapeName=dyn_l_frontHair_01_capsule, center=<5 0 2.5>, rotation=<rot 90 0 0>}]
 dyn_l_frontHair_02 [{shapeType=Capsule, length=4.4322, radius=3.0283, shapeName=dyn_l_frontHair_02_capsule, center=<3.5 0 3>, rotation=<rot 90 0 0>}]
 dyn_l_frontHair_03 [{shapeType=Capsule, length=3.3653, radius=4.0952, shapeName=dyn_l_frontHair_03_capsule, center=<3.5 0 3>, rotation=<rot 90 0 0>}]
 dyn_l_frontHair_04 [{shapeType=Capsule, length=4.7757, radius=2.6848, shapeName=dyn_l_frontHair_04_capsule, center=<3.5 0 3>, rotation=<rot 90 0 0>}]
 dyn_l_frontHair_05 [{shapeType=Capsule, length=4.7757, radius=2.6848, shapeName=dyn_l_frontHair_05_capsule, center=<3.5 0 3>, rotation=<rot 90 0 0>}]
 dyn_r_frontHair_01 [{shapeType=Capsule, length=4.4408, radius=2.6848, shapeName=dyn_r_frontHair_01_capsule, center=<3.5 0 3>, rotation=<rot 90 0 0>}]
 dyn_r_frontHair_02 [{shapeType=Capsule, length=2.0622, radius=5.3984, shapeName=dyn_r_frontHair_02_capsule, center=<3.5 0 3>, rotation=<rot 90 0 0>}]
 dyn_r_frontHair_03 [{shapeType=Capsule, length=2.0252, radius=5.4354, shapeName=dyn_r_frontHair_03_capsule, center=<3.5 0 4>, rotation=<rot 90 0 0>}]
 dyn_r_frontHair_04 [{shapeType=Capsule, length=5.7643, radius=1.6962, shapeName=dyn_r_frontHair_04_capsule, center=<3.5 0 4>, rotation=<rot 90 0 0>}]
 dyn_r_frontHair_05 [{shapeType=Capsule, length=5.7643, radius=1.6962, shapeName=dyn_r_frontHair_05_capsule, center=<3.5 0 1>, rotation=<rot 90 0 0>}]
 dyn_shield_MSTR [{shapeType=Capsule, length=0.2, radius=0.2, shapeName=dyn_shield_MSTR_capsule}]
 head [{shapeType=Capsule, length=5.6879, radius=7.6817, shapeName=head_capsule, rotation=<rot -90 0 0>}]
 neck_01 [{shapeType=Capsule, length=56.0047, radius=5.9953, shapeName=neck_01_capsule, center=<1.2016 -0.8407 0>}]
 neck_02 [{shapeType=Capsule, length=57.7917, radius=4.2083, shapeName=neck_01_capsule, center=<2.8325 -0.5621 0>}]
 pelvis [{shapeType=Capsule, length=11.8827, radius=13.5481, shapeName=pelvis_capsule}]
 spine_01 [{shapeType=Capsule, length=4.8006, radius=12.5978, shapeName=spine_01_capsule}]
 spine_02 [{shapeType=Capsule, length=4.8006, radius=12.5978, shapeName=spine_02_capsule}]
 spine_03 [{shapeType=Capsule, length=5.4936, radius=13.156, shapeName=spine_03_capsule}]
 spine_04 [{shapeType=Capsule, length=9.5785, radius=12.4913, shapeName=spine_04_capsule}]
 spine_05 [
  {shapeType=Capsule, length=21.3257, radius=11.0749, shapeName=spine_05_capsule, center=<0.3862 -1.3412 0>}
  {shapeType=Capsule, length=6.7923, radius=4, shapeName=spine_05_capsule, center=<12.6087 -1.3088 7.7433>}
 ]
 upperarm_l [{shapeType=Capsule, length=28.1916, radius=6.3667, shapeName=upperarm_l_capsule, center=<18.7505 0 0>, rotation=<rot 90 0 0>}]
 upperarm_r [{shapeType=Capsule, length=28.1916, radius=6.3667, shapeName=upperarm_l_capsule, center=<-20.3966 0 0>, rotation=<rot 90 0 0>}]
}
constraints [
 {bone1Name=dyn_backHair_A_01, bone2Name=dyn_backhair_Main, swing1LimitDegrees=50, swing2LimitDegrees=50, twistLimitDegrees=50}
 {bone1Name=dyn_backHair_A_02, bone2Name=dyn_backHair_A_01, swing1LimitDegrees=50, swing2LimitDegrees=50, twistLimitDegrees=50}
 {bone1Name=dyn_backHair_A_03, bone2Name=dyn_backHair_A_02, swing1LimitDegrees=50, swing2LimitDegrees=50, twistLimitDegrees=50}
 {bone1Name=dyn_backHair_A_04, bone2Name=dyn_backHair_A_03, swing1LimitDegrees=50, swing2LimitDegrees=50, twistLimitDegrees=50}
 {bone1Name=dyn_backHair_B_01, bone2Name=dyn_backhair_Main, swing1LimitDegrees=50, swing2LimitDegrees=50, twistLimitDegrees=50}
 {bone1Name=dyn_backHair_B_02, bone2Name=dyn_backHair_B_01, swing1LimitDegrees=50, swing2LimitDegrees=50, twistLimitDegrees=50}
 {bone1Name=dyn_backHair_B_03, bone2Name=dyn_backHair_B_02, swing1LimitDegrees=50, swing2LimitDegrees=50, twistLimitDegrees=50}
 {bone1Name=dyn_backHair_B_04, bone2Name=dyn_backHair_B_03, swing1LimitDegrees=50, swing2LimitDegrees=50, twistLimitDegrees=50}
 {bone1Name=dyn_backHair_C_01, bone2Name=dyn_backhair_Main, swing1LimitDegrees=50, swing2LimitDegrees=50, twistLimitDegrees=50}
 {bone1Name=dyn_backHair_C_02, bone2Name=dyn_backHair_C_01, swing1LimitDegrees=50, swing2LimitDegrees=50, twistLimitDegrees=50}
 {bone1Name=dyn_backHair_C_03, bone2Name=dyn_backHair_C_02, swing1LimitDegrees=50, swing2LimitDegrees=50, twistLimitDegrees=50}
 {bone1Name=dyn_backHair_C_04, bone2Name=dyn_backHair_C_03, swing1LimitDegrees=50, swing2LimitDegrees=50, twistLimitDegrees=50}
 {bone1Name=dyn_backHair_D_01, bone2Name=dyn_backhair_Main, swing1LimitDegrees=20, swing2LimitDegrees=20, twistLimitDegrees=20}
 {bone1Name=dyn_backHair_D_02, bone2Name=dyn_backHair_D_01, swing1LimitDegrees=20, swing2LimitDegrees=20, twistLimitDegrees=20}
 {bone1Name=dyn_backHair_D_03, bone2Name=dyn_backHair_D_02, swing1LimitDegrees=50, swing2LimitDegrees=50, twistLimitDegrees=50}
 {bone1Name=dyn_backHair_D_04, bone2Name=dyn_backHair_D_03, swing1LimitDegrees=50, swing2LimitDegrees=50, twistLimitDegrees=50}
 {bone1Name=dyn_l_frontHair_01, bone2Name=head, swing1LimitDegrees=50, swing2LimitDegrees=50, twistLimitDegrees=50}
 {bone1Name=dyn_l_frontHair_02, bone2Name=dyn_l_frontHair_01, swing1LimitDegrees=50, swing2LimitDegrees=50, twistLimitDegrees=50}
 {bone1Name=dyn_l_frontHair_03, bone2Name=dyn_l_frontHair_02, swing1LimitDegrees=50, swing2LimitDegrees=50, twistLimitDegrees=50}
 {bone1Name=dyn_l_frontHair_04, bone2Name=dyn_l_frontHair_03, swing1LimitDegrees=50, swing2LimitDegrees=50, twistLimitDegrees=50}
 {bone1Name=dyn_l_frontHair_05, bone2Name=dyn_l_frontHair_04, swing1LimitDegrees=50, swing2LimitDegrees=50, twistLimitDegrees=50}
 {bone1Name=dyn_r_frontHair_01, bone2Name=head, swing1LimitDegrees=50, swing2LimitDegrees=50, twistLimitDegrees=50}
 {bone1Name=dyn_r_frontHair_02, bone2Name=dyn_r_frontHair_01, swing1LimitDegrees=50, swing2LimitDegrees=50, twistLimitDegrees=50}
 {bone1Name=dyn_r_frontHair_03, bone2Name=dyn_r_frontHair_02, swing1LimitDegrees=50, swing2LimitDegrees=50, twistLimitDegrees=50}
 {bone1Name=dyn_r_frontHair_04, bone2Name=dyn_r_frontHair_03, swing1LimitDegrees=50, swing2LimitDegrees=50, twistLimitDegrees=50}
 {bone1Name=dyn_r_frontHair_05, bone2Name=dyn_r_frontHair_04, swing1LimitDegrees=50, swing2LimitDegrees=50, twistLimitDegrees=50}
 {
  bone1Name=head
  bone2Name=neck_02
  swing1LimitDegrees=45
  swing1Motion=Free
  swing2LimitDegrees=45
  swing2Motion=Free
  twistLimitDegrees=45
  twistMotion=Free
 }
 {
  bone1Name=spine_05
  bone2Name=spine_04
  swing1LimitDegrees=45
  swing1Motion=Free
  swing2LimitDegrees=45
  swing2Motion=Free
  twistLimitDegrees=45
  twistMotion=Free
 }
 {
  bone1Name=spine_04
  bone2Name=spine_03
  swing1LimitDegrees=45
  swing1Motion=Free
  swing2LimitDegrees=45
  swing2Motion=Free
  twistLimitDegrees=45
  twistMotion=Free
 }
 {
  bone1Name=spine_03
  bone2Name=spine_02
  swing1LimitDegrees=45
  swing1Motion=Free
  swing2LimitDegrees=45
  swing2Motion=Free
  twistLimitDegrees=45
  twistMotion=Free
 }
 {
  bone1Name=spine_02
  bone2Name=spine_01
  swing1LimitDegrees=45
  swing1Motion=Free
  swing2LimitDegrees=45
  swing2Motion=Free
  twistLimitDegrees=45
  twistMotion=Free
 }
 {
  bone1Name=spine_01
  bone2Name=pelvis
  swing1LimitDegrees=45
  swing1Motion=Free
  swing2LimitDegrees=45
  swing2Motion=Free
  twistLimitDegrees=45
  twistMotion=Free
 }
 {
  bone1Name=dyn_backhair_MSTR
  bone2Name=head
  swing1LimitDegrees=45
  swing1Motion=Free
  swing2LimitDegrees=45
  swing2Motion=Free
  twistLimitDegrees=45
  twistMotion=Free
 }
 {
  bone1Name=dyn_shield_MSTR
  bone2Name=head
  swing1LimitDegrees=45
  swing1Motion=Free
  swing2LimitDegrees=45
  swing2Motion=Free
  twistLimitDegrees=45
  twistMotion=Free
 }
 {
  bone1Name=upperarm_l
  bone2Name=clavicle_l
  swing1LimitDegrees=45
  swing1Motion=Free
  swing2LimitDegrees=45
  swing2Motion=Free
  twistLimitDegrees=45
  twistMotion=Free
 }
 {
  bone1Name=upperarm_r
  bone2Name=clavicle_r
  swing1LimitDegrees=45
  swing1Motion=Free
  swing2LimitDegrees=45
  swing2Motion=Free
  twistLimitDegrees=45
  twistMotion=Free
 }
 {
  bone1Name=clavicle_l
  bone2Name=spine_05
  swing1LimitDegrees=45
  swing1Motion=Free
  swing2LimitDegrees=45
  swing2Motion=Free
  twistLimitDegrees=45
  twistMotion=Free
 }
 {
  bone1Name=clavicle_r
  bone2Name=spine_05
  swing1LimitDegrees=45
  swing1Motion=Free
  swing2LimitDegrees=45
  swing2Motion=Free
  twistLimitDegrees=45
  twistMotion=Free
 }
 {
  bone1Name=neck_01
  bone2Name=spine_05
  swing1LimitDegrees=45
  swing1Motion=Free
  swing2LimitDegrees=45
  swing2Motion=Free
  twistLimitDegrees=45
  twistMotion=Free
 }
 {
  bone1Name=neck_02
  bone2Name=neck_01
  swing1LimitDegrees=45
  swing1Motion=Free
  swing2LimitDegrees=45
  swing2Motion=Free
  twistLimitDegrees=45
  twistMotion=Free
 }
 {bone1Name=dyn_backhair_Main, bone2Name=dyn_backhair_MSTR, swing1LimitDegrees=20, swing2LimitDegrees=20, twistLimitDegrees=20}
 {bone1Name=dyn_backHair_E_01, bone2Name=dyn_backhair_Main, swing1LimitDegrees=15, swing2LimitDegrees=15, twistLimitDegrees=15}
 {bone1Name=dyn_backHair_E_02, bone2Name=dyn_backHair_E_01, swing1LimitDegrees=15, swing2LimitDegrees=15, twistLimitDegrees=15}
 {bone1Name=dyn_backHair_E_03, bone2Name=dyn_backHair_E_02, swing1LimitDegrees=50, swing2LimitDegrees=50, twistLimitDegrees=50}
 {bone1Name=dyn_backHair_E_04, bone2Name=dyn_backHair_E_03, swing1LimitDegrees=50, swing2LimitDegrees=50, twistLimitDegrees=50}
 {bone1Name=dyn_backHair_F_01, bone2Name=dyn_backhair_Main, swing1LimitDegrees=15, swing2LimitDegrees=15, twistLimitDegrees=15}
 {bone1Name=dyn_backHair_F_02, bone2Name=dyn_backHair_F_01, swing1LimitDegrees=15, swing2LimitDegrees=15, twistLimitDegrees=15}
 {bone1Name=dyn_backHair_F_03, bone2Name=dyn_backHair_F_02, swing1LimitDegrees=50, swing2LimitDegrees=50, twistLimitDegrees=50}
 {bone1Name=dyn_backHair_F_04, bone2Name=dyn_backHair_F_03, swing1LimitDegrees=50, swing2LimitDegrees=50, twistLimitDegrees=50}
]
