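# Control Rig for the M_MED_Imitator face accessory that adjusts the back-hair root bone according to head pose.
# A spherical pose reader on the head weights a small rotation offset on dyn_backhair_Main, then a parent constraint of that bone to spine_05.
# The Emote bool, clamped and inverted, switches off only the constraint, without smoothing because AlphaInterp interpolation is disabled.
# The pose reader has debug drawing enabled.
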ControlRigBlueprint /BRCosmetics/Characters/Player/Male/Medium/Bodies/M_MED_Imitator/Meshes/Parts/M_MED_Imitator_FaceAcc_CtrlRig
graphs [
 RigVMFunctionLibrary {links [], nodes []}
 RigVMModel {nodes [
   RigUnit_BeginExecution = &RigUnit_BeginExecution()
   n1 = &RigUnit_SphericalPoseReader(
    ActiveRegionScaleFactors=(PositiveWidth=1,NegativeWidth=1,PositiveHeight=1,NegativeHeight=1)
    ActiveRegionSize=0.2
    Debug=(bDrawDebug=True,DebugScale=25,DebugSegments=20,DebugThickness=0.25)
    DriverAxis=(X=0,Y=1,Z=0)
    DriverItem=(Type=Bone,Name="head")
    FalloffRegionScaleFactors=(PositiveWidth=1,NegativeWidth=1,PositiveHeight=1,NegativeHeight=1)
    FalloffSize=0.5
    FlipWidthScaling=false
    OptionalParentItem=(Type=Bone)
    RotationOffset=(X=0,Y=-90,Z=0)
    ExecutePin=@RigUnit_BeginExecution.ExecutePin
   )
   > &RigUnit_ModifyTransforms(
    ItemToModify=((Item=(Type=Bone,Name="dyn_backhair_Main"),Transform=(Rotation=(X=0,Y=0,Z=0,W=1),Translation=(X=0,Y=0,Z=0),Scale3D=(X=1,Y=1,Z=1))))
    ItemToModify.0.Transform.Rotation=&RigVMFunction_MathQuaternionFromEuler(Euler=(X=0,Y=-5,Z=10)).Result
    Weight=@n1.OutputParam
   )
   RigVMFunction_MathBoolToFloat = &RigVMFunction_MathBoolToFloat(Value=RigVMVariableNode() {variable=Emote}.Value)
   AlphaInterp = &RigVMFunction_AlphaInterp(
    InterpSpeedDecreasing=10
    InterpSpeedIncreasing=10
    Scale=1
    bClampResult=true
    bInterpResult=False
    Value=@RigVMFunction_MathBoolToFloat.Result
   )
   Remap = &RigVMFunction_MathDoubleRemap(SourceMaximum=1, SourceMinimum=0, TargetMaximum=0, TargetMinimum=1, Value=@AlphaInterp.Result)
   Multiply = &RigVMFunction_MathDoubleMul(B=@Remap.Result, A=@n1.OutputParam)
   > &RigUnit_ParentConstraint(
    AdvancedSettings=(RotationOrderForFilter=XZY)
    Child=(Type=Bone,Name="dyn_backhair_Main")
    Parents=((Item=(Type=Bone,Name="spine_05"),Weight=1))
    Weight=@Multiply.Result
   )
  ]}
]
hierarchy [
 root {loc=<0 0 0>, rot=<0 0 0 1>}
 pelvis {parent=root, loc=<0 -0.7071 93.1191>, rot=<0.0127 0.707 -0.0127 -0.707>}
 spine_01 {parent=pelvis, loc=<6.7494 -0.6388 0>, rot=<0 0 0.034 -0.9994>}
 spine_02 {parent=spine_01, loc=<7.2518 -0.035 0>, rot=<0 0 0.0238 -0.9997>}
 spine_03 {parent=spine_02, loc=<6.0104 0 0>, rot=<0 0 0.0149 -0.9999>}
 spine_04 {parent=spine_03, loc=<8.1024 0 0>, rot=<0 0 -0.0132 -0.9999>}
 spine_05 {parent=spine_04, loc=<10.0146 0 0>, rot=<0 0 -0.0015 -1>}
 attach_backpack {parent=spine_05, loc=<-0.123 -13.473 0>, rot=<-0.0093 -0.707 0.0093 -0.707>}
 neck_01 {parent=spine_05, loc=<11.3138 0.1529 0>, rot=<0 0 -0.2003 -0.9797>}
 neck_02 {parent=neck_01, loc=<4.5627 0 0>, rot=<0 0 0.0504 -0.9987>}
 head {parent=neck_02, loc=<5.3914 0 0>, rot=<0 0 0.115 -0.9934>}
 dyn_l_frontHair_MSTR {parent=head, loc=<-1.696 -1.5329 -6.5651>, rot=<-0.0746 -0.7064 -0.6679 -0.2219>}
 dyn_l_frontHair_01 {parent=dyn_l_frontHair_MSTR, loc=<0 0 0>, rot=<0 0 0 1>}
 dyn_l_frontHair_02 {parent=dyn_l_frontHair_01, loc=<12.0527 0.5219 0.6457>, rot=<-0.0448 0.0849 -0.0627 -0.9934>}
 dyn_l_frontHair_03 {parent=dyn_l_frontHair_02, loc=<8.2515 0 0>, rot=<0 -0.1244 0 -0.9922>}
 dyn_l_frontHair_04 {parent=dyn_l_frontHair_03, loc=<9.5544 0 -0.1938>, rot=<0 -0.1036 0 -0.9946>}
 dyn_l_frontHair_05 {parent=dyn_l_frontHair_04, loc=<8.7438 0 0>, rot=<0 0 0 1>}
 dyn_r_frontHair_MSTR {parent=head, loc=<-1.7061 -1.5262 6.5565>, rot=<0.1849 0.6566 0.7215 0.1187>}
 dyn_r_frontHair_01 {parent=dyn_r_frontHair_MSTR, loc=<0 0 0>, rot=<0 0 0 1>}
 dyn_r_frontHair_02 {parent=dyn_r_frontHair_01, loc=<12.0527 0.5219 0.6457>, rot=<-0.0508 0.0815 0.0087 -0.9953>}
 dyn_r_frontHair_03 {parent=dyn_r_frontHair_02, loc=<8.2515 0 0>, rot=<0.0026 -0.1244 0.0207 -0.992>}
 dyn_r_frontHair_04 {parent=dyn_r_frontHair_03, loc=<9.5544 0 -0.1938>, rot=<0 -0.1036 0 -0.9946>}
 dyn_r_frontHair_05 {parent=dyn_r_frontHair_04, loc=<8.7438 0 0>, rot=<0 0 0 1>}
 dyn_backhair_MSTR {parent=head, loc=<-0.3326 -8.5138 0.0004>, rot=<0.2603 -0.6575 0.6595 -0.255>}
 dyn_backhair_Main {parent=dyn_backhair_MSTR, loc=<0 0 0>, rot=<0 0 0 1>}
 dyn_backHair_A_MSTR {parent=dyn_backhair_Main, loc=<3.9109 6.6833 -2.212>, rot=<0.2232 -0.0946 -0.2719 -0.9313>}
 dyn_backHair_A_01 {parent=dyn_backHair_A_MSTR, loc=<0 0 0>, rot=<0 0 0 1>}
 dyn_backHair_A_02 {parent=dyn_backHair_A_01, loc=<8.2515 0 0>, rot=<0 -0.1014 0 -0.9948>}
 dyn_backHair_A_03 {parent=dyn_backHair_A_02, loc=<8.6344 0 0>, rot=<0 -0.1036 0 -0.9946>}
 dyn_backHair_A_04 {parent=dyn_backHair_A_03, loc=<7.8212 0 0>, rot=<0 0 0 1>}
 dyn_backHair_C_MSTR {parent=dyn_backhair_Main, loc=<6.7233 2.4808 -0.2055>, rot=<0.0667 -0.1479 -0.0558 -0.9852>}
 dyn_backHair_C_01 {parent=dyn_backHair_C_MSTR, loc=<0 0 0>, rot=<0 0 0 1>}
 dyn_backHair_C_02 {parent=dyn_backHair_C_01, loc=<9.0275 0 -0.1598>, rot=<0 -0.154 0 -0.9881>}
 dyn_backHair_C_03 {parent=dyn_backHair_C_02, loc=<8.6344 0 0>, rot=<0 -0.0757 0 -0.9971>}
 dyn_backHair_C_04 {parent=dyn_backHair_C_03, loc=<11.4398 0 0>, rot=<0 0 0 1>}
 dyn_backHair_D_MSTR {parent=dyn_backhair_Main, loc=<7.0007 -2.5315 -0.4233>, rot=<-0.0134 -0.1469 0.0355 -0.9884>}
 dyn_backHair_D_01 {parent=dyn_backHair_D_MSTR, loc=<0 0 0>, rot=<0 0 0 1>}
 dyn_backHair_D_02 {parent=dyn_backHair_D_01, loc=<9.0275 0 -0.1598>, rot=<0 -0.154 0 -0.9881>}
 dyn_backHair_D_03 {parent=dyn_backHair_D_02, loc=<8.6344 0 0>, rot=<0 -0.0756 -0.0148 -0.997>}
 dyn_backHair_D_04 {parent=dyn_backHair_D_03, loc=<11.2548 0 0>, rot=<0 0 0 1>}
 dyn_backHair_F_MSTR {parent=dyn_backhair_Main, loc=<3.3934 -6.391 -2.3048>, rot=<-0.1094 -0.1068 0.2367 -0.9595>}
 dyn_backHair_F_01 {parent=dyn_backHair_F_MSTR, loc=<0 0 0>, rot=<0 0 0 1>}
 dyn_backHair_F_02 {parent=dyn_backHair_F_01, loc=<9.0275 0 -0.1598>, rot=<0 -0.1318 0 -0.9913>}
 dyn_backHair_F_03 {parent=dyn_backHair_F_02, loc=<8.6344 0 0>, rot=<0 -0.0757 0 -0.9971>}
 dyn_backHair_F_04 {parent=dyn_backHair_F_03, loc=<7.8212 0 0>, rot=<0 0 0 1>}
 dyn_backHair_E_MSTR {parent=dyn_backhair_Main, loc=<4.6554 -4.434 -3.7887>, rot=<-0.0449 -0.1223 0.1307 -0.9828>}
 dyn_backHair_E_01 {parent=dyn_backHair_E_MSTR, loc=<0 0 0>, rot=<0 0 0 1>}
 dyn_backHair_E_02 {parent=dyn_backHair_E_01, loc=<9.0275 0 -0.1598>, rot=<0 -0.1318 0 -0.9913>}
 dyn_backHair_E_03 {parent=dyn_backHair_E_02, loc=<8.6344 0 0>, rot=<0 -0.0757 0 -0.9971>}
 dyn_backHair_E_04 {parent=dyn_backHair_E_03, loc=<7.8212 0 0>, rot=<0 0 0 1>}
 dyn_backHair_B_MSTR {parent=dyn_backhair_Main, loc=<4.7932 4.4791 -3.6909>, rot=<0.085 -0.1312 -0.1583 -0.9749>}
 dyn_backHair_B_01 {parent=dyn_backHair_B_MSTR, loc=<0 0 0>, rot=<0 0 0 1>}
 dyn_backHair_B_02 {parent=dyn_backHair_B_01, loc=<8.6053 0 -0.0729>, rot=<0 -0.1014 0 -0.9948>}
 dyn_backHair_B_03 {parent=dyn_backHair_B_02, loc=<8.6344 0 0>, rot=<0 -0.1347 0 -0.9909>}
 dyn_backHair_B_04 {parent=dyn_backHair_B_03, loc=<7.8212 0 0>, rot=<0 0 0 1>}
 dyn_shield_MSTR {parent=head, loc=<6.2371 6.0922 -0.3393>, rot=<0.426 0.5599 0.5639 0.4324>}
 dyn_shield {parent=dyn_shield_MSTR, loc=<0 0 0>, rot=<0 0 0 1>}
 clavicle_l {parent=spine_05, loc=<4.3575 -0.1466 -2.3223>, rot=<0.0465 -0.7289 -0.009 -0.6829>}
 upperarm_l {parent=clavicle_l, loc=<15.5111 0 0>, rot=<0.0226 -0.3234 -0.0088 -0.9459>}
 lowerarm_l {parent=upperarm_l, loc=<26.6514 0 0>, rot=<0.0391 -0.0099 -0.2453 -0.9686>}
 hand_l {parent=lowerarm_l, loc=<22.3491 0 0>, rot=<-0.7081 -0.0507 0.0975 -0.6975>}
 thumb_01_l {parent=hand_l, loc=<1.3118 -0.3191 -2.459>, rot=<0.6546 -0.4042 0.0664 -0.6354>}
 thumb_02_l {parent=thumb_01_l, loc=<5.9336 -0.0448 0.0401>, rot=<-0.0037 0.0213 0.298 -0.9543>}
 thumb_03_l {parent=thumb_02_l, loc=<2.2223 0.3682 -0.0907>, rot=<0.0118 -0.013 0.0066 -0.9998>}
 index_metacarpal_l {parent=hand_l, loc=<1.6535 0.0183 -2.96>, rot=<0.0158 -0.0124 -0.0173 -0.9996>}
 index_01_l {parent=index_metacarpal_l, loc=<8.3008 0.0357 -0.0524>, rot=<-0.0044 0.0485 0.0924 -0.9945>}
 index_02_l {parent=index_01_l, loc=<3.3587 -0.3004 0.0452>, rot=<-0.002 0.0027 0.3225 -0.9466>}
 index_03_l {parent=index_02_l, loc=<2.5578 -0.123 -0.0252>, rot=<-0.002 0.0067 0.2558 -0.9667>}
 middle_metacarpal_l {parent=hand_l, loc=<2.1856 -0.1725 -0.5984>, rot=<0.006 0.0168 -0.0329 -0.9993>}
 middle_01_l {parent=middle_metacarpal_l, loc=<7.3275 -0.2447 0.0129>, rot=<-0.0208 0.0042 0.0858 -0.9961>}
 middle_02_l {parent=middle_01_l, loc=<4.0123 0.1401 0.1175>, rot=<-0.0218 0.0258 0.3512 -0.9357>}
 middle_03_l {parent=middle_02_l, loc=<2.6077 0.0534 0.1373>, rot=<0.0082 0.0188 0.1708 -0.9851>}
 ring_metacarpal_l {parent=hand_l, loc=<2.2074 -0.3625 1.3638>, rot=<-0.0284 0.0722 -0.0458 -0.9959>}
 ring_01_l {parent=ring_metacarpal_l, loc=<7.0295 -0.6197 -0.1446>, rot=<-0.0489 -0.0252 0.1892 -0.9804>}
 ring_02_l {parent=ring_01_l, loc=<3.6174 0.4647 -0.1027>, rot=<0 0 0.2806 -0.9598>}
 ring_03_l {parent=ring_02_l, loc=<2.4152 0.114 0.0199>, rot=<0 0 0.1189 -0.9929>}
 pinky_metacarpal_l {parent=hand_l, loc=<1.7616 -0.9753 3.3814>, rot=<-0.0397 0.0814 -0.069 -0.9935>}
 pinky_01_l {parent=pinky_metacarpal_l, loc=<7.1709 -0.4831 -0.1094>, rot=<-0.0654 -0.0035 0.2242 -0.9723>}
 pinky_02_l {parent=pinky_01_l, loc=<2.8274 0.0893 -0.1043>, rot=<0.0037 0.0011 0.2854 -0.9584>}
 pinky_03_l {parent=pinky_02_l, loc=<2.2046 0.0307 -0.0078>, rot=<-0.0002 0.0098 0.1445 -0.9895>}
 lowerarm_twist_02_l {parent=lowerarm_l, loc=<9.6199 0 0>, rot=<0 0 0 -1>}
 lowerarm_twist_01_l {parent=lowerarm_l, loc=<17.8549 0 0>, rot=<0 0 0 -1>}
 upperarm_twist_01_l {parent=upperarm_l, loc=<0.0041 0 0>, rot=<0 0 0 1>}
 upperarm_twist_02_l {parent=upperarm_l, loc=<13.7941 0 0>, rot=<0 0 0 1>}
 attach_cape {parent=spine_05, loc=<9.017 -11.714 0>, rot=<0.0616 -0.7044 -0.0616 -0.7044>}
 clavicle_r {parent=spine_05, loc=<4.3575 -0.1466 2.3223>, rot=<0.7289 0.0465 -0.6829 0.009>}
 upperarm_r {parent=clavicle_r, loc=<-15.5111 0 0>, rot=<0.0226 -0.3234 -0.0088 -0.9459>}
 lowerarm_r {parent=upperarm_r, loc=<-26.6514 0 0>, rot=<0.0391 -0.0099 -0.2453 -0.9686>}
 hand_r {parent=lowerarm_r, loc=<-22.3491 0 0>, rot=<-0.7081 -0.0507 0.0975 -0.6975>}
 thumb_01_r {parent=hand_r, loc=<-1.3118 0.3191 2.459>, rot=<0.6546 -0.4042 0.0664 -0.6354>}
 thumb_02_r {parent=thumb_01_r, loc=<-5.9336 0.0448 -0.0401>, rot=<-0.0037 0.0213 0.298 -0.9543>}
 thumb_03_r {parent=thumb_02_r, loc=<-2.2223 -0.3682 0.0907>, rot=<0.0118 -0.013 0.0066 -0.9998>}
 index_metacarpal_r {parent=hand_r, loc=<-1.6535 -0.0183 2.96>, rot=<0.0158 -0.0124 -0.0173 -0.9996>}
 index_01_r {parent=index_metacarpal_r, loc=<-8.3008 -0.0357 0.0524>, rot=<-0.0044 0.0485 0.0924 -0.9945>}
 index_02_r {parent=index_01_r, loc=<-3.3587 0.3004 -0.0452>, rot=<-0.002 0.0027 0.3225 -0.9466>}
 index_03_r {parent=index_02_r, loc=<-2.5578 0.123 0.0252>, rot=<-0.002 0.0067 0.2558 -0.9667>}
 middle_metacarpal_r {parent=hand_r, loc=<-2.1856 0.1725 0.5984>, rot=<0.006 0.0168 -0.0329 -0.9993>}
 middle_01_r {parent=middle_metacarpal_r, loc=<-7.3275 0.2447 -0.0129>, rot=<-0.0208 0.0042 0.0858 -0.9961>}
 middle_02_r {parent=middle_01_r, loc=<-4.0123 -0.1401 -0.1175>, rot=<-0.0218 0.0258 0.3512 -0.9357>}
 middle_03_r {parent=middle_02_r, loc=<-2.6077 -0.0534 -0.1373>, rot=<0.0082 0.0188 0.1708 -0.9851>}
 ring_metacarpal_r {parent=hand_r, loc=<-2.2074 0.3625 -1.3638>, rot=<-0.0284 0.0722 -0.0458 -0.9959>}
 ring_01_r {parent=ring_metacarpal_r, loc=<-7.0295 0.6197 0.1446>, rot=<-0.0489 -0.0252 0.1892 -0.9804>}
 ring_02_r {parent=ring_01_r, loc=<-3.6174 -0.4647 0.1027>, rot=<0 0 0.2806 -0.9598>}
 ring_03_r {parent=ring_02_r, loc=<-2.4152 -0.114 -0.0199>, rot=<0 0 0.1189 -0.9929>}
 pinky_metacarpal_r {parent=hand_r, loc=<-1.7616 0.9753 -3.3814>, rot=<-0.0397 0.0814 -0.069 -0.9935>}
 pinky_01_r {parent=pinky_metacarpal_r, loc=<-7.1709 0.4831 0.1094>, rot=<-0.0654 -0.0035 0.2242 -0.9723>}
 pinky_02_r {parent=pinky_01_r, loc=<-2.8274 -0.0893 0.1043>, rot=<0.0037 0.0011 0.2854 -0.9584>}
 pinky_03_r {parent=pinky_02_r, loc=<-2.2046 -0.0307 0.0078>, rot=<-0.0002 0.0098 0.1445 -0.9895>}
 lowerarm_twist_02_r {parent=lowerarm_r, loc=<-9.6199 0 0>, rot=<0 0 0 -1>}
 lowerarm_twist_01_r {parent=lowerarm_r, loc=<-17.8549 0 0>, rot=<0 0 0 -1>}
 upperarm_twist_01_r {parent=upperarm_r, loc=<-0.0041 0 0>, rot=<0 0 0 -1>}
 upperarm_twist_02_r {parent=upperarm_r, loc=<-13.7941 0 0>, rot=<0 0 0 -1>}
 thigh_l {parent=pelvis, loc=<-2.5693 -0.0109 -7.7501>, rot=<0.0012 0.0572 0.0155 -0.9982>}
 calf_l {parent=thigh_l, loc=<-39.0484 0 0>, rot=<0.0006 -0.0161 -0.0391 -0.9991>}
 foot_l {parent=calf_l, loc=<-42.561 0.1077 0.4292>, rot=<0.0418 -0.0455 0.0343 -0.9975>}
 ball_l {parent=foot_l, loc=<-6.1599 13.8264 0.6806>, rot=<0.0001 -0.0001 -0.7063 -0.7079>}
 calf_twist_02_l {parent=calf_l, loc=<-18.8064 -0.0967 0.0397>, rot=<0.0237 0.0095 -0.0035 -0.9997>}
 calf_twist_01_l {parent=calf_l, loc=<-35.813 -0.0967 0.0397>, rot=<0.0237 0.0095 -0.0035 -0.9997>}
 thigh_twist_01_l {parent=thigh_l, loc=<-21.6303 0 0>, rot=<0.0013 0 0 -1>}
 thigh_r {parent=pelvis, loc=<-2.5693 -0.0109 7.7501>, rot=<0.0572 -0.0012 0.9982 0.0155>}
 calf_r {parent=thigh_r, loc=<39.0484 0 0>, rot=<0.0006 -0.0161 -0.0391 -0.9991>}
 foot_r {parent=calf_r, loc=<42.561 -0.1077 -0.4292>, rot=<0.0418 -0.0455 0.0343 -0.9975>}
 ball_r {parent=foot_r, loc=<6.1599 -13.8264 -0.6806>, rot=<0.0001 -0.0001 -0.7063 -0.7079>}
 calf_twist_02_r {parent=calf_r, loc=<18.8064 0.0967 -0.0397>, rot=<0.0237 0.0095 -0.0035 -0.9997>}
 calf_twist_01_r {parent=calf_r, loc=<35.813 0.0967 -0.0397>, rot=<0.0237 0.0095 -0.0035 -0.9997>}
 thigh_twist_01_r {parent=thigh_r, loc=<21.6303 0 0>, rot=<0.0013 0 0 -1>}
 ik_hand_root {parent=root, loc=<0 0 0>, rot=<0 0 0 1>}
 ik_hand_gun {parent=ik_hand_root, loc=<-51.9727 4.0883 102.6935>, rot=<0.6084 0.1793 0.4613 -0.6205>}
]
variables [Emote {type=bool, default=False}]
