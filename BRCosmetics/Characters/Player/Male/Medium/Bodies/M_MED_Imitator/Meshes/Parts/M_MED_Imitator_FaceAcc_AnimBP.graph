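# Animation Blueprint for the M_MED_Imitator face accessory skeleton that drives the character's dynamic back and front hair bones.
# From cached PreSim pose, weighted multi-bone constraints attach hair bones to spine, clavicles and nearby body bones, mimicking skinning.
# Additive rotations then offset back-hair master bones before Trail controllers add lagging motion with planar limits.
# Two separate single-target spine_05 constraint chains are left unconnected.
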AnimBlueprint /BRCosmetics/Characters/Player/Male/Medium/Bodies/M_MED_Imitator/Meshes/Parts/M_MED_Imitator_FaceAcc_AnimBP
target_skeleton=&/M_MED_Imitator_FaceAcc_Skeleton
graphs [
 AnimGraph {nodes [
   n1 = "Constraint - dyn_backHair_A_03 (P)"!(ComponentPose, bAlphaBoolEnabled:bool=True, ConstraintWeights_0=1) -> (Pose) {
    alphaInputType=Bool
    alphaBoolBlend {blendInTime=0.1, blendOption=Linear, blendOutTime=0.1}
    boneToModify {boneName=dyn_backHair_A_03}
    constraintSetup [{targetBone {boneName=spine_05}}]
    constraintWeights [1]
   }
   n2 = "Constraint - dyn_backHair_B_03 (P)"!(ComponentPose=@n1, bAlphaBoolEnabled:bool=True, ConstraintWeights_0=1) -> (Pose) {
    alphaInputType=Bool
    alphaBoolBlend {blendInTime=0.1, blendOption=Linear, blendOutTime=0.1}
    boneToModify {boneName=dyn_backHair_B_03}
    constraintSetup [{targetBone {boneName=spine_05}}]
    constraintWeights [1]
   }
   n3 = "Constraint - dyn_backHair_C_03 (P)"!(ComponentPose=@n2, bAlphaBoolEnabled:bool=True, ConstraintWeights_0=1) -> (Pose) {
    alphaInputType=Bool
    alphaBoolBlend {blendInTime=0.1, blendOption=Linear, blendOutTime=0.1}
    boneToModify {boneName=dyn_backHair_C_03}
    constraintSetup [{targetBone {boneName=spine_05}}]
    constraintWeights [1]
   }
   n4 = "Constraint - dyn_backHair_D_03 (P)"!(ComponentPose=@n3, bAlphaBoolEnabled:bool=True, ConstraintWeights_0=1) -> (Pose) {
    alphaInputType=Bool
    alphaBoolBlend {blendInTime=0.1, blendOption=Linear, blendOutTime=0.1}
    boneToModify {boneName=dyn_backHair_D_03}
    constraintSetup [{targetBone {boneName=spine_05}}]
    constraintWeights [1]
   }
   n5 = "Constraint - dyn_backHair_E_03 (P)"!(ComponentPose=@n4, bAlphaBoolEnabled:bool=True, ConstraintWeights_0=1) -> (Pose) {
    alphaInputType=Bool
    alphaBoolBlend {blendInTime=0.1, blendOption=Linear, blendOutTime=0.1}
    boneToModify {boneName=dyn_backHair_E_03}
    constraintSetup [{targetBone {boneName=spine_05}}]
    constraintWeights [1]
   }
   "Constraint - dyn_backHair_F_03 (P)"!(ComponentPose=@n5, bAlphaBoolEnabled:bool=True, ConstraintWeights_0=1) -> (Pose) {
    alphaInputType=Bool
    alphaBoolBlend {blendInTime=0.1, blendOption=Linear, blendOutTime=0.1}
    boneToModify {boneName=dyn_backHair_F_03}
    constraintSetup [{targetBone {boneName=spine_05}}]
    constraintWeights [1]
   }
   n6 = "Constraint - dyn_backHair_A_04 (P)"!(ComponentPose, bAlphaBoolEnabled:bool=True, ConstraintWeights_0=1) -> (Pose) {
    alphaInputType=Bool
    alphaBoolBlend {blendInTime=0.1, blendOption=Linear, blendOutTime=0.1}
    boneToModify {boneName=dyn_backHair_A_04}
    constraintSetup [{targetBone {boneName=spine_05}}]
    constraintWeights [1]
   }
   n7 = "Constraint - dyn_backHair_B_04 (P)"!(ComponentPose=@n6, bAlphaBoolEnabled:bool=True, ConstraintWeights_0=1) -> (Pose) {
    alphaInputType=Bool
    alphaBoolBlend {blendInTime=0.1, blendOption=Linear, blendOutTime=0.1}
    boneToModify {boneName=dyn_backHair_B_04}
    constraintSetup [{targetBone {boneName=spine_05}}]
    constraintWeights [1]
   }
   n8 = "Constraint - dyn_backHair_C_04 (P)"!(ComponentPose=@n7, bAlphaBoolEnabled:bool=True, ConstraintWeights_0=1) -> (Pose) {
    alphaInputType=Bool
    alphaBoolBlend {blendInTime=0.1, blendOption=Linear, blendOutTime=0.1}
    boneToModify {boneName=dyn_backHair_C_04}
    constraintSetup [{targetBone {boneName=spine_05}}]
    constraintWeights [1]
   }
   n9 = "Constraint - dyn_backHair_D_04 (P)"!(ComponentPose=@n8, bAlphaBoolEnabled:bool=True, ConstraintWeights_0=1) -> (Pose) {
    alphaInputType=Bool
    alphaBoolBlend {blendInTime=0.1, blendOption=Linear, blendOutTime=0.1}
    boneToModify {boneName=dyn_backHair_D_04}
    constraintSetup [{targetBone {boneName=spine_05}}]
    constraintWeights [1]
   }
   n10 = "Constraint - dyn_backHair_E_04 (P)"!(ComponentPose=@n9, bAlphaBoolEnabled:bool=True, ConstraintWeights_0=1) -> (Pose) {
    alphaInputType=Bool
    alphaBoolBlend {blendInTime=0.1, blendOption=Linear, blendOutTime=0.1}
    boneToModify {boneName=dyn_backHair_E_04}
    constraintSetup [{targetBone {boneName=spine_05}}]
    constraintWeights [1]
   }
   "Constraint - dyn_backHair_F_04 (P)"!(ComponentPose=@n10, bAlphaBoolEnabled:bool=True, ConstraintWeights_0=1) -> (Pose) {
    alphaInputType=Bool
    alphaBoolBlend {blendInTime=0.1, blendOption=Linear, blendOutTime=0.1}
    boneToModify {boneName=dyn_backHair_F_04}
    constraintSetup [{targetBone {boneName=spine_05}}]
    constraintWeights [1]
   }
   n11 = "Constraint - dyn_backHair_A_MSTR (PPPPPP)"!(
    ComponentPose=Local To Component(LocalPose=Use cached pose 'PreSim'())
    Alpha=1
    ConstraintWeights_0=0.6012
    ConstraintWeights_1:float=0.3121
    ConstraintWeights_2:float=0.0633
    ConstraintWeights_3:float=0.02
    ConstraintWeights_4:float=0.0021
    ConstraintWeights_5:float=0.0013
   ) -> (Pose) {
    alphaBoolBlend {blendOption=Linear}
    boneToModify {boneName=dyn_backHair_A_MSTR}
    constraintSetup [
     {targetBone {boneName=spine_05}}
     {targetBone {boneName=clavicle_l}}
     {targetBone {boneName=spine_03}}
     {targetBone {boneName=spine_04}}
     {targetBone {boneName=clavicle_r}}
     {targetBone {boneName=deform_pec_l}}
    ]
    constraintWeights [1,1,1,1,1,1]
   }
   n12 = "Constraint - dyn_backHair_A_01 (PPPPPP)"!(
    ComponentPose=@n11
    Alpha=1
    ConstraintWeights_0=0.6012
    ConstraintWeights_1:float=0.3121
    ConstraintWeights_2:float=0.0633
    ConstraintWeights_3:float=0.02
    ConstraintWeights_4:float=0.0021
    ConstraintWeights_5:float=0.0013
   ) -> (Pose) {
    alphaBoolBlend {blendOption=Linear}
    boneToModify {boneName=dyn_backHair_A_01}
    constraintSetup [
     {targetBone {boneName=spine_05}}
     {targetBone {boneName=clavicle_l}}
     {targetBone {boneName=spine_03}}
     {targetBone {boneName=spine_04}}
     {targetBone {boneName=clavicle_r}}
     {targetBone {boneName=deform_pec_l}}
    ]
    constraintWeights [1,1,1,1,1,1]
   }
   n13 = "Constraint - dyn_backHair_A_02 (PPPPP)"!(
    ComponentPose=@n12
    Alpha=1
    ConstraintWeights_0=0.4327
    ConstraintWeights_1:float=0.4106
    ConstraintWeights_2:float=0.1208
    ConstraintWeights_3:float=0.036
    ConstraintWeights_4:float=0
   ) -> (Pose) {
    alphaBoolBlend {blendOption=Linear}
    boneToModify {boneName=dyn_backHair_A_02}
    constraintSetup [
     {targetBone {boneName=spine_05}}
     {targetBone {boneName=clavicle_l}}
     {targetBone {boneName=spine_03}}
     {targetBone {boneName=spine_04}}
     {targetBone {boneName=upperarm_l}}
    ]
    constraintWeights [1,1,1,1,1]
   }
   n14 = "Constraint - dyn_backHair_A_03 (PPPPP)"!(
    ComponentPose=@n13
    Alpha=1
    ConstraintWeights_0=0.5622
    ConstraintWeights_1:float=0.245
    ConstraintWeights_2:float=0.108
    ConstraintWeights_3:float=0.0808
    ConstraintWeights_4:float=0.004
   ) -> (Pose) {
    alphaBoolBlend {blendOption=Linear}
    boneToModify {boneName=dyn_backHair_A_03}
    constraintSetup [
     {targetBone {boneName=clavicle_l}}
     {targetBone {boneName=spine_03}}
     {targetBone {boneName=spine_04}}
     {targetBone {boneName=spine_05}}
     {targetBone {boneName=spine_02}}
    ]
    constraintWeights [1,1,1,1,1]
   }
   n15 = "Constraint - dyn_backHair_A_04 (PPPPPP)"!(
    ComponentPose=@n14
    Alpha=1
    ConstraintWeights_0=0.3884
    ConstraintWeights_1:float=0.2915
    ConstraintWeights_2:float=0.221
    ConstraintWeights_3:float=0.0651
    ConstraintWeights_4:float=0.025
    ConstraintWeights_5:float=0.009
   ) -> (Pose) {
    alphaBoolBlend {blendOption=Linear}
    boneToModify {boneName=dyn_backHair_A_04}
    constraintSetup [
     {targetBone {boneName=clavicle_l}}
     {targetBone {boneName=spine_03}}
     {targetBone {boneName=spine_04}}
     {targetBone {boneName=spine_02}}
     {targetBone {boneName=spine_05}}
     {targetBone {boneName=spine_01}}
    ]
    constraintWeights [1,1,1,1,1,1]
   }
   n16 = "Constraint - dyn_backHair_C_MSTR (PPPPP)"!(
    ComponentPose=@n15
    Alpha=1
    ConstraintWeights_0=1
    ConstraintWeights_1:float=0
    ConstraintWeights_2:float=0
    ConstraintWeights_3:float=0
    ConstraintWeights_4:float=0
   ) -> (Pose) {
    alphaBoolBlend {blendOption=Linear}
    boneToModify {boneName=dyn_backHair_C_MSTR}
    constraintSetup [
     {targetBone {boneName=spine_05}}
     {targetBone {boneName=clavicle_l}}
     {targetBone {boneName=spine_03}}
     {targetBone {boneName=spine_04}}
     {targetBone {boneName=clavicle_r}}
    ]
    constraintWeights [1,1,1,1,1]
   }
   n17 = "Constraint - dyn_backHair_C_01 (PPPPP)"!(
    ComponentPose=@n16
    Alpha=1
    ConstraintWeights_0=1
    ConstraintWeights_1:float=0
    ConstraintWeights_2:float=0
    ConstraintWeights_3:float=0
    ConstraintWeights_4:float=0
   ) -> (Pose) {
    alphaBoolBlend {blendOption=Linear}
    boneToModify {boneName=dyn_backHair_C_01}
    constraintSetup [
     {targetBone {boneName=spine_05}}
     {targetBone {boneName=clavicle_l}}
     {targetBone {boneName=spine_03}}
     {targetBone {boneName=spine_04}}
     {targetBone {boneName=clavicle_r}}
    ]
    constraintWeights [1,1,1,1,1]
   }
   n18 = "Constraint - dyn_backHair_C_02 (PPPPP)"!(
    ComponentPose=@n17
    Alpha=1
    ConstraintWeights_0=0.4493
    ConstraintWeights_1:float=0.2778
    ConstraintWeights_2:float=0.1875
    ConstraintWeights_3:float=0.0821
    ConstraintWeights_4:float=0.0034
   ) -> (Pose) {
    alphaBoolBlend {blendOption=Linear}
    boneToModify {boneName=dyn_backHair_C_02}
    constraintSetup [
     {targetBone {boneName=spine_05}}
     {targetBone {boneName=spine_03}}
     {targetBone {boneName=clavicle_l}}
     {targetBone {boneName=spine_04}}
     {targetBone {boneName=clavicle_r}}
    ]
    constraintWeights [1,1,1,1,1]
   }
   n19 = "Constraint - dyn_backHair_C_03 (PPPPP)"!(
    ComponentPose=@n18
    Alpha=1
    ConstraintWeights_0=0.3753
    ConstraintWeights_1:float=0.2285
    ConstraintWeights_2:float=0.2033
    ConstraintWeights_3:float=0.1677
    ConstraintWeights_4:float=0.0252
   ) -> (Pose) {
    alphaBoolBlend {blendOption=Linear}
    boneToModify {boneName=dyn_backHair_C_03}
    constraintSetup [
     {targetBone {boneName=spine_03}}
     {targetBone {boneName=spine_05}}
     {targetBone {boneName=spine_04}}
     {targetBone {boneName=clavicle_l}}
     {targetBone {boneName=spine_02}}
    ]
    constraintWeights [1,1,1,1,1]
   }
   n20 = "Constraint - dyn_backHair_C_04 (PPPPPPP)"!(
    ComponentPose=@n19
    Alpha=1
    ConstraintWeights_0=0.3989
    ConstraintWeights_1:float=0.2283
    ConstraintWeights_2:float=0.141
    ConstraintWeights_3:float=0.1057
    ConstraintWeights_4:float=0.058
    ConstraintWeights_5:float=0.048
    ConstraintWeights_6:float=0.0201
   ) -> (Pose) {
    alphaBoolBlend {blendOption=Linear}
    boneToModify {boneName=dyn_backHair_C_04}
    constraintSetup [
     {targetBone {boneName=spine_03}}
     {targetBone {boneName=spine_04}}
     {targetBone {boneName=spine_02}}
     {targetBone {boneName=clavicle_l}}
     {targetBone {boneName=spine_01}}
     {targetBone {boneName=spine_05}}
     {targetBone {boneName=pelvis}}
    ]
    constraintWeights [1,1,1,1,1,1,1]
   }
   n21 = "Constraint - dyn_backHair_D_MSTR (PPPPP)"!(
    ComponentPose=@n20
    Alpha=1
    ConstraintWeights_0=0.6044
    ConstraintWeights_1:float=0.2487
    ConstraintWeights_2:float=0.1127
    ConstraintWeights_3:float=0.0268
    ConstraintWeights_4:float=0.0073
   ) -> (Pose) {
    alphaBoolBlend {blendOption=Linear}
    boneToModify {boneName=dyn_backHair_D_MSTR}
    constraintSetup [
     {targetBone {boneName=spine_05}}
     {targetBone {boneName=clavicle_r}}
     {targetBone {boneName=spine_03}}
     {targetBone {boneName=spine_04}}
     {targetBone {boneName=clavicle_l}}
    ]
    constraintWeights [1,1,1,1,1]
   }
   n22 = "Constraint - dyn_backHair_D_01 (PPPPP)"!(
    ComponentPose=@n21
    Alpha=1
    ConstraintWeights_0=0.6044
    ConstraintWeights_1:float=0.2487
    ConstraintWeights_2:float=0.1127
    ConstraintWeights_3:float=0.0268
    ConstraintWeights_4:float=0.0073
   ) -> (Pose) {
    alphaBoolBlend {blendOption=Linear}
    boneToModify {boneName=dyn_backHair_D_01}
    constraintSetup [
     {targetBone {boneName=spine_05}}
     {targetBone {boneName=clavicle_r}}
     {targetBone {boneName=spine_03}}
     {targetBone {boneName=spine_04}}
     {targetBone {boneName=clavicle_l}}
    ]
    constraintWeights [1,1,1,1,1]
   }
   n23 = "Constraint - dyn_backHair_D_02 (PPPPP)"!(
    ComponentPose=@n22
    Alpha=1
    ConstraintWeights_0=0.4286
    ConstraintWeights_1:float=0.2494
    ConstraintWeights_2:float=0.2479
    ConstraintWeights_3:float=0.0733
    ConstraintWeights_4:float=0.0008
   ) -> (Pose) {
    alphaBoolBlend {blendOption=Linear}
    boneToModify {boneName=dyn_backHair_D_02}
    constraintSetup [
     {targetBone {boneName=spine_05}}
     {targetBone {boneName=spine_03}}
     {targetBone {boneName=clavicle_r}}
     {targetBone {boneName=spine_04}}
     {targetBone {boneName=clavicle_l}}
    ]
    constraintWeights [1,1,1,1,1]
   }
   n24 = "Constraint - dyn_backHair_D_03 (PPPPP)"!(
    ComponentPose=@n23
    Alpha=1
    ConstraintWeights_0=0.3499
    ConstraintWeights_1:float=0.2637
    ConstraintWeights_2:float=0.1911
    ConstraintWeights_3:float=0.1704
    ConstraintWeights_4:float=0.0248
   ) -> (Pose) {
    alphaBoolBlend {blendOption=Linear}
    boneToModify {boneName=dyn_backHair_D_03}
    constraintSetup [
     {targetBone {boneName=spine_03}}
     {targetBone {boneName=clavicle_r}}
     {targetBone {boneName=spine_04}}
     {targetBone {boneName=spine_05}}
     {targetBone {boneName=spine_02}}
    ]
    constraintWeights [1,1,1,1,1]
   }
   n25 = "Constraint - dyn_backHair_D_04 (PPPPPPP)"!(
    ComponentPose=@n24
    Alpha=1
    ConstraintWeights_0=0.3989
    ConstraintWeights_1:float=0.2283
    ConstraintWeights_2:float=0.141
    ConstraintWeights_3:float=0.1057
    ConstraintWeights_4:float=0.058
    ConstraintWeights_5:float=0.048
    ConstraintWeights_6:float=0.0201
   ) -> (Pose) {
    alphaBoolBlend {blendOption=Linear}
    boneToModify {boneName=dyn_backHair_D_04}
    constraintSetup [
     {targetBone {boneName=spine_03}}
     {targetBone {boneName=spine_04}}
     {targetBone {boneName=spine_02}}
     {targetBone {boneName=clavicle_r}}
     {targetBone {boneName=spine_01}}
     {targetBone {boneName=spine_05}}
     {targetBone {boneName=pelvis}}
    ]
    constraintWeights [1,1,1,1,1,1,1]
   }
   n26 = "Constraint - dyn_backHair_F_MSTR (PPPPP)"!(
    ComponentPose=@n25
    Alpha=1
    ConstraintWeights_0=0.5761
    ConstraintWeights_1:float=0.3201
    ConstraintWeights_2:float=0.0808
    ConstraintWeights_3:float=0.02
    ConstraintWeights_4:float=0.0029
   ) -> (Pose) {
    alphaBoolBlend {blendOption=Linear}
    boneToModify {boneName=dyn_backHair_F_MSTR}
    constraintSetup [
     {targetBone {boneName=spine_05}}
     {targetBone {boneName=clavicle_r}}
     {targetBone {boneName=spine_03}}
     {targetBone {boneName=spine_04}}
     {targetBone {boneName=clavicle_l}}
    ]
    constraintWeights [1,1,1,1,1]
   }
   n27 = "Constraint - dyn_backHair_F_01 (PPPPP)"!(
    ComponentPose=@n26
    Alpha=1
    ConstraintWeights_0=0.5761
    ConstraintWeights_1:float=0.3201
    ConstraintWeights_2:float=0.0808
    ConstraintWeights_3:float=0.02
    ConstraintWeights_4:float=0.0029
   ) -> (Pose) {
    alphaBoolBlend {blendOption=Linear}
    boneToModify {boneName=dyn_backHair_F_01}
    constraintSetup [
     {targetBone {boneName=spine_05}}
     {targetBone {boneName=clavicle_r}}
     {targetBone {boneName=spine_03}}
     {targetBone {boneName=spine_04}}
     {targetBone {boneName=clavicle_l}}
    ]
    constraintWeights [1,1,1,1,1]
   }
   n28 = "Constraint - dyn_backHair_F_02 (PPPPP)"!(
    ComponentPose=@n27
    Alpha=1
    ConstraintWeights_0=0.5059
    ConstraintWeights_1:float=0.3355
    ConstraintWeights_2:float=0.1226
    ConstraintWeights_3:float=0.0359
    ConstraintWeights_4:float=0.0001
   ) -> (Pose) {
    alphaBoolBlend {blendOption=Linear}
    boneToModify {boneName=dyn_backHair_F_02}
    constraintSetup [
     {targetBone {boneName=clavicle_r}}
     {targetBone {boneName=spine_05}}
     {targetBone {boneName=spine_03}}
     {targetBone {boneName=spine_04}}
     {targetBone {boneName=clavicle_l}}
    ]
    constraintWeights [1,1,1,1,1]
   }
   n29 = "Constraint - dyn_backHair_F_03 (PPPPP)"!(
    ComponentPose=@n28
    Alpha=1
    ConstraintWeights_0=0.5601
    ConstraintWeights_1:float=0.2454
    ConstraintWeights_2:float=0.1085
    ConstraintWeights_3:float=0.082
    ConstraintWeights_4:float=0.0041
   ) -> (Pose) {
    alphaBoolBlend {blendOption=Linear}
    boneToModify {boneName=dyn_backHair_F_03}
    constraintSetup [
     {targetBone {boneName=clavicle_r}}
     {targetBone {boneName=spine_03}}
     {targetBone {boneName=spine_04}}
     {targetBone {boneName=spine_05}}
     {targetBone {boneName=spine_02}}
    ]
    constraintWeights [1,1,1,1,1]
   }
   n30 = "Constraint - dyn_backHair_F_04 (PPPPPP)"!(
    ComponentPose=@n29
    Alpha=1
    ConstraintWeights_0=0.393
    ConstraintWeights_1:float=0.2839
    ConstraintWeights_2:float=0.2269
    ConstraintWeights_3:float=0.0677
    ConstraintWeights_4:float=0.0183
    ConstraintWeights_5:float=0.0102
   ) -> (Pose) {
    alphaBoolBlend {blendOption=Linear}
    boneToModify {boneName=dyn_backHair_F_04}
    constraintSetup [
     {targetBone {boneName=clavicle_r}}
     {targetBone {boneName=spine_03}}
     {targetBone {boneName=spine_04}}
     {targetBone {boneName=spine_02}}
     {targetBone {boneName=spine_05}}
     {targetBone {boneName=spine_01}}
    ]
    constraintWeights [1,1,1,1,1,1]
   }
   n31 = "Constraint - dyn_backHair_E_MSTR (PPPPP)"!(
    ComponentPose=@n30
    Alpha=1
    ConstraintWeights_0=0.5869
    ConstraintWeights_1:float=0.2917
    ConstraintWeights_2:float=0.0941
    ConstraintWeights_3:float=0.0229
    ConstraintWeights_4:float=0.0044
   ) -> (Pose) {
    alphaBoolBlend {blendOption=Linear}
    boneToModify {boneName=dyn_backHair_E_MSTR}
    constraintSetup [
     {targetBone {boneName=spine_05}}
     {targetBone {boneName=clavicle_r}}
     {targetBone {boneName=spine_03}}
     {targetBone {boneName=spine_04}}
     {targetBone {boneName=clavicle_l}}
    ]
    constraintWeights [1,1,1,1,1]
   }
   n32 = "Constraint - dyn_backHair_E_01 (PPPPP)"!(
    ComponentPose=@n31
    Alpha=1
    ConstraintWeights_0=0.5869
    ConstraintWeights_1:float=0.2917
    ConstraintWeights_2:float=0.0941
    ConstraintWeights_3:float=0.0229
    ConstraintWeights_4:float=0.0044
   ) -> (Pose) {
    alphaBoolBlend {blendOption=Linear}
    boneToModify {boneName=dyn_backHair_E_01}
    constraintSetup [
     {targetBone {boneName=spine_05}}
     {targetBone {boneName=clavicle_r}}
     {targetBone {boneName=spine_03}}
     {targetBone {boneName=spine_04}}
     {targetBone {boneName=clavicle_l}}
    ]
    constraintWeights [1,1,1,1,1]
   }
   n33 = "Constraint - dyn_backHair_E_02 (PPPP)"!(
    ComponentPose=@n32
    Alpha=1
    ConstraintWeights_0=0.4088
    ConstraintWeights_1:float=0.3523
    ConstraintWeights_2:float=0.1858
    ConstraintWeights_3:float=0.0532
   ) -> (Pose) {
    alphaBoolBlend {blendOption=Linear}
    boneToModify {boneName=dyn_backHair_E_02}
    constraintSetup [{targetBone {boneName=spine_05}}, {targetBone {boneName=clavicle_r}}, {targetBone {boneName=spine_03}}, {targetBone {boneName=spine_04}}]
    constraintWeights [1,1,1,1]
   }
   n34 = "Constraint - dyn_backHair_E_03 (PPPPPP)"!(
    ComponentPose=@n33
    Alpha=1
    ConstraintWeights_0=0.3796
    ConstraintWeights_1:float=0.3089
    ConstraintWeights_2:float=0.1741
    ConstraintWeights_3:float=0.108
    ConstraintWeights_4:float=0.0294
    ConstraintWeights_5:float=0
   ) -> (Pose) {
    alphaBoolBlend {blendOption=Linear}
    boneToModify {boneName=dyn_backHair_E_03}
    constraintSetup [
     {targetBone {boneName=clavicle_r}}
     {targetBone {boneName=spine_03}}
     {targetBone {boneName=spine_04}}
     {targetBone {boneName=spine_05}}
     {targetBone {boneName=spine_02}}
     {targetBone {boneName=spine_01}}
    ]
    constraintWeights [1,1,1,1,1,1]
   }
   n35 = "Constraint - dyn_backHair_E_04 (PPPPPP)"!(
    ComponentPose=@n34
    Alpha=1
    ConstraintWeights_0=0.3461
    ConstraintWeights_1:float=0.2346
    ConstraintWeights_2:float=0.2337
    ConstraintWeights_3:float=0.1061
    ConstraintWeights_4:float=0.0422
    ConstraintWeights_5:float=0.0373
   ) -> (Pose) {
    alphaBoolBlend {blendOption=Linear}
    boneToModify {boneName=dyn_backHair_E_04}
    constraintSetup [
     {targetBone {boneName=spine_03}}
     {targetBone {boneName=clavicle_r}}
     {targetBone {boneName=spine_04}}
     {targetBone {boneName=spine_02}}
     {targetBone {boneName=spine_05}}
     {targetBone {boneName=spine_01}}
    ]
    constraintWeights [1,1,1,1,1,1]
   }
   n36 = "Constraint - dyn_backHair_B_MSTR (PPPPP)"!(
    ComponentPose=@n35
    Alpha=1
    ConstraintWeights_0=0.6488
    ConstraintWeights_1:float=0.2273
    ConstraintWeights_2:float=0.0925
    ConstraintWeights_3:float=0.0251
    ConstraintWeights_4:float=0.0063
   ) -> (Pose) {
    alphaBoolBlend {blendOption=Linear}
    boneToModify {boneName=dyn_backHair_B_MSTR}
    constraintSetup [
     {targetBone {boneName=spine_05}}
     {targetBone {boneName=clavicle_l}}
     {targetBone {boneName=spine_03}}
     {targetBone {boneName=spine_04}}
     {targetBone {boneName=clavicle_r}}
    ]
    constraintWeights [1,1,1,1,1]
   }
   n37 = "Constraint - dyn_backHair_B_01 (PPPPP)"!(
    ComponentPose=@n36
    Alpha=1
    ConstraintWeights_0=0.6488
    ConstraintWeights_1:float=0.2273
    ConstraintWeights_2:float=0.0925
    ConstraintWeights_3:float=0.0251
    ConstraintWeights_4:float=0.0063
   ) -> (Pose) {
    alphaBoolBlend {blendOption=Linear}
    boneToModify {boneName=dyn_backHair_B_01}
    constraintSetup [
     {targetBone {boneName=spine_05}}
     {targetBone {boneName=clavicle_l}}
     {targetBone {boneName=spine_03}}
     {targetBone {boneName=spine_04}}
     {targetBone {boneName=clavicle_r}}
    ]
    constraintWeights [1,1,1,1,1]
   }
   n38 = "Constraint - dyn_backHair_B_02 (PPPPP)"!(
    ComponentPose=@n37
    Alpha=1
    ConstraintWeights_0=0.4133
    ConstraintWeights_1:float=0.3645
    ConstraintWeights_2:float=0.1732
    ConstraintWeights_3:float=0.049
    ConstraintWeights_4:float=0
   ) -> (Pose) {
    alphaBoolBlend {blendOption=Linear}
    boneToModify {boneName=dyn_backHair_B_02}
    constraintSetup [
     {targetBone {boneName=spine_05}}
     {targetBone {boneName=clavicle_l}}
     {targetBone {boneName=spine_03}}
     {targetBone {boneName=spine_04}}
     {targetBone {boneName=upperarm_l}}
    ]
    constraintWeights [1,1,1,1,1]
   }
   n39 = "Constraint - dyn_backHair_B_03 (PPPPP)"!(
    ComponentPose=@n38
    Alpha=1
    ConstraintWeights_0=0.4369
    ConstraintWeights_1:float=0.2936
    ConstraintWeights_2:float=0.1652
    ConstraintWeights_3:float=0.0789
    ConstraintWeights_4:float=0.0254
   ) -> (Pose) {
    alphaBoolBlend {blendOption=Linear}
    boneToModify {boneName=dyn_backHair_B_03}
    constraintSetup [
     {targetBone {boneName=clavicle_l}}
     {targetBone {boneName=spine_03}}
     {targetBone {boneName=spine_04}}
     {targetBone {boneName=spine_05}}
     {targetBone {boneName=spine_02}}
    ]
    constraintWeights [1,1,1,1,1]
   }
   n40 = "Constraint - dyn_backHair_B_04 (PPPPPPP)"!(
    ComponentPose=@n39
    Alpha=1
    ConstraintWeights_0=0.3388
    ConstraintWeights_1:float=0.2439
    ConstraintWeights_2:float=0.2397
    ConstraintWeights_3:float=0.1101
    ConstraintWeights_4:float=0.0383
    ConstraintWeights_5:float=0.0285
    ConstraintWeights_6:float=0.0008
   ) -> (Pose) {
    alphaBoolBlend {blendOption=Linear}
    boneToModify {boneName=dyn_backHair_B_04}
    constraintSetup [
     {targetBone {boneName=spine_03}}
     {targetBone {boneName=clavicle_l}}
     {targetBone {boneName=spine_04}}
     {targetBone {boneName=spine_02}}
     {targetBone {boneName=spine_01}}
     {targetBone {boneName=spine_05}}
     {targetBone {boneName=pelvis}}
    ]
    constraintWeights [1,1,1,1,1,1,1]
   }
   n41 = "Constraint - dyn_l_frontHair_MSTR (PPPPPPP)"!(
    ComponentPose=@n40
    Alpha=1
    ConstraintWeights_0=1
    ConstraintWeights_1:float=0
    ConstraintWeights_2:float=0
    ConstraintWeights_3:float=0
    ConstraintWeights_4:float=0
    ConstraintWeights_5:float=0
    ConstraintWeights_6:float=0
   ) -> (Pose) {
    alphaBoolBlend {blendOption=Linear}
    boneToModify {boneName=dyn_l_frontHair_MSTR}
    constraintSetup [
     {targetBone {boneName=spine_05}}
     {targetBone {boneName=clavicle_l}}
     {targetBone {boneName=spine_03}}
     {targetBone {boneName=spine_04}}
     {targetBone {boneName=deform_pec_l}}
     {targetBone {boneName=upperarm_l}}
     {targetBone {boneName=clavicle_r}}
    ]
    constraintWeights [1,1,1,1,1,1,1]
   }
   n42 = "Constraint - dyn_l_frontHair_01 (PPPPPPP)"!(
    ComponentPose=@n41
    Alpha=1
    ConstraintWeights_0=1
    ConstraintWeights_1:float=0
    ConstraintWeights_2:float=0
    ConstraintWeights_3:float=0
    ConstraintWeights_4:float=0
    ConstraintWeights_5:float=0
    ConstraintWeights_6:float=0
   ) -> (Pose) {
    alphaBoolBlend {blendOption=Linear}
    boneToModify {boneName=dyn_l_frontHair_01}
    constraintSetup [
     {targetBone {boneName=spine_05}}
     {targetBone {boneName=clavicle_l}}
     {targetBone {boneName=spine_03}}
     {targetBone {boneName=spine_04}}
     {targetBone {boneName=deform_pec_l}}
     {targetBone {boneName=upperarm_l}}
     {targetBone {boneName=clavicle_r}}
    ]
    constraintWeights [1,1,1,1,1,1,1]
   }
   n43 = "Constraint - dyn_l_frontHair_02 (PPPPPP)"!(
    ComponentPose=@n42
    Alpha=1
    ConstraintWeights_0=0.4616
    ConstraintWeights_1:float=0.3671
    ConstraintWeights_2:float=0.0716
    ConstraintWeights_3:float=0.0663
    ConstraintWeights_4:float=0.0335
    ConstraintWeights_5:float=0
   ) -> (Pose) {
    alphaBoolBlend {blendOption=Linear}
    boneToModify {boneName=dyn_l_frontHair_02}
    constraintSetup [
     {targetBone {boneName=spine_05}}
     {targetBone {boneName=clavicle_l}}
     {targetBone {boneName=spine_04}}
     {targetBone {boneName=deform_pec_l}}
     {targetBone {boneName=spine_03}}
     {targetBone {boneName=upperarm_l}}
    ]
    constraintWeights [1,1,1,1,1,1]
   }
   n44 = "Constraint - dyn_l_frontHair_03 (PPPPPP)"!(
    ComponentPose=@n43
    Alpha=1
    ConstraintWeights_0=0.2924
    ConstraintWeights_1:float=0.2032
    ConstraintWeights_2:float=0.1954
    ConstraintWeights_3:float=0.1604
    ConstraintWeights_4:float=0.1112
    ConstraintWeights_5:float=0.0374
   ) -> (Pose) {
    alphaBoolBlend {blendOption=Linear}
    boneToModify {boneName=dyn_l_frontHair_03}
    constraintSetup [
     {targetBone {boneName=deform_pec_l}}
     {targetBone {boneName=clavicle_l}}
     {targetBone {boneName=spine_04}}
     {targetBone {boneName=spine_05}}
     {targetBone {boneName=spine_03}}
     {targetBone {boneName=spine_02}}
    ]
    constraintWeights [1,1,1,1,1,1]
   }
   n45 = "Constraint - dyn_l_frontHair_04 (PPPPPPP)"!(
    ComponentPose=@n44
    Alpha=1
    ConstraintWeights_0=0.2487
    ConstraintWeights_1:float=0.2264
    ConstraintWeights_2:float=0.2233
    ConstraintWeights_3:float=0.1252
    ConstraintWeights_4:float=0.1066
    ConstraintWeights_5:float=0.052
    ConstraintWeights_6:float=0.0177
   ) -> (Pose) {
    alphaBoolBlend {blendOption=Linear}
    boneToModify {boneName=dyn_l_frontHair_04}
    constraintSetup [
     {targetBone {boneName=deform_pec_l}}
     {targetBone {boneName=spine_04}}
     {targetBone {boneName=spine_03}}
     {targetBone {boneName=clavicle_l}}
     {targetBone {boneName=spine_02}}
     {targetBone {boneName=spine_05}}
     {targetBone {boneName=spine_01}}
    ]
    constraintWeights [1,1,1,1,1,1,1]
   }
   n46 = "Constraint - dyn_l_frontHair_05 (PPPPPPP)"!(
    ComponentPose=@n45
    Alpha=1
    ConstraintWeights_0=0.3717
    ConstraintWeights_1:float=0.2551
    ConstraintWeights_2:float=0.1233
    ConstraintWeights_3:float=0.1136
    ConstraintWeights_4:float=0.0601
    ConstraintWeights_5:float=0.0453
    ConstraintWeights_6:float=0.0308
   ) -> (Pose) {
    alphaBoolBlend {blendOption=Linear}
    boneToModify {boneName=dyn_l_frontHair_05}
    constraintSetup [
     {targetBone {boneName=spine_03}}
     {targetBone {boneName=spine_02}}
     {targetBone {boneName=spine_04}}
     {targetBone {boneName=spine_01}}
     {targetBone {boneName=clavicle_l}}
     {targetBone {boneName=pelvis}}
     {targetBone {boneName=deform_pec_l}}
    ]
    constraintWeights [1,1,1,1,1,1,1]
   }
   n47 = "Constraint - dyn_r_frontHair_MSTR (PPPPP)"!(
    ComponentPose=@n46
    Alpha=1
    ConstraintWeights_0=0.666
    ConstraintWeights_1:float=0.2727
    ConstraintWeights_2:float=0.0314
    ConstraintWeights_3:float=0.0193
    ConstraintWeights_4:float=0.0104
   ) -> (Pose) {
    alphaBoolBlend {blendOption=Linear}
    boneToModify {boneName=dyn_r_frontHair_MSTR}
    constraintSetup [
     {targetBone {boneName=spine_05}}
     {targetBone {boneName=clavicle_r}}
     {targetBone {boneName=spine_03}}
     {targetBone {boneName=spine_04}}
     {targetBone {boneName=deform_pec_r}}
    ]
    constraintWeights [1,1,1,1,1]
   }
   n48 = "Constraint - dyn_r_frontHair_01 (PPPPP)"!(
    ComponentPose=@n47
    Alpha=1
    ConstraintWeights_0=0.666
    ConstraintWeights_1:float=0.2727
    ConstraintWeights_2:float=0.0314
    ConstraintWeights_3:float=0.0193
    ConstraintWeights_4:float=0.0104
   ) -> (Pose) {
    alphaBoolBlend {blendOption=Linear}
    boneToModify {boneName=dyn_r_frontHair_01}
    constraintSetup [
     {targetBone {boneName=spine_05}}
     {targetBone {boneName=clavicle_r}}
     {targetBone {boneName=spine_03}}
     {targetBone {boneName=spine_04}}
     {targetBone {boneName=deform_pec_r}}
    ]
    constraintWeights [1,1,1,1,1]
   }
   n49 = "Constraint - dyn_r_frontHair_02 (PPPPP)"!(
    ComponentPose=@n48
    Alpha=1
    ConstraintWeights_0=0.6401
    ConstraintWeights_1:float=0.1767
    ConstraintWeights_2:float=0.0805
    ConstraintWeights_3:float=0.067
    ConstraintWeights_4:float=0.0356
   ) -> (Pose) {
    alphaBoolBlend {blendOption=Linear}
    boneToModify {boneName=dyn_r_frontHair_02}
    constraintSetup [
     {targetBone {boneName=spine_05}}
     {targetBone {boneName=clavicle_r}}
     {targetBone {boneName=spine_04}}
     {targetBone {boneName=deform_pec_r}}
     {targetBone {boneName=spine_03}}
    ]
    constraintWeights [1,1,1,1,1]
   }
   n50 = "Constraint - dyn_r_frontHair_03 (PPPPPP)"!(
    ComponentPose=@n49
    Alpha=1
    ConstraintWeights_0=0.251
    ConstraintWeights_1:float=0.23
    ConstraintWeights_2:float=0.1929
    ConstraintWeights_3:float=0.1585
    ConstraintWeights_4:float=0.1199
    ConstraintWeights_5:float=0.0477
   ) -> (Pose) {
    alphaBoolBlend {blendOption=Linear}
    boneToModify {boneName=dyn_r_frontHair_03}
    constraintSetup [
     {targetBone {boneName=spine_04}}
     {targetBone {boneName=deform_pec_r}}
     {targetBone {boneName=spine_05}}
     {targetBone {boneName=clavicle_r}}
     {targetBone {boneName=spine_03}}
     {targetBone {boneName=spine_02}}
    ]
    constraintWeights [1,1,1,1,1,1]
   }
   n51 = "Constraint - dyn_r_frontHair_04 (PPPPPPP)"!(
    ComponentPose=@n50
    Alpha=1
    ConstraintWeights_0=0.2902
    ConstraintWeights_1:float=0.2127
    ConstraintWeights_2:float=0.1776
    ConstraintWeights_3:float=0.1444
    ConstraintWeights_4:float=0.1045
    ConstraintWeights_5:float=0.0464
    ConstraintWeights_6:float=0.0241
   ) -> (Pose) {
    alphaBoolBlend {blendOption=Linear}
    boneToModify {boneName=dyn_r_frontHair_04}
    constraintSetup [
     {targetBone {boneName=spine_03}}
     {targetBone {boneName=spine_04}}
     {targetBone {boneName=deform_pec_r}}
     {targetBone {boneName=spine_02}}
     {targetBone {boneName=clavicle_r}}
     {targetBone {boneName=spine_01}}
     {targetBone {boneName=spine_05}}
    ]
    constraintWeights [1,1,1,1,1,1,1]
   }
   n52 = "Constraint - dyn_r_frontHair_05 (PPPPPPP)"!(
    ComponentPose=@n51
    Alpha=1
    ConstraintWeights_0=0.3717
    ConstraintWeights_1:float=0.2551
    ConstraintWeights_2:float=0.1233
    ConstraintWeights_3:float=0.1136
    ConstraintWeights_4:float=0.0601
    ConstraintWeights_5:float=0.0453
    ConstraintWeights_6:float=0.0308
   ) -> (Pose) {
    alphaBoolBlend {blendOption=Linear}
    boneToModify {boneName=dyn_r_frontHair_05}
    constraintSetup [
     {targetBone {boneName=spine_03}}
     {targetBone {boneName=spine_02}}
     {targetBone {boneName=spine_04}}
     {targetBone {boneName=spine_01}}
     {targetBone {boneName=clavicle_r}}
     {targetBone {boneName=pelvis}}
     {targetBone {boneName=deform_pec_r}}
    ]
    constraintWeights [1,1,1,1,1,1,1]
   }
   n53 = "Transform (Modify) Bone - Bone: dyn_backHair_A_MSTR"!(ComponentPose=@n52, bAlphaBoolEnabled:bool=True, Translation="0,0,0", Rotation="-6.878,-16.652,-0.021", Scale="1,1,1") -> (Pose) {
    alphaInputType=Bool
    rotationMode=BMM_Additive
    rotationSpace=BCS_BoneSpace
    scaleMode=BMM_Additive
    scaleSpace=BCS_BoneSpace
    translationMode=BMM_Additive
    translationSpace=BCS_BoneSpace
    alphaBoolBlend {blendInTime=0.1, blendOption=Linear, blendOutTime=0.1}
    boneToModify {boneName=dyn_backHair_A_MSTR}
    rotation=<rot -6.878 -16.652 -0.021>
   }
   n54 = "Transform (Modify) Bone - Bone: dyn_backHair_B_MSTR"!(ComponentPose=@n53, bAlphaBoolEnabled:bool=True, Translation="0,0,0", Rotation="-0.006,-8.715,-0.011", Scale="1,1,1") -> (Pose) {
    alphaInputType=Bool
    rotationMode=BMM_Additive
    rotationSpace=BCS_BoneSpace
    scaleMode=BMM_Additive
    scaleSpace=BCS_BoneSpace
    translationMode=BMM_Additive
    translationSpace=BCS_BoneSpace
    alphaBoolBlend {blendInTime=0.1, blendOption=Linear, blendOutTime=0.1}
    boneToModify {boneName=dyn_backHair_B_MSTR}
    rotation=<rot -0.006 -8.715 -0.011>
   }
   n55 = "Transform (Modify) Bone - Bone: dyn_backHair_C_MSTR"!(ComponentPose=@n54, bAlphaBoolEnabled:bool=True, Translation="0,0,0", Rotation="0,0,0", Scale="1,1,1") -> (Pose) {
    alphaInputType=Bool
    rotationMode=BMM_Additive
    rotationSpace=BCS_BoneSpace
    scaleMode=BMM_Additive
    scaleSpace=BCS_BoneSpace
    translationMode=BMM_Additive
    translationSpace=BCS_BoneSpace
    alphaBoolBlend {blendInTime=0.1, blendOption=Linear, blendOutTime=0.1}
    boneToModify {boneName=dyn_backHair_C_MSTR}
   }
   n56 = "Transform (Modify) Bone - Bone: dyn_backHair_D_MSTR"!(ComponentPose=@n55, bAlphaBoolEnabled:bool=True, Translation="0,0,0", Rotation="0.030,-0.179,0.009", Scale="1,1,1") -> (Pose) {
    alphaInputType=Bool
    rotationMode=BMM_Additive
    rotationSpace=BCS_BoneSpace
    scaleMode=BMM_Additive
    scaleSpace=BCS_BoneSpace
    translationMode=BMM_Additive
    translationSpace=BCS_BoneSpace
    alphaBoolBlend {blendInTime=0.1, blendOption=Linear, blendOutTime=0.1}
    boneToModify {boneName=dyn_backHair_D_MSTR}
    rotation=<rot 0.03 -0.179 0.009>
   }
   n57 = "Transform (Modify) Bone - Bone: dyn_backHair_E_MSTR"!(ComponentPose=@n56, bAlphaBoolEnabled:bool=True, Translation="0,0,0", Rotation="0.013,8.553,-0.007", Scale="1,1,1") -> (Pose) {
    alphaInputType=Bool
    rotationMode=BMM_Additive
    rotationSpace=BCS_BoneSpace
    scaleMode=BMM_Additive
    scaleSpace=BCS_BoneSpace
    translationMode=BMM_Additive
    translationSpace=BCS_BoneSpace
    alphaBoolBlend {blendInTime=0.1, blendOption=Linear, blendOutTime=0.1}
    boneToModify {boneName=dyn_backHair_E_MSTR}
    rotation=<rot 0.013 8.553 -0.007>
   }
   n58 = "Transform (Modify) Bone - Bone: dyn_backHair_F_MSTR"!(ComponentPose=@n57, bAlphaBoolEnabled:bool=True, Translation="0,0,0", Rotation="-0.010,12.588,-0.007", Scale="1,1,1") -> (Pose) {
    alphaInputType=Bool
    rotationMode=BMM_Additive
    rotationSpace=BCS_BoneSpace
    scaleMode=BMM_Additive
    scaleSpace=BCS_BoneSpace
    translationMode=BMM_Additive
    translationSpace=BCS_BoneSpace
    alphaBoolBlend {blendInTime=0.1, blendOption=Linear, blendOutTime=0.1}
    boneToModify {boneName=dyn_backHair_F_MSTR}
    rotation=<rot -0.01 12.588 -0.007>
   }
   n59 = Trail controller - Bone: dyn_backHair_A_04!(ComponentPose=@n58, Alpha=0.8, FakeVelocity:Vector="0,0,0") -> (Pose) {
    bActorSpaceFakeVel=true
    bShowTrailLocation=true
    bUsePlanarLimit=true
    chainLength=4
    debugLifeTime=0.1
    maxDeltaTime=0.0333
    relaxationSpeedScale=3
    alphaScaleBiasClamp {bInterpResult=true}
    baseJoint {boneName=root}
    planarLimits [
     {drivingBone {boneName=root}, planeTransform {location=<0 0 10>, rotation=<rot 0 0 0>, scale=<1 1 1>}}
     {drivingBone {boneName=spine_03}, planeTransform {location=<0 -20 0>, rotation=<rot 0 15 -89.9999>, scale=<1 1 1>}}
    ]
    rotationLimits [{limitMax=<180 180 180>}, {limitMax=<180 180 180>}, {limitMax=<180 180 180>}, {limitMax=<180 180 180>}]
    rotationOffsets [<0 0 0>, <0 0 0>, <0 0 0>, <0 0 0>]
    trailBone {boneName=dyn_backHair_A_04}
    trailRelaxationSpeed {editorCurveData {keys [{arriveTangent=-72.0739, interpMode=RCIM_Cubic, leaveTangent=-72.0739, tangentMode=RCTM_User, time=0, value=35}, <at 1 5 RCIM_Cubic>]}}
   }
   n60 = Trail controller - Bone: dyn_backHair_B_04!(ComponentPose=@n59, Alpha=0.6, FakeVelocity:Vector="0,0,0") -> (Pose) {
    bActorSpaceFakeVel=true
    bShowTrailLocation=true
    bUsePlanarLimit=true
    chainLength=4
    debugLifeTime=0.1
    maxDeltaTime=0.0333
    relaxationSpeedScale=3
    alphaScaleBiasClamp {bInterpResult=true}
    baseJoint {boneName=root}
    planarLimits [
     {drivingBone {boneName=root}, planeTransform {location=<0 0 10>, rotation=<rot 0 0 0>, scale=<1 1 1>}}
     {drivingBone {boneName=spine_03}, planeTransform {location=<0 -20 0>, rotation=<rot 0 15 -89.9999>, scale=<1 1 1>}}
    ]
    rotationLimits [{limitMax=<180 180 180>}, {limitMax=<180 180 180>}, {limitMax=<180 180 180>}, {limitMax=<180 180 180>}]
    rotationOffsets [<0 0 0>, <0 0 0>, <0 0 0>, <0 0 0>]
    trailBone {boneName=dyn_backHair_B_04}
    trailRelaxationSpeed {editorCurveData {keys [{arriveTangent=-72.0739, interpMode=RCIM_Cubic, leaveTangent=-72.0739, tangentMode=RCTM_User, time=0, value=35}, <at 1 5 RCIM_Cubic>]}}
   }
   n61 = Trail controller - Bone: dyn_backHair_C_04!(ComponentPose=@n60, Alpha=0.7, FakeVelocity:Vector="0,0,0") -> (Pose) {
    bActorSpaceFakeVel=true
    bShowTrailLocation=true
    bUsePlanarLimit=true
    chainLength=4
    debugLifeTime=0.1
    maxDeltaTime=0.0333
    relaxationSpeedScale=3
    alphaScaleBiasClamp {bInterpResult=true}
    baseJoint {boneName=root}
    planarLimits [
     {drivingBone {boneName=root}, planeTransform {location=<0 0 10>, rotation=<rot 0 0 0>, scale=<1 1 1>}}
     {drivingBone {boneName=spine_03}, planeTransform {location=<0 -20 0>, rotation=<rot 0 15 -89.9999>, scale=<1 1 1>}}
    ]
    rotationLimits [{limitMax=<180 180 180>}, {limitMax=<180 180 180>}, {limitMax=<180 180 180>}, {limitMax=<180 180 180>}]
    rotationOffsets [<0 0 0>, <0 0 0>, <0 0 0>, <0 0 0>]
    trailBone {boneName=dyn_backHair_C_04}
    trailRelaxationSpeed {editorCurveData {keys [{arriveTangent=-72.0739, interpMode=RCIM_Cubic, leaveTangent=-72.0739, tangentMode=RCTM_User, time=0, value=35}, <at 1 5 RCIM_Cubic>]}}
   }
   n62 = Trail controller - Bone: dyn_backHair_D_04!(ComponentPose=@n61, Alpha=0.77, FakeVelocity:Vector="0,0,0") -> (Pose) {
    bActorSpaceFakeVel=true
    bShowTrailLocation=true
    bUsePlanarLimit=true
    chainLength=4
    debugLifeTime=0.1
    maxDeltaTime=0.0333
    relaxationSpeedScale=3
    alphaScaleBiasClamp {bInterpResult=true}
    baseJoint {boneName=root}
    planarLimits [
     {drivingBone {boneName=root}, planeTransform {location=<0 0 10>, rotation=<rot 0 0 0>, scale=<1 1 1>}}
     {drivingBone {boneName=spine_03}, planeTransform {location=<0 -20 0>, rotation=<rot 0 15 -89.9999>, scale=<1 1 1>}}
    ]
    rotationLimits [{limitMax=<180 180 180>}, {limitMax=<180 180 180>}, {limitMax=<180 180 180>}, {limitMax=<180 180 180>}]
    rotationOffsets [<0 0 0>, <0 0 0>, <0 0 0>, <0 0 0>]
    trailBone {boneName=dyn_backHair_D_04}
    trailRelaxationSpeed {editorCurveData {keys [{arriveTangent=-72.0739, interpMode=RCIM_Cubic, leaveTangent=-72.0739, tangentMode=RCTM_User, time=0, value=35}, <at 1 5 RCIM_Cubic>]}}
   }
   n63 = Trail controller - Bone: dyn_backHair_E_04!(ComponentPose=@n62, Alpha=0.8, FakeVelocity:Vector="0,0,0") -> (Pose) {
    bActorSpaceFakeVel=true
    bShowTrailLocation=true
    bUsePlanarLimit=true
    chainLength=4
    debugLifeTime=0.1
    maxDeltaTime=0.0333
    relaxationSpeedScale=3
    alphaScaleBiasClamp {bInterpResult=true}
    baseJoint {boneName=root}
    planarLimits [
     {drivingBone {boneName=root}, planeTransform {location=<0 0 10>, rotation=<rot 0 0 0>, scale=<1 1 1>}}
     {drivingBone {boneName=spine_03}, planeTransform {location=<0 -20 0>, rotation=<rot 0 15 -89.9999>, scale=<1 1 1>}}
    ]
    rotationLimits [{limitMax=<180 180 180>}, {limitMax=<180 180 180>}, {limitMax=<180 180 180>}, {limitMax=<180 180 180>}]
    rotationOffsets [<0 0 0>, <0 0 0>, <0 0 0>, <0 0 0>]
    trailBone {boneName=dyn_backHair_E_04}
    trailRelaxationSpeed {editorCurveData {keys [{arriveTangent=-72.0739, interpMode=RCIM_Cubic, leaveTangent=-72.0739, tangentMode=RCTM_User, time=0, value=35}, <at 1 5 RCIM_Cubic>]}}
   }
   n64 = Trail controller - Bone: dyn_backHair_F_04!(ComponentPose=@n63, Alpha=0.9, FakeVelocity:Vector="0,0,0") -> (Pose) {
    bActorSpaceFakeVel=true
    bShowTrailLocation=true
    bUsePlanarLimit=true
    chainLength=4
    debugLifeTime=0.1
    maxDeltaTime=0.0333
    relaxationSpeedScale=3
    alphaScaleBiasClamp {bInterpResult=true}
    baseJoint {boneName=root}
    planarLimits [
     {drivingBone {boneName=root}, planeTransform {location=<0 0 10>, rotation=<rot 0 0 0>, scale=<1 1 1>}}
     {drivingBone {boneName=spine_03}, planeTransform {location=<0 -20 0>, rotation=<rot 0 15 -89.9999>, scale=<1 1 1>}}
    ]
    rotationLimits [{limitMax=<180 180 180>}, {limitMax=<180 180 180>}, {limitMax=<180 180 180>}, {limitMax=<180 180 180>}]
    rotationOffsets [<0 0 0>, <0 0 0>, <0 0 0>, <0 0 0>]
    trailBone {boneName=dyn_backHair_F_04}
    trailRelaxationSpeed {editorCurveData {keys [{arriveTangent=-72.0739, interpMode=RCIM_Cubic, leaveTangent=-72.0739, tangentMode=RCTM_User, time=0, value=35}, <at 1 5 RCIM_Cubic>]}}
   }
   n65 = Trail controller - Bone: dyn_l_frontHair_05!(ComponentPose=@n64, Alpha=0.5, FakeVelocity:Vector="0,0,0") -> (Pose) {
    bActorSpaceFakeVel=true
    bShowTrailLocation=true
    bUsePlanarLimit=true
    chainLength=5
    debugLifeTime=0.1
    maxDeltaTime=0.0333
    relaxationSpeedScale=3
    alphaScaleBiasClamp {bInterpResult=true}
    baseJoint {boneName=root}
    planarLimits [
     {drivingBone {boneName=root}, planeTransform {location=<0 0 10>, rotation=<rot 0 0 0>, scale=<1 1 1>}}
     {drivingBone {boneName=spine_03}, planeTransform {location=<0 10 50>, rotation=<rot 0 0 90>, scale=<1 1 1>}}
    ]
    rotationLimits [{limitMax=<180 180 180>}, {limitMax=<180 180 180>}, {limitMax=<180 180 180>}, {limitMax=<180 180 180>}, {limitMax=<180 180 180>}]
    rotationOffsets [<0 0 0>, <0 0 0>, <0 0 0>, <0 0 0>, <0 0 0>]
    trailBone {boneName=dyn_l_frontHair_05}
    trailRelaxationSpeed {editorCurveData {keys [{arriveTangent=-72.0739, interpMode=RCIM_Cubic, leaveTangent=-72.0739, tangentMode=RCTM_User, time=0, value=35}, <at 1 5 RCIM_Cubic>]}}
   }
   n66 = Trail controller - Bone: dyn_r_frontHair_05!(ComponentPose=@n65, Alpha=0.5, FakeVelocity:Vector="0,0,0") -> (Pose) {
    bActorSpaceFakeVel=true
    bLimitRotation=true
    bShowTrailLocation=true
    chainLength=5
    debugLifeTime=0.1
    maxDeltaTime=0.0333
    relaxationSpeedScale=3
    alphaScaleBiasClamp {bInterpResult=true}
    baseJoint {boneName=root}
    planarLimits [
     {drivingBone {boneName=root}, planeTransform {location=<0 0 10>, rotation=<rot 0 0 0>, scale=<1 1 1>}}
     {drivingBone {boneName=spine_03}, planeTransform {location=<0 -20 0>, rotation=<rot 0 15 -89.9999>, scale=<1 1 1>}}
    ]
    rotationLimits [
     {limitMax=<50 30 30>, limitMin=<-50 -50 -50>}
     {limitMax=<50 30 30>, limitMin=<-50 -50 -50>}
     {limitMax=<50 30 30>, limitMin=<-50 -50 -50>}
     {limitMax=<50 30 30>, limitMin=<-50 -50 -50>}
     {limitMax=<50 30 30>, limitMin=<-50 -50 -50>}
    ]
    rotationOffsets [<0 0 0>, <0 0 0>, <0 0 0>, <0 0 0>, <0 0 0>]
    trailBone {boneName=dyn_r_frontHair_05}
    trailRelaxationSpeed {editorCurveData {keys [{arriveTangent=-72.0739, interpMode=RCIM_Cubic, leaveTangent=-72.0739, tangentMode=RCTM_User, time=0, value=35}, <at 1 5 RCIM_Cubic>]}}
   }
   n67 = "Constraint - dyn_backHair_A_01 (P)"!(ComponentPose=Local To Component(LocalPose=Use cached pose 'PostSim'()), bAlphaBoolEnabled:bool=True, ConstraintWeights_0=1) -> (Pose) {
    alphaInputType=Bool
    alphaBoolBlend {blendInTime=0.1, blendOption=Linear, blendOutTime=0.1}
    boneToModify {boneName=dyn_backHair_A_01}
    constraintSetup [{targetBone {boneName=spine_05}}]
    constraintWeights [1]
   }
   n68 = "Constraint - dyn_backHair_B_01 (P)"!(ComponentPose=@n67, bAlphaBoolEnabled:bool=True, ConstraintWeights_0=1) -> (Pose) {
    alphaInputType=Bool
    alphaBoolBlend {blendInTime=0.1, blendOption=Linear, blendOutTime=0.1}
    boneToModify {boneName=dyn_backHair_B_01}
    constraintSetup [{targetBone {boneName=spine_05}}]
    constraintWeights [1]
   }
   n69 = "Constraint - dyn_backHair_C_01 (P)"!(ComponentPose=@n68, bAlphaBoolEnabled:bool=True, ConstraintWeights_0=1) -> (Pose) {
    alphaInputType=Bool
    alphaBoolBlend {blendInTime=0.1, blendOption=Linear, blendOutTime=0.1}
    boneToModify {boneName=dyn_backHair_C_01}
    constraintSetup [{targetBone {boneName=spine_05}}]
    constraintWeights [1]
   }
   n70 = "Constraint - dyn_backHair_D_01 (P)"!(ComponentPose=@n69, bAlphaBoolEnabled:bool=True, ConstraintWeights_0=1) -> (Pose) {
    alphaInputType=Bool
    alphaBoolBlend {blendInTime=0.1, blendOption=Linear, blendOutTime=0.1}
    boneToModify {boneName=dyn_backHair_D_01}
    constraintSetup [{targetBone {boneName=spine_05}}]
    constraintWeights [1]
   }
   n71 = "Constraint - dyn_backHair_E_01 (P)"!(ComponentPose=@n70, bAlphaBoolEnabled:bool=True, ConstraintWeights_0=1) -> (Pose) {
    alphaInputType=Bool
    alphaBoolBlend {blendInTime=0.1, blendOption=Linear, blendOutTime=0.1}
    boneToModify {boneName=dyn_backHair_E_01}
    constraintSetup [{targetBone {boneName=spine_05}}]
    constraintWeights [1]
   }
   n72 = "Constraint - dyn_backHair_F_01 (P)"!(ComponentPose=@n71, bAlphaBoolEnabled:bool=True, ConstraintWeights_0=1) -> (Pose) {
    alphaInputType=Bool
    alphaBoolBlend {blendInTime=0.1, blendOption=Linear, blendOutTime=0.1}
    boneToModify {boneName=dyn_backHair_F_01}
    constraintSetup [{targetBone {boneName=spine_05}}]
    constraintWeights [1]
   }
   n73 = "Constraint - dyn_backHair_A_02 (P)"!(ComponentPose=@n72, bAlphaBoolEnabled:bool=True, ConstraintWeights_0=1) -> (Pose) {
    alphaInputType=Bool
    alphaBoolBlend {blendInTime=0.1, blendOption=Linear, blendOutTime=0.1}
    boneToModify {boneName=dyn_backHair_A_02}
    constraintSetup [{targetBone {boneName=spine_05}}]
    constraintWeights [1]
   }
   n74 = "Constraint - dyn_backHair_B_02 (P)"!(ComponentPose=@n73, bAlphaBoolEnabled:bool=True, ConstraintWeights_0=1) -> (Pose) {
    alphaInputType=Bool
    alphaBoolBlend {blendInTime=0.1, blendOption=Linear, blendOutTime=0.1}
    boneToModify {boneName=dyn_backHair_B_02}
    constraintSetup [{targetBone {boneName=spine_05}}]
    constraintWeights [1]
   }
   n75 = "Constraint - dyn_backHair_C_02 (P)"!(ComponentPose=@n74, bAlphaBoolEnabled:bool=True, ConstraintWeights_0=1) -> (Pose) {
    alphaInputType=Bool
    alphaBoolBlend {blendInTime=0.1, blendOption=Linear, blendOutTime=0.1}
    boneToModify {boneName=dyn_backHair_C_02}
    constraintSetup [{targetBone {boneName=spine_05}}]
    constraintWeights [1]
   }
   n76 = "Constraint - dyn_backHair_D_02 (P)"!(ComponentPose=@n75, bAlphaBoolEnabled:bool=True, ConstraintWeights_0=1) -> (Pose) {
    alphaInputType=Bool
    alphaBoolBlend {blendInTime=0.1, blendOption=Linear, blendOutTime=0.1}
    boneToModify {boneName=dyn_backHair_D_02}
    constraintSetup [{targetBone {boneName=spine_05}}]
    constraintWeights [1]
   }
   n77 = "Constraint - dyn_backHair_E_02 (P)"!(ComponentPose=@n76, bAlphaBoolEnabled:bool=True, ConstraintWeights_0=1) -> (Pose) {
    alphaInputType=Bool
    alphaBoolBlend {blendInTime=0.1, blendOption=Linear, blendOutTime=0.1}
    boneToModify {boneName=dyn_backHair_E_02}
    constraintSetup [{targetBone {boneName=spine_05}}]
    constraintWeights [1]
   }
   n78 = "Constraint - dyn_backHair_F_02 (P)"!(ComponentPose=@n77, bAlphaBoolEnabled:bool=True, ConstraintWeights_0=1) -> (Pose) {
    alphaInputType=Bool
    alphaBoolBlend {blendInTime=0.1, blendOption=Linear, blendOutTime=0.1}
    boneToModify {boneName=dyn_backHair_F_02}
    constraintSetup [{targetBone {boneName=spine_05}}]
    constraintWeights [1]
   }
   n79 = "Constraint - dyn_backHair_A_03 (P)"(ComponentPose=@n78, bAlphaBoolEnabled=True, ConstraintWeights_0=1) {
    alphaInputType=Bool
    alphaBoolBlend {blendInTime=0.1, blendOption=Linear, blendOutTime=0.1}
    boneToModify {boneName=dyn_backHair_A_03}
    constraintSetup [{targetBone {boneName=spine_05}}]
    constraintWeights [1]
   }
   n80 = "Constraint - dyn_backHair_B_03 (P)"(ComponentPose=@n79, bAlphaBoolEnabled=True, ConstraintWeights_0=1) {
    alphaInputType=Bool
    alphaBoolBlend {blendInTime=0.1, blendOption=Linear, blendOutTime=0.1}
    boneToModify {boneName=dyn_backHair_B_03}
    constraintSetup [{targetBone {boneName=spine_05}}]
    constraintWeights [1]
   }
   n81 = "Constraint - dyn_backHair_C_03 (P)"(ComponentPose=@n80, bAlphaBoolEnabled=True, ConstraintWeights_0=1) {
    alphaInputType=Bool
    alphaBoolBlend {blendInTime=0.1, blendOption=Linear, blendOutTime=0.1}
    boneToModify {boneName=dyn_backHair_C_03}
    constraintSetup [{targetBone {boneName=spine_05}}]
    constraintWeights [1]
   }
   n82 = "Constraint - dyn_backHair_D_03 (P)"(ComponentPose=@n81, bAlphaBoolEnabled=True, ConstraintWeights_0=1) {
    alphaInputType=Bool
    alphaBoolBlend {blendInTime=0.1, blendOption=Linear, blendOutTime=0.1}
    boneToModify {boneName=dyn_backHair_D_03}
    constraintSetup [{targetBone {boneName=spine_05}}]
    constraintWeights [1]
   }
   n83 = "Constraint - dyn_backHair_E_03 (P)"(ComponentPose=@n82, bAlphaBoolEnabled=True, ConstraintWeights_0=1) {
    alphaInputType=Bool
    alphaBoolBlend {blendInTime=0.1, blendOption=Linear, blendOutTime=0.1}
    boneToModify {boneName=dyn_backHair_E_03}
    constraintSetup [{targetBone {boneName=spine_05}}]
    constraintWeights [1]
   }
   n84 = "Constraint - dyn_backHair_F_03 (P)"(ComponentPose=@n83, bAlphaBoolEnabled=True, ConstraintWeights_0=1) {
    alphaInputType=Bool
    alphaBoolBlend {blendInTime=0.1, blendOption=Linear, blendOutTime=0.1}
    boneToModify {boneName=dyn_backHair_F_03}
    constraintSetup [{targetBone {boneName=spine_05}}]
    constraintWeights [1]
   }
   n85 = "Constraint - dyn_backHair_A_04 (P)"(ComponentPose=@n84, bAlphaBoolEnabled=True, ConstraintWeights_0=1) {
    alphaInputType=Bool
    alphaBoolBlend {blendInTime=0.1, blendOption=Linear, blendOutTime=0.1}
    boneToModify {boneName=dyn_backHair_A_04}
    constraintSetup [{targetBone {boneName=spine_05}}]
    constraintWeights [1]
   }
   n86 = "Constraint - dyn_backHair_B_04 (P)"(ComponentPose=@n85, bAlphaBoolEnabled=True, ConstraintWeights_0=1) {
    alphaInputType=Bool
    alphaBoolBlend {blendInTime=0.1, blendOption=Linear, blendOutTime=0.1}
    boneToModify {boneName=dyn_backHair_B_04}
    constraintSetup [{targetBone {boneName=spine_05}}]
    constraintWeights [1]
   }
   n87 = "Constraint - dyn_backHair_C_04 (P)"(ComponentPose=@n86, bAlphaBoolEnabled=True, ConstraintWeights_0=1) {
    alphaInputType=Bool
    alphaBoolBlend {blendInTime=0.1, blendOption=Linear, blendOutTime=0.1}
    boneToModify {boneName=dyn_backHair_C_04}
    constraintSetup [{targetBone {boneName=spine_05}}]
    constraintWeights [1]
   }
   n88 = "Constraint - dyn_backHair_D_04 (P)"(ComponentPose=@n87, bAlphaBoolEnabled=True, ConstraintWeights_0=1) {
    alphaInputType=Bool
    alphaBoolBlend {blendInTime=0.1, blendOption=Linear, blendOutTime=0.1}
    boneToModify {boneName=dyn_backHair_D_04}
    constraintSetup [{targetBone {boneName=spine_05}}]
    constraintWeights [1]
   }
   n89 = "Constraint - dyn_backHair_E_04 (P)"(ComponentPose=@n88, bAlphaBoolEnabled=True, ConstraintWeights_0=1) {
    alphaInputType=Bool
    alphaBoolBlend {blendInTime=0.1, blendOption=Linear, blendOutTime=0.1}
    boneToModify {boneName=dyn_backHair_E_04}
    constraintSetup [{targetBone {boneName=spine_05}}]
    constraintWeights [1]
   }
   n90 = "Constraint - dyn_backHair_F_04 (P)"(ComponentPose=@n89, bAlphaBoolEnabled=True, ConstraintWeights_0=1) {
    alphaInputType=Bool
    alphaBoolBlend {blendInTime=0.1, blendOption=Linear, blendOutTime=0.1}
    boneToModify {boneName=dyn_backHair_F_04}
    constraintSetup [{targetBone {boneName=spine_05}}]
    constraintWeights [1]
   }
   n91 = "Transform (Modify) Bone - Bone: dyn_backhair_MSTR"!(ComponentPose=@n90, bAlphaBoolEnabled:bool=True, Translation="0,0,0", Rotation="0,0,0", Scale="0.7,1,1") -> (Pose) {
    alphaInputType=Bool
    rotationMode=BMM_Additive
    rotationSpace=BCS_BoneSpace
    scaleMode=BMM_Additive
    scaleSpace=BCS_BoneSpace
    translationMode=BMM_Additive
    translationSpace=BCS_BoneSpace
    alphaBoolBlend {blendInTime=0.1, blendOption=Linear, blendOutTime=0.1}
    boneToModify {boneName=dyn_backhair_MSTR}
    scale {x=0.7}
   }
   n92 = "Transform (Modify) Bone - Bone: dyn_backHair_A_MSTR"(ComponentPose=@n91, bAlphaBoolEnabled=True, Translation="0,0,0", Rotation="0,-10,0", Scale="1,1,1") {
    alphaInputType=Bool
    rotationMode=BMM_Additive
    rotationSpace=BCS_BoneSpace
    scaleMode=BMM_Additive
    scaleSpace=BCS_BoneSpace
    translationMode=BMM_Additive
    translationSpace=BCS_BoneSpace
    alphaBoolBlend {blendInTime=0.1, blendOption=Linear, blendOutTime=0.1}
    boneToModify {boneName=dyn_backHair_A_MSTR}
    rotation {yaw=-10}
   }
   n93 = "Transform (Modify) Bone - Bone: dyn_backHair_F_MSTR"(ComponentPose=@n92, bAlphaBoolEnabled=True, Translation="0,-2,-1", Rotation="0,10,0", Scale="1,1,1") {
    alphaInputType=Bool
    rotationMode=BMM_Additive
    rotationSpace=BCS_BoneSpace
    scaleMode=BMM_Additive
    scaleSpace=BCS_BoneSpace
    translationMode=BMM_Additive
    translationSpace=BCS_BoneSpace
    alphaBoolBlend {blendInTime=0.1, blendOption=Linear, blendOutTime=0.1}
    boneToModify {boneName=dyn_backHair_F_MSTR}
    rotation {yaw=10}
    translation {y=-2, z=-1}
   }
   Output Pose(Result=Blend Poses by bool(BlendPose_0=Component To Local(ComponentPose=@n93), BlendPose_1=Component To Local(ComponentPose=@n66), bActiveValue=Get bRigidBodyEnabled()) {blendType=Linear}) {layerGroup=None}
   n94 = Copy Pose From Mesh(SourceMeshComponent=Get Body_Mesh())
   n95 = Copy Pose From Mesh() {bCopyCurves=true, bUseAttachedParent=true}
   n96 = Look At - Bone: dyn_backhair_MSTR(
    ComponentPose=Local To Component(LocalPose=Layered blend per bone(BasePose=@n95, BlendPoses_0=@n94) {layerSetup [{branchFilters [{boneName=upperarm_l}, {boneName=upperarm_r}]}]})
    Alpha=1
   ) {
    alphaBoolBlend {blendOption=Linear}
    boneToModify {boneName=dyn_backhair_MSTR}
    lookAtLocation=<20.2874 -65.6244 -5>
    lookAtTarget {boneReference {boneName=pelvis}}
    lookAt_Axis {axis=<1 0>}
   }
   n97 = Look At - Bone: dyn_backhair_Main(ComponentPose=@n96, Alpha=1) {
    alphaBoolBlend {blendOption=Linear}
    boneToModify {boneName=dyn_backhair_Main}
    lookAtLocation=<20.2874 -65.6244 -5>
    lookAtTarget {boneReference {boneName=pelvis}}
    lookAt_Axis {axis=<1 0>}
   }
   n98 = Look At - Bone: dyn_backHair_A_MSTR(ComponentPose=@n97, Alpha=1) {
    alphaBoolBlend {blendOption=Linear}
    boneToModify {boneName=dyn_backHair_A_MSTR}
    lookAtLocation=<0 -20 -5>
    lookAtTarget {boneReference {boneName=pelvis}}
    lookAt_Axis {axis=<1 0>}
   }
   n99 = Look At - Bone: dyn_backHair_B_MSTR(ComponentPose=@n98, Alpha=1) {
    alphaBoolBlend {blendOption=Linear}
    boneToModify {boneName=dyn_backHair_B_MSTR}
    lookAtLocation=<0 -20 -2>
    lookAtTarget {boneReference {boneName=pelvis}}
    lookAt_Axis {axis=<1 0>}
   }
   n100 = Look At - Bone: dyn_backHair_E_MSTR(ComponentPose=@n99, Alpha=1) {
    alphaBoolBlend {blendOption=Linear}
    boneToModify {boneName=dyn_backHair_E_MSTR}
    lookAtLocation=<0 -20 2>
    lookAtTarget {boneReference {boneName=pelvis}}
    lookAt_Axis {axis=<1 0>}
   }
   n101 = Look At - Bone: dyn_backHair_C_MSTR(ComponentPose=@n100, Alpha=1) {
    alphaBoolBlend {blendOption=Linear}
    boneToModify {boneName=dyn_backHair_C_MSTR}
    lookAtLocation=<0 -20>
    lookAtTarget {boneReference {boneName=pelvis}}
    lookAt_Axis {axis=<1 0>}
   }
   n102 = Look At - Bone: dyn_backHair_F_MSTR(ComponentPose=@n101, Alpha=1) {
    alphaBoolBlend {blendOption=Linear}
    boneToModify {boneName=dyn_backHair_F_MSTR}
    lookAtLocation=<0 -20 5>
    lookAtTarget {boneReference {boneName=pelvis}}
    lookAt_Axis {axis=<1 0>}
   }
   n103 = Control Rig!(Source=Component To Local(ComponentPose=@n102), bAlphaBoolEnabled:bool=True, Emote:bool=Get bIsPlayingEmote()) -> (Pose) {
    alphaInputType=Bool
    alphaBoolBlend {blendOption=Linear}
    controlRigAssetReference {blueprintRigClass=&/M_MED_Imitator_FaceAcc_CtrlRig.M_MED_Imitator_FaceAcc_CtrlRig_C}
   }
   Save cached pose 'PreSim'(Pose=@n103) {cacheName=PreSim}
   n104 = RigidBody!(
    ComponentPose=Local To Component(LocalPose=Use cached pose 'PreSim'())
    Alpha
    OverrideWorldGravity:Vector="0,0,-980"
    ExternalForce
    ComponentLinearAccScale:Vector="0.4,0.4,0.5"
   ) -> (Pose) {
    bClampLinearTranslationLimitToRefPose=true
    bOverrideWorldGravity=true
    overridePhysicsAsset=&/M_MED_Imitator_FaceAcc_Physics
    simulationSpace=BaseBoneSpace
    alphaBoolBlend {blendOption=Linear}
    baseBoneRef {boneName=spine_05}
    componentAppliedLinearAccClamp=<2500 2500 2500>
    componentLinearAccScale=<0.4 0.4 0.5>
    componentLinearVelScale=<0.3 0.3 0.3>
    overrideWorldGravity {z=-980}
    simSpaceSettings {
     maxAngularAcceleration=1000
     maxAngularVelocity=3
     maxLinearAcceleration=700
     maxLinearVelocity=500
     velocityScaleZ=0
     worldAlpha=0.7
     externalLinearDragV=<0.2 0.2>
    }
   }
   n105 = Component To Local(ComponentPose=@n104)
   n106 = RigidBody(ComponentPose=Local To Component(LocalPose=Use cached pose 'PreSim'()), OverrideWorldGravity="0,0,-980", ComponentLinearAccScale="0,0,0") {
    bClampLinearTranslationLimitToRefPose=true
    bOverrideWorldGravity=true
    overridePhysicsAsset=&/M_MED_Imitator_Crouching_Physics
    simulationSpace=BaseBoneSpace
    alphaBoolBlend {blendOption=Linear}
    baseBoneRef {boneName=spine_05}
    componentAppliedLinearAccClamp=<2500 2500 2500>
    componentLinearAccScale=<0.4 0.4 0.5>
    componentLinearVelScale=<0.3 0.3 0.3>
    overrideWorldGravity {z=-980}
    simSpaceSettings {
     maxAngularAcceleration=1000
     maxAngularVelocity=3
     maxLinearAcceleration=700
     maxLinearVelocity=500
     velocityScaleZ=0
     worldAlpha=0.7
     externalLinearDragV=<0.2 0.2>
    }
   }
   n107 = Component To Local(ComponentPose=@n106)
   n108 = Get bIsCrouching()
   n109 = RigidBody(
    ComponentPose=Local To Component(LocalPose=Use cached pose 'PreSim'())
    OverrideWorldGravity="0,0,-980"
    ComponentLinearAccScale="0.4,0.4,0.5"
   ) {
    bClampLinearTranslationLimitToRefPose=true
    bOverrideWorldGravity=true
    overridePhysicsAsset=&/M_MED_Imitator_FE_FaceAcc_Physics
    simulationSpace=BaseBoneSpace
    alphaBoolBlend {blendOption=Linear}
    baseBoneRef {boneName=spine_05}
    componentAppliedLinearAccClamp=<2500 2500 2500>
    componentLinearAccScale=<0.4 0.4 0.5>
    componentLinearVelScale=<0.3 0.3 0.3>
    overrideWorldGravity {z=-980}
    simSpaceSettings {
     maxAngularAcceleration=1000
     maxAngularVelocity=3
     maxLinearAcceleration=700
     maxLinearVelocity=500
     velocityScaleZ=0
     worldAlpha=0.7
     externalLinearDragV=<0.2 0.2>
    }
   }
   n110 = "Transform (Modify) Bone - Bone: dyn_backHair_F_MSTR"!(ComponentPose=@n109, Alpha:float=1, Translation="0,-2,0", Rotation="0,0,0", Scale="1,1,1") -> (Pose) {rotationMode=BMM_Additive, translationMode=BMM_Additive, boneToModify {boneName=dyn_backHair_F_MSTR}, translation {y=-2}}
   n111 = RigidBody(ComponentPose=Local To Component(LocalPose=Use cached pose 'PreSim'()), OverrideWorldGravity="0,0,-980", ComponentLinearAccScale="0,0,0") {
    bClampLinearTranslationLimitToRefPose=true
    bOverrideWorldGravity=true
    overridePhysicsAsset=&M_MED_Imitator_FaceAcc_Physics
    simulationSpace=BaseBoneSpace
    alphaBoolBlend {blendOption=Linear}
    baseBoneRef {boneName=spine_05}
    componentAppliedLinearAccClamp=<2500 2500 2500>
    componentLinearAccScale=<0.4 0.4 0.5>
    overrideWorldGravity {z=-980}
    simSpaceSettings {
     maxAngularAcceleration=1000
     maxAngularVelocity=3
     maxLinearAcceleration=700
     maxLinearVelocity=500
     velocityScaleZ=0
     worldAlpha=0.7
     externalLinearDragV=<0.2 0.2>
    }
   }
   n112 = "Constraint - dyn_backhair_Main (R)"!(ComponentPose=@n111, bAlphaBoolEnabled:bool=True, ConstraintWeights_0=1) -> (Pose) {
    alphaInputType=Bool
    alphaBoolBlend {blendInTime=0.1, blendOption=Linear, blendOutTime=0.1}
    boneToModify {boneName=dyn_backhair_Main}
    constraintSetup [{transformType=Rotation, targetBone {boneName=spine_05}}]
    constraintWeights [1]
   }
   n113 = "Constraint - dyn_backHair_A_01 (P)"(ComponentPose=@n112, bAlphaBoolEnabled=True, ConstraintWeights_0=1) {
    alphaInputType=Bool
    alphaBoolBlend {blendInTime=0.1, blendOption=Linear, blendOutTime=0.1}
    boneToModify {boneName=dyn_backHair_A_01}
    constraintSetup [{targetBone {boneName=spine_05}}]
    constraintWeights [1]
   }
   n114 = "Constraint - dyn_backHair_B_01 (P)"(ComponentPose=@n113, bAlphaBoolEnabled=True, ConstraintWeights_0=1) {
    alphaInputType=Bool
    alphaBoolBlend {blendInTime=0.1, blendOption=Linear, blendOutTime=0.1}
    boneToModify {boneName=dyn_backHair_B_01}
    constraintSetup [{targetBone {boneName=spine_05}}]
    constraintWeights [1]
   }
   n115 = "Constraint - dyn_backHair_C_01 (P)"(ComponentPose=@n114, bAlphaBoolEnabled=True, ConstraintWeights_0=1) {
    alphaInputType=Bool
    alphaBoolBlend {blendInTime=0.1, blendOption=Linear, blendOutTime=0.1}
    boneToModify {boneName=dyn_backHair_C_01}
    constraintSetup [{targetBone {boneName=spine_05}}]
    constraintWeights [1]
   }
   n116 = "Constraint - dyn_backHair_D_01 (P)"(ComponentPose=@n115, bAlphaBoolEnabled=True, ConstraintWeights_0=1) {
    alphaInputType=Bool
    alphaBoolBlend {blendInTime=0.1, blendOption=Linear, blendOutTime=0.1}
    boneToModify {boneName=dyn_backHair_D_01}
    constraintSetup [{targetBone {boneName=spine_05}}]
    constraintWeights [1]
   }
   n117 = "Constraint - dyn_backHair_E_01 (P)"(ComponentPose=@n116, bAlphaBoolEnabled=True, ConstraintWeights_0=1) {
    alphaInputType=Bool
    alphaBoolBlend {blendInTime=0.1, blendOption=Linear, blendOutTime=0.1}
    boneToModify {boneName=dyn_backHair_E_01}
    constraintSetup [{targetBone {boneName=spine_05}}]
    constraintWeights [1]
   }
   n118 = "Constraint - dyn_backHair_F_01 (P)"(ComponentPose=@n117, bAlphaBoolEnabled=True, ConstraintWeights_0=1) {
    alphaInputType=Bool
    alphaBoolBlend {blendInTime=0.1, blendOption=Linear, blendOutTime=0.1}
    boneToModify {boneName=dyn_backHair_F_01}
    constraintSetup [{targetBone {boneName=spine_05}}]
    constraintWeights [1]
   }
   n119 = "Constraint - dyn_backHair_A_02 (P)"(ComponentPose=@n118, bAlphaBoolEnabled=True, ConstraintWeights_0=1) {
    alphaInputType=Bool
    alphaBoolBlend {blendInTime=0.1, blendOption=Linear, blendOutTime=0.1}
    boneToModify {boneName=dyn_backHair_A_02}
    constraintSetup [{targetBone {boneName=spine_05}}]
    constraintWeights [1]
   }
   n120 = "Constraint - dyn_backHair_B_02 (P)"(ComponentPose=@n119, bAlphaBoolEnabled=True, ConstraintWeights_0=1) {
    alphaInputType=Bool
    alphaBoolBlend {blendInTime=0.1, blendOption=Linear, blendOutTime=0.1}
    boneToModify {boneName=dyn_backHair_B_02}
    constraintSetup [{targetBone {boneName=spine_05}}]
    constraintWeights [1]
   }
   n121 = "Constraint - dyn_backHair_C_02 (P)"(ComponentPose=@n120, bAlphaBoolEnabled=True, ConstraintWeights_0=1) {
    alphaInputType=Bool
    alphaBoolBlend {blendInTime=0.1, blendOption=Linear, blendOutTime=0.1}
    boneToModify {boneName=dyn_backHair_C_02}
    constraintSetup [{targetBone {boneName=spine_05}}]
    constraintWeights [1]
   }
   n122 = "Constraint - dyn_backHair_D_02 (P)"(ComponentPose=@n121, bAlphaBoolEnabled=True, ConstraintWeights_0=1) {
    alphaInputType=Bool
    alphaBoolBlend {blendInTime=0.1, blendOption=Linear, blendOutTime=0.1}
    boneToModify {boneName=dyn_backHair_D_02}
    constraintSetup [{targetBone {boneName=spine_05}}]
    constraintWeights [1]
   }
   n123 = "Constraint - dyn_backHair_E_02 (P)"(ComponentPose=@n122, bAlphaBoolEnabled=True, ConstraintWeights_0=1) {
    alphaInputType=Bool
    alphaBoolBlend {blendInTime=0.1, blendOption=Linear, blendOutTime=0.1}
    boneToModify {boneName=dyn_backHair_E_02}
    constraintSetup [{targetBone {boneName=spine_05}}]
    constraintWeights [1]
   }
   n124 = "Constraint - dyn_backHair_F_02 (P)"(ComponentPose=@n123, bAlphaBoolEnabled=True, ConstraintWeights_0=1) {
    alphaInputType=Bool
    alphaBoolBlend {blendInTime=0.1, blendOption=Linear, blendOutTime=0.1}
    boneToModify {boneName=dyn_backHair_F_02}
    constraintSetup [{targetBone {boneName=spine_05}}]
    constraintWeights [1]
   }
   n125 = "Transform (Modify) Bone - Bone: dyn_backHair_A_MSTR"(ComponentPose=@n124, bAlphaBoolEnabled=True, Translation="0,0,0", Rotation="0,-30,0", Scale="1,1,1") {
    alphaInputType=Bool
    rotationMode=BMM_Additive
    rotationSpace=BCS_BoneSpace
    scaleMode=BMM_Additive
    scaleSpace=BCS_BoneSpace
    translationMode=BMM_Additive
    translationSpace=BCS_BoneSpace
    alphaBoolBlend {blendInTime=0.1, blendOption=Linear, blendOutTime=0.1}
    boneToModify {boneName=dyn_backHair_A_MSTR}
    rotation {yaw=-30}
   }
   n126 = "Transform (Modify) Bone - Bone: dyn_backHair_F_MSTR"!(ComponentPose=@n125, bAlphaBoolEnabled:bool=True, Translation="0,0,0", Rotation="0,30,0", Scale="1,1,1") -> (Pose) {
    alphaInputType=Bool
    rotationMode=BMM_Additive
    rotationSpace=BCS_BoneSpace
    scaleMode=BMM_Additive
    scaleSpace=BCS_BoneSpace
    translationMode=BMM_Additive
    translationSpace=BCS_BoneSpace
    alphaBoolBlend {blendInTime=0.1, blendOption=Linear, blendOutTime=0.1}
    boneToModify {boneName=dyn_backHair_F_MSTR}
    rotation {yaw=30}
   }
   n127 = Blend Poses by bool(
    BlendPose_0=Component To Local(ComponentPose=@n126)
    BlendPose_1=Blend Poses by bool(
     BlendPose_0=Component To Local(ComponentPose=@n110)
     BlendPose_1=Blend Poses by bool(BlendPose_0=@n107, BlendPose_1=@n105, bActiveValue=@n108) {blendType=Linear}
     bActiveValue=Get bIsInFrontEnd()
    ) {blendType=Linear}
    bActiveValue=Get isWearingBackpack() -> (isWearingBackpack:bool)
   ) {blendType=Linear}
   n128 = RigidBody(
    ComponentPose=Local To Component(LocalPose=Use cached pose 'PreSim'())
    OverrideWorldGravity="0,0,-980"
    ComponentLinearAccScale="0.4,0.4,0.5"
   ) {
    bClampLinearTranslationLimitToRefPose=true
    bOverrideWorldGravity=true
    overridePhysicsAsset=&/M_MED_Imitator_FaceAcc_Emote_Physics
    simulationSpace=BaseBoneSpace
    alphaBoolBlend {blendOption=Linear}
    baseBoneRef {boneName=spine_05}
    componentAppliedLinearAccClamp=<2500 2500 2500>
    componentLinearAccScale=<0.4 0.4 0.5>
    componentLinearVelScale=<0.3 0.3 0.3>
    overrideWorldGravity {z=-980}
    simSpaceSettings {
     maxAngularAcceleration=1000
     maxAngularVelocity=3
     maxLinearAcceleration=700
     maxLinearVelocity=500
     velocityScaleZ=0
     worldAlpha=0.7
     externalLinearDragV=<0.2 0.2>
    }
   }
   n129 = Component To Local(ComponentPose=@n128)
   n130 = Get bIsPlayingEmote()
   n131 = RigidBody(ComponentPose=Local To Component(LocalPose=Use cached pose 'PreSim'()), OverrideWorldGravity="0,0,-980", ComponentLinearAccScale="0,0,0") {
    bClampLinearTranslationLimitToRefPose=true
    bOverrideWorldGravity=true
    overridePhysicsAsset=&M_MED_Imitator_FaceAcc_Physics
    simulationSpace=BaseBoneSpace
    alphaBoolBlend {blendOption=Linear}
    baseBoneRef {boneName=spine_05}
    componentAppliedLinearAccClamp=<2500 2500 2500>
    componentLinearAccScale=<0.4 0.4 0.5>
    componentLinearVelScale=<0.3 0.3 0.3>
    overrideWorldGravity {z=-980}
    simSpaceSettings {
     maxAngularAcceleration=1000
     maxAngularVelocity=3
     maxLinearAcceleration=700
     maxLinearVelocity=500
     velocityScaleZ=0
     worldAlpha=0.7
     externalLinearDragV=<0.2 0.2>
    }
   }
   n132 = Get bIsSkyDiving()
   n133 = Get bIsInFrontEnd()
   Save cached pose 'PostSim'(Pose=Blend Poses by bool(
     BlendPose_0=Component To Local(ComponentPose=@n131)
     BlendPose_1=Blend Poses by bool(BlendPose_0=@n129, BlendPose_1=@n127, bActiveValue=@n130) {blendType=Linear}
     bActiveValue=AND Boolean(A=@n132, B=@n133)
    ) {blendType=Linear}) {cacheName=PostSim}
  ]}
 EventGraph {nodes [
   Event BlueprintUpdateAnimation()
   > n1 = Sequence!(execute) -> (then_0, then_1, then_2:Exec, then_3:Exec, then_4:Exec, then_5:Exec, then_6:Exec)
   n2 = GetPartSkeletalMeshComponent()
   n3 = Get Dynamic Material Instance(execute=@n1.then_4, SkeletalMeshComponent=@n2)
   > n4 = Branch(Condition=Get bIsPlayingEmote())
   > SetScalarParameterValue(self=@n3, ParameterName=Frame Value Open Mouth Default, Value=0)
   SetScalarParameterValue(execute=@n4.else, self=@n3, ParameterName=Frame Value Open Mouth Default, Value=8)
   n5 = Get Dynamic Material Instance(execute=@n1.then_5, SkeletalMeshComponent=@n2)
   > n6 = Branch(Condition=Get bIsPlayingEmote())
   > SetScalarParameterValue(self=@n5, ParameterName=Frame Value Eye Default, Value=0)
   SetScalarParameterValue(execute=@n6.else, self=@n5, ParameterName=Frame Value Eye Default, Value=2)
   n7 = Get Dynamic Material Instance(execute=@n1.then_6, SkeletalMeshComponent=@n2)
   > n8 = Branch(Condition=Get bIsPlayingEmote())
   > SetScalarParameterValue(self=@n7, ParameterName=Frame Value Eyebrow Default, Value=6)
   SetScalarParameterValue(execute=@n8.else, self=@n7, ParameterName=Frame Value Eyebrow Default, Value=7)
   n9 = Set Body_Mesh(execute=[@n17.then, @n1.then_0], Body_Mesh=GetOwnerSkeletalMeshComponent())
   > Set Anim_Inst(Anim_Inst:Anim Instance Object Reference=GetAnimInstance(self=@n9)) -> (Output_Get:Anim Instance Object Reference)
   n10 = Branch(execute=@n1.then_3, Condition=NOT Boolean(A=Get bIsPlayingEmote()))
   > Set isNotPlayingEmote(isNotPlayingEmote:bool=true) -> (Output_Get:bool)
   Set isNotPlayingEmote(execute=@n10.else, isNotPlayingEmote=false)
   n11 = Gravity Override State Function DEPRECATED(
    execute=@n1.then_1
    CharacterPartAnimInstance=Self-Reference()
    GameTimeInSeconds=GetGameTimeInSeconds()
    GravityOverrideParams=Get Gravity Override Params()
   )
   > Set GravityOverride(GravityOverride=@n11)
   n12 = Get bIsTargeting()
   n13 = Get bIsPlayingMeleeAnim()
   n14 = Get bIsSkyDiving()
   n15 = Get bIsCrouching()
   Set ADS Hair Velocity(
    execute=@n1.then_2
    ADS Hair Velocity:Vector=SelectVector(A="0.15,0.15,0.25", B="0.38,0.38,0.38", bPickA=OR Boolean!(A=@n12, B=@n13, C:bool=@n14, D:bool=@n15) -> (ReturnValue))
   ) -> (Output_Get:Vector)
   Event OnCharacterCustomizationLoaded()
   n16 = CheckForBackbling()
   > Set isWearingBackpack(isWearingBackpack:bool=@n16) -> (Output_Get:bool)
   > Set isWearingCape(isWearingCape=OR Boolean!(A=@n16.bIsBackblingACape, B=@n16.bIsBackblingLong) -> (ReturnValue))
   n17 = Event BlueprintInitializeAnimation()
   n18 = CheckForBackbling()
   Set isWearingBackpack(isWearingBackpack=@n18)
   > Set isWearingCape(isWearingCape=OR Boolean(A=@n18.bIsBackblingACape, B=@n18.bIsBackblingLong))
  ]}
]
variables [
 ADS Hair Velocity {type=Vector}
 Anim_Inst {type=AnimInstance, default=null}
 Body_Mesh {type=SkeletalMeshComponent}
 Gravity Override Params {type=GravityOverrideParamsStruct}
 GravityOverride {type=Vector}
 isNotPlayingEmote {type=boolean}
 isWearingBackpack {type=boolean}
 isWearingCape {type=boolean}
]
